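# Sound cue for the in-game UI sound that plays when a POI landmark is discovered.
# A short fixed delay leads into a default Modulator and a single landmark-discovered wave, with pitch lowered and volume raised slightly at cue level.
# It uses the non-critical in-game UI sound class and overrides concurrency to one instance, stopping the oldest.
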SoundCue /Game/Sounds/Fort_UI_Sounds/Gameplay/Fort_UI_InGame_POI_Landmark_Discovered_Cue
root=Delay_0
nodes [
 {id=Delay_0, children [Modulator_1], properties {delayMax=0.08, delayMin=0.08}}
 {id=Modulator_1, children [WavePlayer_1]}
 {id=WavePlayer_1, properties {soundWaveAssetPtr=&/InGame_POI_Landmark_Discovered}}
]
properties {
 bOverrideConcurrency=true
 maxDistance=2097152
 pitchMultiplier=0.9
 soundClassObject=&Fort_UI_InGame_NonCritical
 volumeMultiplier=1.1
 concurrencyOverrides {maxCount=1, resolutionRule=StopOldest}
}
