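# HUD widget that briefly announces a newly discovered named location or landmark, built on AthenaDiscoveredAreaOverlay.
# OnNewDiscovery sets text, stores the sound cue, picks the switcher layout and requests prioritized display.
# Becoming active plays BriefShow, whose start triggers the async-loaded cue (or default sounds); finishing or being stomped releases and collapses it.
# Mobile and top-UI playlists reposition the container, and playlist extension widgets fill the named-location underlay.
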WidgetBlueprint /Game/Athena/HUD/Discoverability/PlayerDiscovered
animations [BriefShow {
  binding_count=6
  length_seconds=5
  object=&self:BriefShow
  sequence {bindings [
    MainBox {
     id=#1
     display_name="CommonTextBlock_Description (5)"
     possessed_class=VerticalBox
     tracks [
      {
       type=MovieSceneFloatTrack
       display_name=Render Opacity
       sections [{type=MovieSceneFloatSection, channels [None {type=MovieSceneScriptingFloatChannel, keys [[0,0],[20.9447,1],[80.1753,1],[100,0]]}]}]
      }
      {
       type=MovieScene2DTransformTrack
       sections [{
         type=MovieScene2DTransformSection
         channels [
          Scale.X {type=MovieSceneScriptingFloatChannel, keys [[0,3],[20.9447,1]]}
          Scale.Y {type=MovieSceneScriptingFloatChannel, keys [[0,3],[20.9447,1]]}
         ]
        }]
      }
     ]
    }
    CommonTextBlock_LandmarkName {
     id=#2
     possessed_class=CommonTextBlock
     tracks [{
       type=MovieScene2DTransformTrack
       sections [{
         type=MovieScene2DTransformSection
         channels [
          Scale.X {type=MovieSceneScriptingFloatChannel, keys [[6,1.2189],[39,1]]}
          Scale.Y {type=MovieSceneScriptingFloatChannel, keys [[6,1.2189],[39,1]]}
         ]
        }]
      }]
    }
    Image_BulletLeft {
     id=#3
     possessed_class=Image
     tracks [{
       type=MovieSceneWidgetMaterialTrack
       sections [{
         type=MovieSceneParameterSection
         channels [
          Density {type=MovieSceneScriptingFloatChannel, keys [[0,1.4262],[33.382,2.7441]]}
          Radius {type=MovieSceneScriptingFloatChannel, keys [[0,0.2865],[33.382,1]]}
         ]
        }]
      }]
    }
    CommonTextBlock_Description {
     id=#4
     possessed_class=CommonTextBlock
     tracks [{
       type=MovieScene2DTransformTrack
       sections [{
         type=MovieScene2DTransformSection
         channels [
          Scale.X {type=MovieSceneScriptingFloatChannel, keys [[8.3693,1.2492],[44.923,1]]}
          Scale.Y {type=MovieSceneScriptingFloatChannel, keys [[8.3693,1.2492],[44.923,1]]}
         ]
        }]
      }]
    }
    Overlay_Underlay {
     id=#5
     possessed_class=Overlay
     tracks [
      {
       type=MovieSceneFloatTrack
       display_name=Render Opacity
       sections [{type=MovieSceneFloatSection, channels [None {type=MovieSceneScriptingFloatChannel, keys [[0,0],[20.9447,1],[80.1753,1],[100,0]]}]}]
      }
      {
       type=MovieScene2DTransformTrack
       sections [{
         type=MovieScene2DTransformSection
         channels [
          Scale.X {type=MovieSceneScriptingFloatChannel, keys [[0,3],[20.9447,1]]}
          Scale.Y {type=MovieSceneScriptingFloatChannel, keys [[0,3],[20.9447,1]]}
         ]
        }]
      }
     ]
    }
    Image_BulletRight {
     id=#6
     possessed_class=Image
     tracks [{
       type=MovieSceneWidgetMaterialTrack
       sections [{
         type=MovieSceneParameterSection
         channels [
          Density {type=MovieSceneScriptingFloatChannel, keys [[0,1.4262],[33.382,2.7441]]}
          Radius {type=MovieSceneScriptingFloatChannel, keys [[0,0.2865],[33.382,1]]}
         ]
        }]
      }]
    }
   ]}
 }]
graphs [
 EndOfShow {nodes [
   EndOfShow() {is=title K2Node_FunctionEntry}
   > SetVisibility(InVisibility=Collapsed)
   SetRenderOpacity(self=Get WidgetSwitcher_LandmarkOrPOI() -> (WidgetSwitcher_LandmarkOrPOI:Widget Switcher Object Reference), InOpacity=0)
  ]}
 EventGraph {nodes [
   n1 = PlaylistDataReadyAsync!(execute:Exec) -> (
    then:Exec
    OnReady:Exec
    GameState:Fort Game State Athena Object Reference
    Playlist:Fort Playlist Object Reference
    PlaylistContextTags:"Gameplay Tag Container Structure (by ref)"
   ) {is=title K2Node_AsyncAction}
   Branch(
    execute=@n1.OnReady
    Condition=HasAnyTags!(
     TagContainer:"Gameplay Tag Container Structure (by ref)"=GetAthenaPlaylistContextTags!(self:Fort Game State Athena Object Reference=@n1) -> (ReturnValue:Gameplay Tag Container Structure)
     OtherContainer:"Gameplay Tag Container Structure (by ref)"=Get RelevantPlaylistTags() -> (RelevantPlaylistTags:Gameplay Tag Container Structure)
     bExactMatch:bool=false
    ) -> (ReturnValue:bool)
   )
   > SetRenderTranslation(self=Get Overlay_Container() -> (Overlay_Container:Overlay Object Reference), Translation=(X=0,Y=-20))
   n2 = Brief Show Finished() {is=title K2Node_CustomEvent}
   SetHasActiveContent!(execute=[@n16.then, @n2.then], self:Fort Prioritized Widget Interface, bNewActive:bool=false) -> (then)
   > EndOfShow() {is=title K2Node_CallFunction}
   n3 = Play Brief Show Sound() {is=title K2Node_CustomEvent}
   n4 = Get SoundCue() -> (SoundCue:Sound Cue Soft Object Reference)
   > n5 = Branch(Condition=IsValidSoftObjectReference(SoftObjectReference=@n4))
   > n6 = Async Load Asset(Asset=@n4)
   n7 = Cast To SoundCue!(execute=@n6.Completed, Object=@n6) -> (then, CastFailed, AsSound Cue:Sound Cue Object Reference) {is=title K2Node_DynamicCast}
   > PlaySound2D(Sound=@n7)
   n8 = Get WidgetSwitcher_LandmarkOrPOI()
   n9 = Branch(
    execute=@n5.else
    Condition="Equal (Integer)"(A=Get ActiveWidgetIndex!(self:Widget Switcher Object Reference=@n8) -> (ActiveWidgetIndex:int=0), B=0) {is=title K2Node_CallFunction}
   )
   > PlaySound2D(Sound=/Game/Sounds/Fort_UI_Sounds/Gameplay/Fort_UI_InGame_POI_Discovered_Cue)
   PlaySound2D(execute=@n9.else, Sound=/Game/Sounds/Fort_UI_Sounds/Gameplay/Fort_UI_InGame_POI_Landmark_Discovered_Cue)
   Event Construct()
   > n10 = Branch(Condition=IsMobileGame())
   n11 = Get Overlay_Container()
   n12 = Get!(self:Widget Object Reference=@n11) -> (
    RenderTransform_Translation:Vector 2D Structure
    RenderTransform_Scale:Vector 2D Structure
    RenderTransform_Shear:Vector 2D Structure
    RenderTransform_Angle:float=~
   )
   > n13 = SetRenderTransform!(
    execute
    self:Widget Object Reference=@n11
    InTransform_Translation:Vector 2D Structure=MakeVector2D(X=BreakVector2D(InVec=@n12), Y=200)
    InTransform_Scale:Vector 2D Structure=@n12.RenderTransform_Scale
    InTransform_Shear:Vector 2D Structure=@n12.RenderTransform_Shear
    InTransform_Angle:float=@n12.RenderTransform_Angle
   ) -> (then)
   EndOfShow(execute=[@n13.then, @n10.else])
   > BindToAnimationFinished(Animation=Get BriefShow() -> (BriefShow:Widget Animation Object Reference), Delegate=Create Event())
   > BindToAnimationStarted!(execute, self:User Widget Object Reference, Animation:Widget Animation Object Reference=Get BriefShow(), Delegate:Delegate=@n3) -> (then)
   n14 = Event OnNewDiscovery!() -> (
    OutputDelegate
    then
    LocalizedLocationName:"Text (by ref)"
    DiscoveredSoundCue:"Sound Cue Soft Object Reference (by ref)"
    bIsNamedLocation:bool
    bIsDiscovered:bool
   ) {is=title K2Node_Event}
   > Set SoundCue(SoundCue:Sound Cue Soft Object Reference=@n14.DiscoveredSoundCue) -> (Output_Get:Sound Cue Soft Object Reference)
   > "SetText (Text)"(
    self=[
     Get MainLocationText() -> (MainLocationText:Common Text Object Reference)
     Get CommonTextBlock_LandmarkName() -> (CommonTextBlock_LandmarkName:Common Text Object Reference)
    ]
    InText=@n14.LocalizedLocationName
   )
   > SetHasActiveContent(bNewActive=true)
   > SetActiveWidgetIndex(self=Get WidgetSwitcher_LandmarkOrPOI(), Index=Select(Option 0:int=1, Option 1:int=0, Index=@n14.bIsNamedLocation) -> (ReturnValue:int))
   n15 = Event OnStompFailed() {is=title K2Node_Event}
   Branch(execute=[@n15.then, @n17.then], Condition=IsAnimationPlaying(InAnimation=Get BriefShow()))
   > n16 = StopAnimation(InAnimation=Get BriefShow())
   n17 = Event OnStompedByOtherWidget() {is=title K2Node_Event}
   Event On Become Active() {is=title K2Node_Event}
   > Parent: On Become Active() {is=title K2Node_CallParentFunction}
   > SetVisibility(InVisibility=HitTestInvisible)
   > PlayAnimation(InAnimation=Get BriefShow(), PlaybackSpeed=1.4)
   n18 = Event Handle Playlist Extension Widget Created!() -> (OutputDelegate, then, Widget:User Widget Object Reference, SlotType:EUIExtensionSlot Enum=Invalid) {is=title K2Node_Event}
   > n19 = AddChildToOverlay!(
    execute
    self:Overlay Object Reference=Get Overlay_Underlay() -> (Overlay_Underlay:Overlay Object Reference)
    Content:Widget Object Reference=@n18.Widget
   ) -> (then, ReturnValue:Overlay Slot Object Reference)
   > SetVerticalAlignment(self=@n19, InVerticalAlignment=VAlign_Fill)
   > SetHorizontalAlignment(self=@n19, InHorizontalAlignment=HAlign_Fill)
   Event PreConstruct()
   Event Tick()
   Reroute Node(InputPin:bool=@n14.bIsNamedLocation) -> (OutputPin:bool)
  ]}
]
variables [
 RelevantPlaylistTags {type=GameplayTagContainer, default {gameplayTags [{tagName=Athena.Playlist.HasTopUI}]}}
 SoundCue {type=SoundCue, default=&Fort_UI_InGame_POI_Discovered_Cue, class=~, properties=~}
]
widgets {
 info {inheritedWidgetCount=1, parentClass=/Script/FortniteUI.AthenaDiscoveredAreaOverlay, rootWidgetClass=&Overlay, widgetCount=15}
 widgets [
  Overlay_Container {
   widgetClassPath=&Overlay
   bIsVariable=true
   widget=Overlay_Container
   properties {renderTransform {angle=0, scale=<1 1>, shear=<0 0>, translation=<0 -110>}}
  }
  WidgetSwitcher_LandmarkOrPOI {
   widgetClassPath=&WidgetSwitcher
   bIsVariable=true
   parent=Overlay_Container
   widget=WidgetSwitcher_LandmarkOrPOI
   properties {renderTransformPivot=<0.5 0>}
   slot {class=OverlaySlot, properties {horizontalAlignment=HAlign_Center}}
  }
  MainBox {
   widgetClassPath=&VerticalBox
   bIsVariable=true
   parent=WidgetSwitcher_LandmarkOrPOI
   widget=MainBox
   properties {renderTransformPivot=<0.5 0>}
   slot {class=WidgetSwitcherSlot}
  }
  Overlay_1 {widgetClassPath=&Overlay, parent=MainBox, widget=Overlay_1, slot {class=VerticalBoxSlot}}
  Overlay_Underlay {
   widgetClassPath=&Overlay
   bIsVariable=true
   parent=Overlay_1
   widget=Overlay_Underlay
   slot {class=OverlaySlot, properties {horizontalAlignment=HAlign_Fill, verticalAlignment=VAlign_Fill}}
  }
  HorizontalBox_1 {
   widgetClassPath=&HorizontalBox
   parent=Overlay_1
   widget=HorizontalBox_1
   slot {class=OverlaySlot, properties {horizontalAlignment=HAlign_Center}}
  }
  Image_BulletLeft {
   widgetClassPath=&Image
   bIsVariable=true
   parent=HorizontalBox_1
   widget=Image_BulletLeft
   properties {
    brush {
     bIsDynamicallyLoaded=false
     drawAs=Image
     imageType=NoImage
     mirroring=NoMirror
     resourceName=None
     resourceObject=/FortUILibrary/Foundation/Art_ToolKit/AngledBoxes/M_UI_BulletHud
     tiling=NoTile
     imageSize=<10 10>
     margin=<ltrb 0 0 0 0>
     outlineSettings {
      bUseBrushTransparency=false
      roundingType=HalfHeightRadius
      width=0
      color {colorUseRule=UseColor_Specified, specifiedColor=<rgba 0 0 0 0>}
      cornerRadii=<0 0 0 0>
     }
     tintColor {colorUseRule=UseColor_Specified, specifiedColor=<rgba 1 1 1 1>}
     uVRegion {bIsValid=false, max=<0 0>, min=<0 0>}
    }
    renderTransform {angle=0, scale=<1 1>, shear=<-8 0>, translation=<0 0>}
   }
   slot {class=HorizontalBoxSlot, properties {verticalAlignment=VAlign_Center}}
  }
  MainLocationText {
   widgetClassPath=&CommonTextBlock
   bIsVariable=true
   parent=HorizontalBox_1
   widget=MainLocationText
   properties {
    bAutoCollapseWithEmptyText=true
    justification=Center
    text=Tilted Town
    textTransformPolicy=ToUpper
    font {
     bForceMonospaced=false
     bMaterialIsStencil=false
     fontMaterial=null
     fontObject=&Burbank
     letterSpacing=0
     monospacedWidth=1
     size=40
     skewAmount=0
     typefaceFontName=BurbankRegular-Black
     outlineSettings {
      bApplyOutlineToDropShadows=false
      bMiteredCorners=false
      bSeparateFillAlpha=true
      outlineMaterial=&M-FontShadow
      outlineSize=2
      outlineColor=<rgba 0 0 0 0.5>
     }
    }
    renderTransform {angle=0, scale=<1 1>, shear=<-12 0>, translation=<0 0>}
    shadowOffset=<0 0>
   }
   slot {class=HorizontalBoxSlot, properties {padding=<ltrb 8 0 8 -8>}}
  }
  Image_BulletRight {
   widgetClassPath=&Image
   bIsVariable=true
   parent=HorizontalBox_1
   widget=Image_BulletRight
   properties {
    brush {
     bIsDynamicallyLoaded=false
     drawAs=Image
     imageType=NoImage
     mirroring=NoMirror
     resourceName=None
     resourceObject=&M_UI_BulletHud
     tiling=NoTile
     imageSize=<10 10>
     margin=<ltrb 0 0 0 0>
     outlineSettings {
      bUseBrushTransparency=false
      roundingType=HalfHeightRadius
      width=0
      color {colorUseRule=UseColor_Specified, specifiedColor=<rgba 0 0 0 0>}
      cornerRadii=<0 0 0 0>
     }
     tintColor {colorUseRule=UseColor_Specified, specifiedColor=<rgba 1 1 1 1>}
     uVRegion {bIsValid=false, max=<0 0>, min=<0 0>}
    }
    renderTransform {angle=180, scale=<-1 1>, shear=<-10 0>, translation=<0 0>}
   }
   slot {class=HorizontalBoxSlot, properties {verticalAlignment=VAlign_Center}}
  }
  CommonTextBlock_Description {
   widgetClassPath=&CommonTextBlock
   parent=MainBox
   widget=CommonTextBlock_Description
   properties {
    bAutoCollapseWithEmptyText=true
    justification=Center
    text=New Location Discovered
    textTransformPolicy=ToUpper
    colorAndOpacity {colorUseRule=UseColor_Specified, specifiedColor=<rgba 1 0.5833 0 1>}
    font {
     bForceMonospaced=false
     bMaterialIsStencil=false
     fontMaterial=null
     fontObject=&BurbankSmall
     letterSpacing=0
     monospacedWidth=1
     size=20
     skewAmount=0
     typefaceFontName=Black
     outlineSettings {
      bApplyOutlineToDropShadows=false
      bMiteredCorners=false
      bSeparateFillAlpha=true
      outlineMaterial=&M-FontShadow
      outlineSize=2
      outlineColor=<rgba 0 0 0 0.5>
     }
    }
    renderTransform {angle=0, scale=<1 1>, shear=<-12 0>, translation=<0 0>}
    renderTransformPivot=<0.5 0>
    shadowOffset=<0 0>
   }
   slot {class=VerticalBoxSlot, properties {horizontalAlignment=HAlign_Center, padding=<ltrb 2 0 0 0>}}
  }
  MainBox_Landmark {
   widgetClassPath=&VerticalBox
   bIsVariable=true
   parent=WidgetSwitcher_LandmarkOrPOI
   widget=MainBox_Landmark
   properties {renderTransformPivot=<0.5 0>}
   slot {class=WidgetSwitcherSlot}
  }
  HorizontalBox {
   widgetClassPath=&HorizontalBox
   parent=MainBox_Landmark
   widget=HorizontalBox
   slot {class=VerticalBoxSlot, properties {horizontalAlignment=HAlign_Center}}
  }
  MainLocationText_1 {
   widgetClassPath=&CommonTextBlock
   bIsVariable=true
   parent=HorizontalBox
   widget=MainLocationText_1
   properties {
    bAutoCollapseWithEmptyText=true
    justification=Center
    text="New Landmark!"
    textTransformPolicy=ToUpper
    font {
     bForceMonospaced=false
     bMaterialIsStencil=false
     fontMaterial=null
     fontObject=&Burbank
     letterSpacing=0
     monospacedWidth=1
     size=30
     skewAmount=0
     typefaceFontName=BurbankRegular-Black
     outlineSettings {
      bApplyOutlineToDropShadows=false
      bMiteredCorners=false
      bSeparateFillAlpha=false
      outlineMaterial=null
      outlineSize=0
      outlineColor=<rgba 0 0 0 1>
     }
    }
    renderTransform {angle=0, scale=<1 1>, shear=<-12 0>, translation=<0 0>}
    shadowOffset=<0 0>
   }
   slot {class=HorizontalBoxSlot, properties {padding=<ltrb 8 0 8 0>}}
  }
  CommonTextBlock_LandmarkName {
   widgetClassPath=&CommonTextBlock
   bIsVariable=true
   parent=MainBox_Landmark
   widget=CommonTextBlock_LandmarkName
   properties {
    bAutoCollapseWithEmptyText=true
    justification=Center
    text=Landmark Name
    textTransformPolicy=ToUpper
    font {
     bForceMonospaced=false
     bMaterialIsStencil=false
     fontMaterial=null
     fontObject=&BurbankSmall
     letterSpacing=0
     monospacedWidth=1
     size=22
     skewAmount=0
     typefaceFontName=Black
     outlineSettings {
      bApplyOutlineToDropShadows=false
      bMiteredCorners=false
      bSeparateFillAlpha=false
      outlineMaterial=null
      outlineSize=0
      outlineColor=<rgba 0 0 0 1>
     }
    }
    renderTransformPivot=<0.5 0>
    shadowOffset=<0 0>
   }
   slot {class=VerticalBoxSlot, properties {horizontalAlignment=HAlign_Center}}
  }
  TutorialHighlight {widgetClassPath=&TutorialHighlight}
 ]
}
