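# WBP_Athena_Objectives_Entry is a HUD widget showing one match objective with its state text, reward icon and optional sub-entries.
# UMG animations cover intros, reward reveal, success, failure, timeout, info updates and stage transitions, using slides, fades and a material-tinted shine.
# Optional-entry events push progress text into the main objective and advance the current multi-step entry.
# A completion check compares the current index against the total to set "Quest Completed" transition text.
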WidgetBlueprint /Game/Athena/HUD/Discoverability/MatchObjectivesWidgets/WBP_Athena_Objectives_Entry
animations [
 Anim_Intro_Started {
  binding_count=6
  length_seconds=5
  object=&self:Anim_Intro_Started
  sequence {
   bindings [
    DynamicEntryBox_OptionalEntries {
     id=#1
     possessed_class=FortDynamicEntryBox
     tracks [
      {
       type=MovieSceneFloatTrack
       display_name=Render Opacity
       sections [{
         type=MovieSceneFloatSection
         end_frame=300
         start_frame=0
         channels [None {type=MovieSceneScriptingFloatChannel, keys [[0,0],[61,0],[83,1]]}]
        }]
      }
      {
       type=MovieScene2DTransformTrack
       sections [{
         type=MovieScene2DTransformSection
         end_frame=299
         start_frame=0
         channels [
          Translation.X {type=MovieSceneScriptingFloatChannel, keys [[61,-300],[83,0]]}
          Translation.Y {type=MovieSceneScriptingFloatChannel}
          Angle {type=MovieSceneScriptingFloatChannel}
          Scale.X {type=MovieSceneScriptingFloatChannel}
          Scale.Y {type=MovieSceneScriptingFloatChannel}
          Shear.X {type=MovieSceneScriptingFloatChannel}
          Shear.Y {type=MovieSceneScriptingFloatChannel}
         ]
        }]
      }
     ]
    }
    OverlaySlot_1 {id=#2, parent_id=#3, possessed_class=OverlaySlot, tracks []}
    Image_fx_Shine {
     id=#4
     possessed_class=Image
     tracks [
      {
       type=MovieSceneFloatTrack
       display_name=Render Opacity
       sections [{
         type=MovieSceneFloatSection
         end_frame=299
         start_frame=0
         channels [None {type=MovieSceneScriptingFloatChannel, keys [[0,0],[7,1],[11,1],[21,0]]}]
        }]
      }
      {
       type=MovieScene2DTransformTrack
       sections [{
         type=MovieScene2DTransformSection
         end_frame=300
         start_frame=0
         channels [
          Translation.X {type=MovieSceneScriptingFloatChannel, keys [[0,-331.1956],[18,0.9143]]}
          Translation.Y {type=MovieSceneScriptingFloatChannel, const [0,1]}
          Angle {type=MovieSceneScriptingFloatChannel, const [0,1]}
          Scale.X {type=MovieSceneScriptingFloatChannel, keys [[0,1],[18,1.6777]]}
          Scale.Y {type=MovieSceneScriptingFloatChannel, keys [[0,1.3],[18,1]]}
          Shear.X {type=MovieSceneScriptingFloatChannel, const [0,1]}
          Shear.Y {type=MovieSceneScriptingFloatChannel, const [0,1]}
         ]
        }]
      }
      {
       type=MovieSceneWidgetMaterialTrack
       sections [{
         type=MovieSceneParameterSection
         end_frame=300
         start_frame=-1
         channels [
          ColorR {type=MovieSceneScriptingFloatChannel, keys [[0,1],[7,1],[16,0.0666]]}
          ColorG {type=MovieSceneScriptingFloatChannel, keys [[0,1],[7,1],[16,0.5271]]}
          ColorB {type=MovieSceneScriptingFloatChannel, keys [[0,1],[7,1],[16,0.956]]}
          ColorA {type=MovieSceneScriptingFloatChannel, const [1,1]}
         ]
        }]
      }
     ]
    }
    Overlay_PrimaryObjective {
     id=#5
     possessed_class=Overlay
     tracks [
      {
       type=MovieSceneFloatTrack
       display_name=Render Opacity
       sections [{
         type=MovieSceneFloatSection
         end_frame=299
         start_frame=0
         channels [None {type=MovieSceneScriptingFloatChannel, keys [[0,0],[26.999,0],[58,1]]}]
        }]
      }
      {
       type=MovieScene2DTransformTrack
       sections [{
         type=MovieScene2DTransformSection
         end_frame=299
         start_frame=0
         channels [
          Translation.X {type=MovieSceneScriptingFloatChannel, keys [[26.999,-300],[49,0]]}
          Translation.Y {type=MovieSceneScriptingFloatChannel}
          Angle {type=MovieSceneScriptingFloatChannel}
          Scale.X {type=MovieSceneScriptingFloatChannel}
          Scale.Y {type=MovieSceneScriptingFloatChannel}
          Shear.X {type=MovieSceneScriptingFloatChannel}
          Shear.Y {type=MovieSceneScriptingFloatChannel}
         ]
        }]
      }
     ]
    }
    Overlay_Root {
     id=#6
     possessed_class=Overlay
     tracks [{
       type=MovieSceneFloatTrack
       display_name=Render Opacity
       sections [{type=MovieSceneFloatSection, end_frame=0, start_frame=-1, channels [None {type=MovieSceneScriptingFloatChannel, const [1,1]}]}]
      }]
    }
    Text_ObjState {
     id=#3
     display_name=Text_ObjectiveState
     possessed_class=CommonTextBlock
     tracks [
      {
       type=MovieSceneFloatTrack
       display_name=Render Opacity
       sections [{
         type=MovieSceneFloatSection
         end_frame=299
         start_frame=-1
         channels [None {type=MovieSceneScriptingFloatChannel, keys [[-0.001,0.0168],[14,0],[28,1],[279,1],[299,0]]}]
        }]
      }
      {
       type=MovieScene2DTransformTrack
       sections [{
         type=MovieScene2DTransformSection
         end_frame=299
         start_frame=-1
         channels [
          Translation.X {type=MovieSceneScriptingFloatChannel, keys [[-0.001,-200],[14,-200],[28,0],[279,0],[299,-200]]}
          Translation.Y {type=MovieSceneScriptingFloatChannel, const [0,3]}
          Angle {type=MovieSceneScriptingFloatChannel}
          Scale.X {type=MovieSceneScriptingFloatChannel}
          Scale.Y {type=MovieSceneScriptingFloatChannel}
          Shear.X {type=MovieSceneScriptingFloatChannel}
          Shear.Y {type=MovieSceneScriptingFloatChannel}
         ]
        }]
      }
     ]
    }
   ]
   display_rate {numerator=60}
  }
 }
 Anim_Intro_Reward {
  binding_count=3
  length_seconds=2
  object=&self:Anim_Intro_Reward
  sequence {bindings [
    Overlay_Root {
     id=#7
     possessed_class=Overlay
     tracks [{
       type=MovieSceneFloatTrack
       display_name=Render Opacity
       sections [{type=MovieSceneFloatSection, end_frame=0, start_frame=0, channels [None {type=MovieSceneScriptingFloatChannel, const [1,1]}]}]
      }]
    }
    WBP_Athena_Objectives_Icon {
     id=#8
     possessed_class=WBP_Athena_Objectives_Icon_C
     tracks [
      {
       type=MovieSceneFloatTrack
       display_name=Render Opacity
       sections [{type=MovieSceneFloatSection, end_frame=40, start_frame=0, channels [None {type=MovieSceneScriptingFloatChannel, keys [[0,0],[15,1],[30,1]]}]}]
      }
      {
       type=MovieScene2DTransformTrack
       sections [{
         type=MovieScene2DTransformSection
         end_frame=40
         start_frame=0
         channels [
          Angle {type=MovieSceneScriptingFloatChannel, const [0,1]}
          Scale.X {type=MovieSceneScriptingFloatChannel, keys [[0,0.6],[7,1],[13,1.05],[17,1]]}
          Scale.Y {type=MovieSceneScriptingFloatChannel, keys [[0,0.6],[7,1],[13,1.05],[17,1]]}
          Shear.X {type=MovieSceneScriptingFloatChannel, const [0,1]}
          Shear.Y {type=MovieSceneScriptingFloatChannel, const [0,1]}
         ]
        }]
      }
     ]
    }
    Overlay_RewardIcon {
     id=#9
     possessed_class=Overlay
     tracks [
      {
       type=MovieSceneFloatTrack
       display_name=Render Opacity
       sections [{type=MovieSceneFloatSection, end_frame=40, start_frame=0, channels [None {type=MovieSceneScriptingFloatChannel, keys [[0,0],[15,1],[30,1]]}]}]
      }
      {
       type=MovieScene2DTransformTrack
       sections [{
         type=MovieScene2DTransformSection
         end_frame=15
         start_frame=0
         channels [
          Translation.X {type=MovieSceneScriptingFloatChannel, const [0,1]}
          Translation.Y {type=MovieSceneScriptingFloatChannel, const [0,1]}
          Angle {type=MovieSceneScriptingFloatChannel, const [0,1]}
          Scale.X {type=MovieSceneScriptingFloatChannel, default=1}
          Scale.Y {type=MovieSceneScriptingFloatChannel, default=1}
          Shear.X {type=MovieSceneScriptingFloatChannel, const [0,1]}
          Shear.Y {type=MovieSceneScriptingFloatChannel, const [0,1]}
         ]
        }]
      }
      {
       type=MovieSceneByteTrack
       display_name=Visibility
       sections [{type=MovieSceneByteSection, end_frame=0, start_frame=0, channels [None {type=MovieSceneScriptingByteChannel, const [4,1]}]}]
      }
     ]
    }
   ]}
 }
 Anim_Succeed_Intro {
  binding_count=4
  length_seconds=3.05
  object=&self:Anim_Succeed_Intro
  sequence {bindings [
    Image_fx_Shine {
     id=#10
     possessed_class=Image
     tracks [
      {
       type=MovieSceneWidgetMaterialTrack
       sections [{
         type=MovieSceneParameterSection
         end_frame=61
         start_frame=-1
         channels [
          ColorR {type=MovieSceneScriptingFloatChannel, keys [[0,1],[2.8467,0.0666]]}
          ColorG {type=MovieSceneScriptingFloatChannel, keys [[0,1],[2.8467,0.5271]]}
          ColorB {type=MovieSceneScriptingFloatChannel, keys [[0,1],[2.8467,0.956]]}
          ColorA {type=MovieSceneScriptingFloatChannel, const [1,1]}
         ]
        }]
      }
      {
       type=MovieSceneFloatTrack
       display_name=Render Opacity
       sections [{
         type=MovieSceneFloatSection
         end_frame=60
         start_frame=0
         channels [None {type=MovieSceneScriptingFloatChannel, keys [[0,0],[1.4233,1],[2.8467,1],[5.8967,0]]}]
        }]
      }
      {
       type=MovieScene2DTransformTrack
       sections [{
         type=MovieScene2DTransformSection
         end_frame=61
         start_frame=0
         channels [
          Translation.X {type=MovieSceneScriptingFloatChannel, keys [[0,-331.1956],[3.66,0.9143]]}
          Translation.Y {type=MovieSceneScriptingFloatChannel, const [0,1]}
          Angle {type=MovieSceneScriptingFloatChannel, const [0,1]}
          Scale.X {type=MovieSceneScriptingFloatChannel, keys [[0,1],[3.66,1.6777]]}
          Scale.Y {type=MovieSceneScriptingFloatChannel, keys [[0,1.3],[3.66,1]]}
          Shear.X {type=MovieSceneScriptingFloatChannel, const [0,1]}
          Shear.Y {type=MovieSceneScriptingFloatChannel, const [0,1]}
         ]
        }]
      }
     ]
    }
    Image_fx_Reward_Glow {
     id=#11
     possessed_class=Image
     tracks [
      {
       type=MovieScene2DTransformTrack
       sections [{
         type=MovieScene2DTransformSection
         end_frame=61
         start_frame=0
         channels [
          Translation.X {type=MovieSceneScriptingFloatChannel, const [5,1]}
          Translation.Y {type=MovieSceneScriptingFloatChannel, const [1,1]}
          Angle {type=MovieSceneScriptingFloatChannel, const [0,1]}
          Scale.X {type=MovieSceneScriptingFloatChannel, keys [[0,1],[11.445,1],[19.445,2.6457]]}
          Scale.Y {type=MovieSceneScriptingFloatChannel, keys [[0,1],[11.445,1],[19.445,6.811]]}
          Shear.X {type=MovieSceneScriptingFloatChannel, const [0,1]}
          Shear.Y {type=MovieSceneScriptingFloatChannel, const [0,1]}
         ]
        }]
      }
      {
       type=MovieSceneFloatTrack
       display_name=Render Opacity
       sections [{
         type=MovieSceneFloatSection
         end_frame=24
         start_frame=0
         channels [None {type=MovieSceneScriptingFloatChannel, keys [[0,0],[11.445,0],[19.445,1]]}]
        }]
      }
     ]
    }
    Text_ObjState {
     id=#12
     display_name=Text_ObjectiveState
     possessed_class=CommonTextBlock
     tracks [
      {
       type=MovieSceneFloatTrack
       display_name=Render Opacity
       sections [{
         type=MovieSceneFloatSection
         end_frame=61
         start_frame=-1
         channels [None {type=MovieSceneScriptingFloatChannel, keys [[-0.0003,0.0168],[4.6667,0],[9.3333,1]]}]
        }]
      }
      {
       type=MovieScene2DTransformTrack
       sections [{
         type=MovieScene2DTransformSection
         end_frame=61
         start_frame=-1
         channels [
          Translation.X {type=MovieSceneScriptingFloatChannel, keys [[-0.0003,-200],[4.6667,-200],[9.3333,0]]}
          Translation.Y {type=MovieSceneScriptingFloatChannel, const [0,3]}
          Angle {type=MovieSceneScriptingFloatChannel}
          Scale.X {type=MovieSceneScriptingFloatChannel}
          Scale.Y {type=MovieSceneScriptingFloatChannel}
          Shear.X {type=MovieSceneScriptingFloatChannel}
          Shear.Y {type=MovieSceneScriptingFloatChannel}
         ]
        }]
      }
     ]
    }
    Overlay_Root {
     id=#13
     possessed_class=Overlay
     tracks [
      {
       type=MovieScene2DTransformTrack
       sections [{
         type=MovieScene2DTransformSection
         end_frame=24
         start_frame=0
         channels [
          Translation.X {type=MovieSceneScriptingFloatChannel, const [0,1]}
          Translation.Y {type=MovieSceneScriptingFloatChannel}
          Angle {type=MovieSceneScriptingFloatChannel}
          Scale.X {type=MovieSceneScriptingFloatChannel}
          Scale.Y {type=MovieSceneScriptingFloatChannel}
          Shear.X {type=MovieSceneScriptingFloatChannel}
          Shear.Y {type=MovieSceneScriptingFloatChannel}
         ]
        }]
      }
      {
       type=MovieSceneFloatTrack
       display_name=Render Opacity
       sections [{type=MovieSceneFloatSection, end_frame=24, start_frame=0, channels [None {type=MovieSceneScriptingFloatChannel, const [1,1]}]}]
      }
     ]
    }
   ]}
 }
 Anim_Succeed_Outro {
  binding_count=6
  length_seconds=1
  object=&self:Anim_Succeed_Outro
  sequence {bindings [
    WBP_Athena_Objectives_Icon {
     id=#14
     possessed_class=WBP_Athena_Objectives_Icon_C
     tracks [{
       type=MovieSceneFloatTrack
       display_name=Render Opacity
       sections [{type=MovieSceneFloatSection, end_frame=0, start_frame=0, channels [None {type=MovieSceneScriptingFloatChannel, const [0,1]}]}]
      }]
    }
    Image_fx_Shine {
     id=#10
     possessed_class=Image
     tracks [
      {
       type=MovieSceneFloatTrack
       display_name=Render Opacity
       sections [{type=MovieSceneFloatSection, end_frame=20, start_frame=0, channels [None {type=MovieSceneScriptingFloatChannel, const [0,2]}]}]
      }
      {
       type=MovieScene2DTransformTrack
       sections [{
         type=MovieScene2DTransformSection
         end_frame=20
         start_frame=0
         channels [
          Translation.X {type=MovieSceneScriptingFloatChannel, const [0.9143,2]}
          Translation.Y {type=MovieSceneScriptingFloatChannel, const [0,1]}
          Angle {type=MovieSceneScriptingFloatChannel, const [0,1]}
          Scale.X {type=MovieSceneScriptingFloatChannel, const [1.6777,2]}
          Scale.Y {type=MovieSceneScriptingFloatChannel, const [1,2]}
          Shear.X {type=MovieSceneScriptingFloatChannel, const [0,1]}
          Shear.Y {type=MovieSceneScriptingFloatChannel, const [0,1]}
         ]
        }]
      }
     ]
    }
    Overlay_RewardIcon {
     id=#15
     possessed_class=Overlay
     tracks [{
       type=MovieSceneFloatTrack
       display_name=Render Opacity
       sections [{
         type=MovieSceneFloatSection
         end_frame=20
         start_frame=0
         channels [None {type=MovieSceneScriptingFloatChannel, thinned_from=4, keys [[4.3477,1],[10.145,0],[18.261,0]]}]
        }]
      }]
    }
    Image_fx_Reward_Glow {
     id=#11
     possessed_class=Image
     tracks [
      {
       type=MovieScene2DTransformTrack
       sections [{
         type=MovieScene2DTransformSection
         end_frame=20
         start_frame=0
         channels [
          Translation.X {type=MovieSceneScriptingFloatChannel, const [5,1]}
          Translation.Y {type=MovieSceneScriptingFloatChannel, const [1,1]}
          Angle {type=MovieSceneScriptingFloatChannel, const [0,1]}
          Scale.X {type=MovieSceneScriptingFloatChannel, const [2.6457,2]}
          Scale.Y {type=MovieSceneScriptingFloatChannel, const [6.811,2]}
          Shear.X {type=MovieSceneScriptingFloatChannel, const [0,1]}
          Shear.Y {type=MovieSceneScriptingFloatChannel, const [0,1]}
         ]
        }]
      }
      {
       type=MovieSceneFloatTrack
       display_name=Render Opacity
       sections [{
         type=MovieSceneFloatSection
         end_frame=20
         start_frame=0
         channels [None {type=MovieSceneScriptingFloatChannel, keys [[4.3477,1],[18.2607,0]]}]
        }]
      }
     ]
    }
    Text_ObjState {
     id=#12
     display_name=Text_ObjectiveState
     possessed_class=CommonTextBlock
     tracks [
      {
       type=MovieSceneFloatTrack
       display_name=Render Opacity
       sections [{type=MovieSceneFloatSection, end_frame=20, start_frame=0, channels [None {type=MovieSceneScriptingFloatChannel, const [1,1]}]}]
      }
      {
       type=MovieScene2DTransformTrack
       sections [{
         type=MovieScene2DTransformSection
         end_frame=20
         start_frame=0
         channels [
          Translation.X {type=MovieSceneScriptingFloatChannel, const [0,1]}
          Translation.Y {type=MovieSceneScriptingFloatChannel}
          Angle {type=MovieSceneScriptingFloatChannel}
          Scale.X {type=MovieSceneScriptingFloatChannel}
          Scale.Y {type=MovieSceneScriptingFloatChannel}
          Shear.X {type=MovieSceneScriptingFloatChannel}
          Shear.Y {type=MovieSceneScriptingFloatChannel}
         ]
        }]
      }
     ]
    }
    Overlay_Root {
     id=#13
     possessed_class=Overlay
     tracks [
      {
       type=MovieScene2DTransformTrack
       sections [{
         type=MovieScene2DTransformSection
         end_frame=20
         start_frame=0
         channels [
          Translation.X {type=MovieSceneScriptingFloatChannel, thinned_from=4, keys [[4.3477,0],[12.5607,-100],[18.261,-100]]}
          Translation.Y {type=MovieSceneScriptingFloatChannel}
          Angle {type=MovieSceneScriptingFloatChannel}
          Scale.X {type=MovieSceneScriptingFloatChannel}
          Scale.Y {type=MovieSceneScriptingFloatChannel}
          Shear.X {type=MovieSceneScriptingFloatChannel}
          Shear.Y {type=MovieSceneScriptingFloatChannel}
         ]
        }]
      }
      {
       type=MovieSceneFloatTrack
       display_name=Render Opacity
       sections [{
         type=MovieSceneFloatSection
         end_frame=20
         start_frame=0
         channels [None {type=MovieSceneScriptingFloatChannel, thinned_from=4, keys [[4.3477,1],[12.5607,0],[18.261,0]]}]
        }]
      }
     ]
    }
   ]}
 }
 Anim_Fail_Outro {
  binding_count=5
  length_seconds=4.45
  object=&self:Anim_Fail_Outro
  sequence {bindings [
    Image_fx_Shine {
     id=#10
     possessed_class=Image
     tracks [
      {
       type=MovieSceneFloatTrack
       display_name=Render Opacity
       sections [{
         type=MovieSceneFloatSection
         end_frame=89
         start_frame=0
         channels [None {type=MovieSceneScriptingFloatChannel, keys [[0,0],[0.7383,0.2],[3.7703,0.2],[8.6667,0]]}]
        }]
      }
      {
       type=MovieScene2DTransformTrack
       sections [{
         type=MovieScene2DTransformSection
         end_frame=89
         start_frame=0
         channels [
          Translation.X {type=MovieSceneScriptingFloatChannel, keys [[0,-331.1956],[6.646,0.9143]]}
          Translation.Y {type=MovieSceneScriptingFloatChannel, const [0,1]}
          Angle {type=MovieSceneScriptingFloatChannel, const [0,1]}
          Scale.X {type=MovieSceneScriptingFloatChannel, keys [[0,1],[6.646,1.6777]]}
          Scale.Y {type=MovieSceneScriptingFloatChannel, keys [[0,1.3],[6.646,1]]}
          Shear.X {type=MovieSceneScriptingFloatChannel, const [0,1]}
          Shear.Y {type=MovieSceneScriptingFloatChannel, const [0,1]}
         ]
        }]
      }
      {
       type=MovieSceneWidgetMaterialTrack
       sections [{
         type=MovieSceneParameterSection
         end_frame=89
         start_frame=0
         channels [
          ColorR {type=MovieSceneScriptingFloatChannel, const [0.6042,1]}
          ColorG {type=MovieSceneScriptingFloatChannel, const [0,1]}
          ColorB {type=MovieSceneScriptingFloatChannel, const [0.005,1]}
          ColorA {type=MovieSceneScriptingFloatChannel, const [1,1]}
         ]
        }]
      }
     ]
    }
    Text_ObjState {
     id=#16
     display_name=Text_ObjectiveState
     possessed_class=CommonTextBlock
     tracks [
      {
       type=MovieSceneFloatTrack
       display_name=Render Opacity
       sections [{
         type=MovieSceneFloatSection
         end_frame=89
         start_frame=-1
         channels [None {type=MovieSceneScriptingFloatChannel, keys [[-0.0003,0.0168],[4.6667,0],[9.3333,1]]}]
        }]
      }
      {
       type=MovieScene2DTransformTrack
       sections [{
         type=MovieScene2DTransformSection
         end_frame=89
         start_frame=-1
         channels [
          Translation.X {type=MovieSceneScriptingFloatChannel, keys [[-0.0003,-200],[4.6667,-200],[9.3333,0]]}
          Translation.Y {type=MovieSceneScriptingFloatChannel, const [0,3]}
          Angle {type=MovieSceneScriptingFloatChannel}
          Scale.X {type=MovieSceneScriptingFloatChannel}
          Scale.Y {type=MovieSceneScriptingFloatChannel}
          Shear.X {type=MovieSceneScriptingFloatChannel}
          Shear.Y {type=MovieSceneScriptingFloatChannel}
         ]
        }]
      }
     ]
    }
    Overlay_RewardIcon {
     id=#15
     possessed_class=Overlay
     tracks [
      {
       type=MovieScene2DTransformTrack
       sections [{
         type=MovieScene2DTransformSection
         channels [
          Translation.X {type=MovieSceneScriptingFloatChannel, keys [[4,350],[9,415],[17.3333,415],[24,300],[31.3333,300]]}
          Translation.Y {type=MovieSceneScriptingFloatChannel}
          Angle {type=MovieSceneScriptingFloatChannel}
          Scale.X {type=MovieSceneScriptingFloatChannel}
          Scale.Y {type=MovieSceneScriptingFloatChannel}
          Shear.X {type=MovieSceneScriptingFloatChannel}
          Shear.Y {type=MovieSceneScriptingFloatChannel}
         ]
        }]
      }
      {
       type=MovieSceneFloatTrack
       display_name=Render Opacity
       sections [{type=MovieSceneFloatSection, end_frame=89, start_frame=0, channels [None {type=MovieSceneScriptingFloatChannel, const [0,3]}]}]
      }
     ]
    }
    Overlay_Root {
     id=#13
     possessed_class=Overlay
     tracks [
      {
       type=MovieScene2DTransformTrack
       sections [{
         type=MovieScene2DTransformSection
         end_frame=89
         start_frame=0
         channels [
          Translation.X {type=MovieSceneScriptingFloatChannel, keys [[22,0],[75.148,0],[84.5927,-100],[89,-100]]}
          Translation.Y {type=MovieSceneScriptingFloatChannel}
          Angle {type=MovieSceneScriptingFloatChannel}
          Scale.X {type=MovieSceneScriptingFloatChannel}
          Scale.Y {type=MovieSceneScriptingFloatChannel}
          Shear.X {type=MovieSceneScriptingFloatChannel}
          Shear.Y {type=MovieSceneScriptingFloatChannel}
         ]
        }]
      }
      {
       type=MovieSceneFloatTrack
       display_name=Render Opacity
       sections [{
         type=MovieSceneFloatSection
         end_frame=89
         start_frame=0
         channels [None {type=MovieSceneScriptingFloatChannel, thinned_from=5, keys [[0,1],[75.148,1],[84.5927,0],[89,0]]}]
        }]
      }
     ]
    }
    Text_ObjState_AdditionalInfo {
     id=#17
     display_name=Text_Update_Fail
     possessed_class=CommonTextBlock
     tracks [
      {
       type=MovieSceneFloatTrack
       display_name=Render Opacity
       sections [{
         type=MovieSceneFloatSection
         end_frame=88
         start_frame=0
         channels [None {type=MovieSceneScriptingFloatChannel, keys [[22.3333,0],[29.9917,1],[80,0.9791],[87.6667,0]]}]
        }]
      }
      {
       type=MovieScene2DTransformTrack
       sections [{
         type=MovieScene2DTransformSection
         end_frame=89
         start_frame=0
         channels [
          Translation.X {type=MovieSceneScriptingFloatChannel, keys [[22.3333,-30],[29.9917,0],[80,0],[88,-30]]}
          Translation.Y {type=MovieSceneScriptingFloatChannel, const [0,2]}
          Angle {type=MovieSceneScriptingFloatChannel, const [0,1]}
          Scale.X {type=MovieSceneScriptingFloatChannel, const [1,1]}
          Scale.Y {type=MovieSceneScriptingFloatChannel, const [1,1]}
          Shear.X {type=MovieSceneScriptingFloatChannel, const [0,1]}
          Shear.Y {type=MovieSceneScriptingFloatChannel, const [0,1]}
         ]
        }]
      }
     ]
    }
   ]}
 }
 Anim_Fail_TimeOut_Outro {
  binding_count=6
  length_seconds=4.45
  object=&self:Anim_Fail_TimeOut_Outro
  sequence {bindings [
    Text_ObjState {
     id=#18
     display_name=Text_ObjectiveState
     possessed_class=CommonTextBlock
     tracks [
      {
       type=MovieSceneFloatTrack
       display_name=Render Opacity
       sections [{
         type=MovieSceneFloatSection
         end_frame=89
         start_frame=-1
         channels [None {type=MovieSceneScriptingFloatChannel, keys [[-0.0003,0.0168],[4.6667,0],[9.3333,1]]}]
        }]
      }
      {
       type=MovieScene2DTransformTrack
       sections [{
         type=MovieScene2DTransformSection
         end_frame=89
         start_frame=-1
         channels [
          Translation.X {type=MovieSceneScriptingFloatChannel, keys [[-0.0003,-200],[4.6667,-200],[9.3333,0]]}
          Translation.Y {type=MovieSceneScriptingFloatChannel, const [0,3]}
          Angle {type=MovieSceneScriptingFloatChannel}
          Scale.X {type=MovieSceneScriptingFloatChannel}
          Scale.Y {type=MovieSceneScriptingFloatChannel}
          Shear.X {type=MovieSceneScriptingFloatChannel}
          Shear.Y {type=MovieSceneScriptingFloatChannel}
         ]
        }]
      }
     ]
    }
    Text_ObjState_AdditionalInfo {
     id=#19
     display_name=Text_Update
     possessed_class=CommonTextBlock
     tracks [
      {
       type=MovieSceneFloatTrack
       display_name=Render Opacity
       sections [{
         type=MovieSceneFloatSection
         end_frame=88
         start_frame=0
         channels [None {type=MovieSceneScriptingFloatChannel, keys [[22.3333,0],[29.9917,1],[80,0.9791],[87.6667,0]]}]
        }]
      }
      {
       type=MovieScene2DTransformTrack
       sections [{
         type=MovieScene2DTransformSection
         end_frame=89
         start_frame=0
         channels [
          Translation.X {type=MovieSceneScriptingFloatChannel, keys [[22.3333,-30],[29.9917,0],[80,0],[88,-30]]}
          Translation.Y {type=MovieSceneScriptingFloatChannel, const [0,2]}
          Angle {type=MovieSceneScriptingFloatChannel, const [0,1]}
          Scale.X {type=MovieSceneScriptingFloatChannel, const [1,1]}
          Scale.Y {type=MovieSceneScriptingFloatChannel, const [1,1]}
          Shear.X {type=MovieSceneScriptingFloatChannel, const [0,1]}
          Shear.Y {type=MovieSceneScriptingFloatChannel, const [0,1]}
         ]
        }]
      }
     ]
    }
    Image_fx_Shine {
     id=#10
     possessed_class=Image
     tracks [
      {
       type=MovieSceneFloatTrack
       display_name=Render Opacity
       sections [{
         type=MovieSceneFloatSection
         end_frame=89
         start_frame=0
         channels [None {type=MovieSceneScriptingFloatChannel, keys [[0,0],[0.7383,0.2],[3.7703,0.2],[8.6667,0]]}]
        }]
      }
      {
       type=MovieScene2DTransformTrack
       sections [{
         type=MovieScene2DTransformSection
         end_frame=89
         start_frame=0
         channels [
          Translation.X {type=MovieSceneScriptingFloatChannel, keys [[0,-331.1956],[6.646,0.9143]]}
          Translation.Y {type=MovieSceneScriptingFloatChannel, const [0,1]}
          Angle {type=MovieSceneScriptingFloatChannel, const [0,1]}
          Scale.X {type=MovieSceneScriptingFloatChannel, keys [[0,1],[6.646,1.6777]]}
          Scale.Y {type=MovieSceneScriptingFloatChannel, keys [[0,1.3],[6.646,1]]}
          Shear.X {type=MovieSceneScriptingFloatChannel, const [0,1]}
          Shear.Y {type=MovieSceneScriptingFloatChannel, const [0,1]}
         ]
        }]
      }
      {
       type=MovieSceneWidgetMaterialTrack
       sections [{
         type=MovieSceneParameterSection
         end_frame=89
         start_frame=0
         channels [
          ColorR {type=MovieSceneScriptingFloatChannel, const [0.6042,1]}
          ColorG {type=MovieSceneScriptingFloatChannel, const [0,1]}
          ColorB {type=MovieSceneScriptingFloatChannel, const [0.005,1]}
          ColorA {type=MovieSceneScriptingFloatChannel, const [1,1]}
         ]
        }]
      }
     ]
    }
    Overlay_RewardIcon {
     id=#15
     possessed_class=Overlay
     tracks [
      {
       type=MovieScene2DTransformTrack
       sections [{
         type=MovieScene2DTransformSection
         channels [
          Translation.X {type=MovieSceneScriptingFloatChannel, keys [[7.3333,415],[17.3333,415],[24,300],[31.3333,300]]}
          Translation.Y {type=MovieSceneScriptingFloatChannel}
          Angle {type=MovieSceneScriptingFloatChannel}
          Scale.X {type=MovieSceneScriptingFloatChannel}
          Scale.Y {type=MovieSceneScriptingFloatChannel}
          Shear.X {type=MovieSceneScriptingFloatChannel}
          Shear.Y {type=MovieSceneScriptingFloatChannel}
         ]
        }]
      }
      {
       type=MovieSceneFloatTrack
       display_name=Render Opacity
       sections [{type=MovieSceneFloatSection, end_frame=89, start_frame=0, channels [None {type=MovieSceneScriptingFloatChannel, const [0,4]}]}]
      }
     ]
    }
    Image_fx_TimeOut {
     id=#20
     possessed_class=Image
     tracks [
      {
       type=MovieScene2DTransformTrack
       sections [{
         type=MovieScene2DTransformSection
         end_frame=89
         start_frame=0
         channels [
          Translation.X {type=MovieSceneScriptingFloatChannel, const [0,1]}
          Translation.Y {type=MovieSceneScriptingFloatChannel, const [1,1]}
          Angle {type=MovieSceneScriptingFloatChannel, const [0,1]}
          Scale.X {type=MovieSceneScriptingFloatChannel, keys [[0,1],[12,0.9],[17,1.1]]}
          Scale.Y {type=MovieSceneScriptingFloatChannel, keys [[0,1],[10,0.9],[17,1.1]]}
          Shear.X {type=MovieSceneScriptingFloatChannel, const [0,1]}
          Shear.Y {type=MovieSceneScriptingFloatChannel, const [0,1]}
         ]
        }]
      }
      {
       type=MovieSceneFloatTrack
       display_name=Render Opacity
       sections [{
         type=MovieSceneFloatSection
         end_frame=89
         start_frame=0
         channels [None {type=MovieSceneScriptingFloatChannel, keys [[0,0],[10,0],[17,1],[79.953,1],[84.7487,0]]}]
        }]
      }
     ]
    }
    Overlay_Root {
     id=#13
     possessed_class=Overlay
     tracks [
      {
       type=MovieScene2DTransformTrack
       sections [{
         type=MovieScene2DTransformSection
         end_frame=89
         start_frame=0
         channels [
          Translation.X {type=MovieSceneScriptingFloatChannel, keys [[22,0],[75.148,0],[84.5927,-100],[89,-100]]}
          Translation.Y {type=MovieSceneScriptingFloatChannel}
          Angle {type=MovieSceneScriptingFloatChannel}
          Scale.X {type=MovieSceneScriptingFloatChannel}
          Scale.Y {type=MovieSceneScriptingFloatChannel}
          Shear.X {type=MovieSceneScriptingFloatChannel}
          Shear.Y {type=MovieSceneScriptingFloatChannel}
         ]
        }]
      }
      {
       type=MovieSceneFloatTrack
       display_name=Render Opacity
       sections [{
         type=MovieSceneFloatSection
         end_frame=89
         start_frame=0
         channels [None {type=MovieSceneScriptingFloatChannel, keys [[22,1],[75.148,1],[84.5927,0],[89,0]]}]
        }]
      }
     ]
    }
   ]}
 }
 Anim_Update_AdditionalInfo {
  binding_count=2
  length_seconds=2.3
  object=&self:Anim_Update_AdditionalInfo
  sequence {bindings [
    Overlay_Root {
     id=#21
     possessed_class=Overlay
     tracks [{
       type=MovieSceneFloatTrack
       display_name=Render Opacity
       sections [{type=MovieSceneFloatSection, end_frame=-1, start_frame=-1, channels [None {type=MovieSceneScriptingFloatChannel, const [1,1]}]}]
      }]
    }
    Text_ObjState_AdditionalInfo {
     id=#22
     display_name=Text_Update
     possessed_class=CommonTextBlock
     tracks [
      {
       type=MovieSceneFloatTrack
       display_name=Render Opacity
       sections [{
         type=MovieSceneFloatSection
         end_frame=45
         start_frame=0
         channels [None {type=MovieSceneScriptingFloatChannel, keys [[0,0],[7.6583,1],[37,0.9791],[45,0]]}]
        }]
      }
      {
       type=MovieScene2DTransformTrack
       sections [{
         type=MovieScene2DTransformSection
         end_frame=45
         start_frame=0
         channels [
          Translation.X {type=MovieSceneScriptingFloatChannel, keys [[0,-30],[7.6583,0],[37,0],[45,-30]]}
          Translation.Y {type=MovieSceneScriptingFloatChannel, const [0,2]}
          Angle {type=MovieSceneScriptingFloatChannel, const [0,1]}
          Scale.X {type=MovieSceneScriptingFloatChannel, const [1,1]}
          Scale.Y {type=MovieSceneScriptingFloatChannel, const [1,1]}
          Shear.X {type=MovieSceneScriptingFloatChannel, const [0,1]}
          Shear.Y {type=MovieSceneScriptingFloatChannel, const [0,1]}
         ]
        }]
      }
     ]
    }
   ]}
 }
 Anim_Stage_Transition {
  binding_count=5
  length_seconds=1.05
  object=&self:Anim_Stage_Transition
  sequence {bindings [
    Text_ObjectiveTitle {
     id=#23
     possessed_class=FortRichTextBlockKeybind
     tracks [{
       type=MovieSceneFloatTrack
       display_name=Render Opacity
       sections [{type=MovieSceneFloatSection, end_frame=5, start_frame=-1, channels [None {type=MovieSceneScriptingFloatChannel, keys [[-1,1],[5,0]]}]}]
      }]
    }
    Overlay_Root {
     id=#24
     possessed_class=Overlay
     tracks [{
       type=MovieSceneFloatTrack
       display_name=Render Opacity
       sections [{type=MovieSceneFloatSection, end_frame=-1, start_frame=-1, channels [None {type=MovieSceneScriptingFloatChannel, const [1,1]}]}]
      }]
    }
    OverlayStageTransition {
     id=#25
     possessed_class=Overlay
     tracks [{
       type=MovieScene2DTransformTrack
       sections [{
         type=MovieScene2DTransformSection
         end_frame=20
         start_frame=-1
         channels [
          Translation.X {type=MovieSceneScriptingFloatChannel}
          Translation.Y {type=MovieSceneScriptingFloatChannel}
          Angle {type=MovieSceneScriptingFloatChannel}
          Scale.X {type=MovieSceneScriptingFloatChannel, keys [[-1,0],[10,1]]}
          Scale.Y {type=MovieSceneScriptingFloatChannel}
          Shear.X {type=MovieSceneScriptingFloatChannel}
          Shear.Y {type=MovieSceneScriptingFloatChannel}
         ]
        }]
      }]
    }
    Text_ObjectiveTransition {
     id=#26
     possessed_class=FortRichTextBlockKeybind
     tracks [{
       type=MovieSceneFloatTrack
       display_name=Render Opacity
       sections [{
         type=MovieSceneFloatSection
         end_frame=13
         start_frame=-1
         channels [None {type=MovieSceneScriptingFloatChannel, keys [[-1,0],[8.844,0],[13.3183,1]]}]
        }]
      }]
    }
    Image_Material_Transition {
     id=#27
     possessed_class=Image
     tracks [{
       type=MovieSceneFloatTrack
       display_name=Render Opacity
       sections [{
         type=MovieSceneFloatSection
         end_frame=20
         start_frame=-1
         channels [None {type=MovieSceneScriptingFloatChannel, keys [[-1,0],[3,0.9],[14,1],[20,1]]}]
        }]
      }]
    }
   ]}
 }
 Anim_Intro_Simple {
  binding_count=5
  length_seconds=1.0167
  object=&self:Anim_Intro_Simple
  sequence {
   bindings [
    DynamicEntryBox_OptionalEntries {
     id=#1
     possessed_class=FortDynamicEntryBox
     tracks [
      {
       type=MovieSceneFloatTrack
       display_name=Render Opacity
       sections [{
         type=MovieSceneFloatSection
         end_frame=61
         start_frame=0
         channels [None {type=MovieSceneScriptingFloatChannel, keys [[0,0],[42.628,0],[58.001,1]]}]
        }]
      }
      {
       type=MovieScene2DTransformTrack
       sections [{
         type=MovieScene2DTransformSection
         end_frame=61
         start_frame=0
         channels [
          Translation.X {type=MovieSceneScriptingFloatChannel, keys [[42.628,-300],[58.001,0]]}
          Translation.Y {type=MovieSceneScriptingFloatChannel}
          Angle {type=MovieSceneScriptingFloatChannel}
          Scale.X {type=MovieSceneScriptingFloatChannel}
          Scale.Y {type=MovieSceneScriptingFloatChannel}
          Shear.X {type=MovieSceneScriptingFloatChannel}
          Shear.Y {type=MovieSceneScriptingFloatChannel}
         ]
        }]
      }
     ]
    }
    Overlay_Root {
     id=#28
     possessed_class=Overlay
     tracks [{
       type=MovieSceneFloatTrack
       display_name=Render Opacity
       sections [{type=MovieSceneFloatSection, end_frame=0, start_frame=0, channels [None {type=MovieSceneScriptingFloatChannel, const [1,1]}]}]
      }]
    }
    Overlay_PrimaryObjective {
     id=#5
     possessed_class=Overlay
     tracks [
      {
       type=MovieSceneFloatTrack
       display_name=Render Opacity
       sections [{
         type=MovieSceneFloatSection
         end_frame=61
         start_frame=0
         channels [None {type=MovieSceneScriptingFloatChannel, keys [[0,0],[18.715,0],[40.205,1]]}]
        }]
      }
      {
       type=MovieScene2DTransformTrack
       sections [{
         type=MovieScene2DTransformSection
         end_frame=61
         start_frame=0
         channels [
          Translation.X {type=MovieSceneScriptingFloatChannel, keys [[18.715,-300],[33.966,0]]}
          Translation.Y {type=MovieSceneScriptingFloatChannel}
          Angle {type=MovieSceneScriptingFloatChannel}
          Scale.X {type=MovieSceneScriptingFloatChannel}
          Scale.Y {type=MovieSceneScriptingFloatChannel}
          Shear.X {type=MovieSceneScriptingFloatChannel}
          Shear.Y {type=MovieSceneScriptingFloatChannel}
         ]
        }]
      }
     ]
    }
    OverlaySlot_1 {id=#29, parent_id=#30, possessed_class=OverlaySlot, tracks []}
    Text_ObjState {
     id=#30
     possessed_class=CommonTextBlock
     tracks [
      {
       type=MovieScene2DTransformTrack
       sections [{
         type=MovieScene2DTransformSection
         end_frame=12
         start_frame=0
         channels [
          Translation.X {type=MovieSceneScriptingFloatChannel, keys [[0,-200],[12,0]]}
          Translation.Y {type=MovieSceneScriptingFloatChannel}
          Angle {type=MovieSceneScriptingFloatChannel}
          Scale.X {type=MovieSceneScriptingFloatChannel}
          Scale.Y {type=MovieSceneScriptingFloatChannel}
          Shear.X {type=MovieSceneScriptingFloatChannel}
          Shear.Y {type=MovieSceneScriptingFloatChannel}
         ]
        }]
      }
      {
       type=MovieSceneFloatTrack
       display_name=Render Opacity
       sections [{type=MovieSceneFloatSection, end_frame=11, start_frame=0, channels [None {type=MovieSceneScriptingFloatChannel, keys [[0,0],[11,1]]}]}]
      }
     ]
    }
   ]
   display_rate {numerator=60}
  }
 }
]
graphs [
 EventGraph {nodes [
   n1 = OnOptionalProgress!() -> (OutputDelegate, then, ProgressText:Text) {is=title K2Node_CustomEvent}
   > UpdateMultiQuestInMain!(execute, self, ObjectiveProgress:Text=@n1.ProgressText) -> (then)
   OnOptionalComplete() {is=title K2Node_CustomEvent}
   > SetCurrentOptionalEntry!(execute, self, OptionalEntry:WBP Athena Objectives Optional Entry Object Reference) -> (then)
   > SetMultiStep()
   n2 = Get CurrentOptionalEntry() -> (CurrentOptionalEntry:int=0)
   > n3 = Branch(Condition="integer >= integer"(A=int - int(A=@n2), B=Get Total Optional Entries() -> (Total Optional Entries:int=0)))
   > Set Text(
    Text=Format Text(Format="LOCTEXT(\"<Transition>Quest Completed</>\")")
    self:Rich Text Block Object Reference=Get Text_ObjectiveTransition() -> (Text_ObjectiveTransition:Fort Rich Text Block Keybind Object Reference)
   )
   > n4 = "Play Animation with Finished event (legacy)"!(
    execute:Exec
    Widget:User Widget Object Reference=Self-Reference()
    InAnimation:Widget Animation Object Reference=Get Anim_Stage_Transition() -> (Anim_Stage_Transition:Widget Animation Object Reference)
    StartAtTime:float=0
    NumLoopsToPlay:int=1
    PlayMode:EUMGSequencePlayMode Enum
    PlaybackSpeed:float=1
   ) -> (then:Exec, Result:UMGSequence Player Object Reference, Finished:Exec) {is=title K2Node_PlayAnimation}
   "Play Animation with Finished event (legacy)"(
    execute=[@n6.then, @n4.Finished]
    Widget=Self-Reference()
    InAnimation=Get Anim_Stage_Transition()
    StartAtTime=0
    NumLoopsToPlay=1
    PlayMode=Reverse
    PlaybackSpeed=1
   )
   SetStageTransitionText!(execute=@n3.else, self, CurrentStage:int=int - int(A=Get CurrentOptionalEntry() -> (CurrentOptionalEntry=0))) -> (then)
   > n5 = "Play Animation with Finished event (legacy)"(Widget=Self-Reference(), InAnimation=Get Anim_Stage_Transition(), StartAtTime=0, NumLoopsToPlay=1, PlaybackSpeed=1)
   SetMultiStepText(execute=@n5.Finished)
   > n6 = SetEntryProgressMultiStep()
   Event Dismiss Objective Entry() {is=title K2Node_CustomEvent}
   > Unbind from Animations()
   n7 = Get Anim_Fail_TimeOut_Outro() -> (Anim_Fail_TimeOut_Outro:Widget Animation Object Reference)
   n8 = Get Anim_Fail_Outro() -> (Anim_Fail_Outro:Widget Animation Object Reference)
   n9 = Get StateEnum() -> (StateEnum:Enum_UrgentQuestState Enum=NewEnumerator0)
   > n10 = "Play Animation with Finished event (legacy)"(
    Widget=Self-Reference()
    InAnimation=Select(
     Option 0:Widget Animation Object Reference=Select(Option 0=@n8, Option 1=@n7, Index=false)
     Option 1:Widget Animation Object Reference=Get Anim_Succeed_Outro() -> (Anim_Succeed_Outro:Widget Animation Object Reference)
     Index=Select!(
      NewEnumerator0:bool=false
      NewEnumerator1:bool=false
      NewEnumerator2:bool=true
      NewEnumerator3:bool=false
      NewEnumerator4:bool=false
      Index:Enum_UrgentQuestState Enum=@n9
     ) -> (ReturnValue:bool)
    ) -> (ReturnValue:Widget Animation Object Reference)
    StartAtTime=0
    NumLoopsToPlay=1
    PlaybackSpeed=1
   )
   Call On Entry Complete!(
    execute:Exec=@n10.Finished
    self:WBP Athena Objectives Entry Object Reference
    ObjectiveEntry:WBP Athena Objectives Entry Object Reference=Self-Reference()
   ) -> (then:Exec) {is=title K2Node_CallDelegate}
   > NotifyReadyToRemoveEntry(self:Athena Objective Entry Base Object Reference)
   > ResetWidgetState()
   > ResetCompletedAnims!(
    execute
    self:WBP Athena Objectives Icon Object Reference=Get WBP_Athena_Objectives_Icon() -> (WBP_Athena_Objectives_Icon:WBP Athena Objectives Icon Object Reference)
    IsQuestCompleted:bool=Get Is Quest Completed() -> (Is Quest Completed:bool)
   ) -> (then)
   > ResetCompletedAnim(self:WBP Athena Objectives Progress Object Reference=Get WBP_Objectives_Progress() -> (WBP_Objectives_Progress:WBP Athena Objectives Progress Object Reference))
   > n11 = For Each Loop(Array:Array of User Widget Object References=GetAllEntries!(self:Dynamic Entry Box Base Object Reference=Get DynamicEntryBox_OptionalEntries() -> (DynamicEntryBox_OptionalEntries:Dynamic Entry Box Object Reference)) -> (ReturnValue:Array of User Widget Object References)) -> (Array Element:User Widget Object Reference)
   Branch(execute=@n11.Completed, Condition=Get Is Quest Completed())
   > ResetCollapsed()
   n12 = Cast To WBP_Athena_Objectives_OptionalEntry!(execute=@n11.LoopBody, Object=@n11) -> (then, CastFailed, AsWBP Athena Objectives Optional Entry:WBP Athena Objectives Optional Entry Object Reference) {is=title K2Node_DynamicCast}
   > ResetEntry(self:WBP Athena Objectives Optional Entry Object Reference=@n12)
   RewardsAnimationSetup() {is=title K2Node_CustomEvent}
   n13 = Length!(TargetSet:Set of Reward Quantity Pair Structures=Get RewardQuantities() -> (RewardQuantities:Set of Reward Quantity Pair Structures)) -> (ReturnValue:int=0) {is=title K2Node_CallFunction}
   > n14 = Set HasRewards(HasRewards:bool="integer > integer"(A=@n13)) -> (Output_Get:bool)
   > n15 = Branch(Condition=@n14)
   n16 = Get RewardCounts() -> (RewardCounts:int=0)
   > n17 = Branch(Condition="integer > integer"(A=@n13, B=@n16))
   n18 = Get Item By Index!(TargetSet:Set of Reward Quantity Pair Structures=Get RewardQuantities(), Index:int=@n16) -> (Item_RewardQuantity_2_#31:int, Item_RewardType_5_#32:EFortEarnedSubRewardType Enum) {is=title K2Node_CallFunction}
   > Set Rewards!(execute, self, Quantity:int=@n18, Type:EFortEarnedSubRewardType Enum=@n18.Item_RewardType_5_#32) -> (then) {is=title K2Node_CallFunction}
   > PlayRewardIconSound()
   > n19 = Branch(Condition="Equal (Byte)"(A:Byte=Get StateEnum() -> (StateEnum=NewEnumerator0), B:Byte=Get StartState() -> (StartState:Enum_UrgentQuestState Enum=NewEnumerator0)) -> (ReturnValue:bool=~) {is=title K2Node_PromotableOperator})
   n20 = Get Anim_Intro_Reward() -> (Anim_Intro_Reward:Widget Animation Object Reference)
   n21 = Get RewardQuantities()
   n22 = AND Boolean(
    A="Equal (Integer)"(A=Length(TargetSet=@n21) -> (ReturnValue=0), B=1)
    B=Break Urgent Quest Data!(UrgentQuestData:"Urgent Quest Data Structure (by ref)"=Get Urgent Quest Data() -> (Urgent Quest Data:Urgent Quest Data Structure)) -> (
     VisualStyleTag:Gameplay Tag Container Structure
     bRewardPersist:bool=~
     AlertIcon:Soft Object Reference
     CharacterIcon:Soft Object Reference
     AlertIconBorderColor:Linear Color Structure
     AlertIconBrushSize:Vector 2D Structure
    ) {is=title K2Node_BreakStruct}.bRewardPersist
   )
   > n23 = Play Animation Time Range with Finished event!(
    execute:Exec
    Widget:User Widget Object Reference=Self-Reference()
    InAnimation:Widget Animation Object Reference=@n20
    StartAtTime:float=0
    EndAtTime:float=Select!(Option 0:float=GetEndTime(self=@n20), Option 1:float=Get PersistEndTime() -> (PersistEndTime:float), Index:bool=@n22) -> (ReturnValue:float)
    NumLoopsToPlay:int=1
    PlayMode:EUMGSequencePlayMode Enum
    PlaybackSpeed:float=1
   ) -> (then:Exec, Result:Widget Animation Handle Structure, Finished:Exec) {is=title K2Node_PlayAnimationTimeRange2}
   Set RewardCounts(execute=[@n23.Finished, @n25.Finished], RewardCounts:int=int + int(A=Get RewardCounts() -> (RewardCounts=0))) -> (Output_Get:int=0)
   > Set Timer by Event(Delegate=Create Event(), Time=0.3, bLooping=false)
   n24 = Get Anim_Intro_Reward()
   n25 = Play Animation Time Range with Finished event(
    execute=@n19.else
    Widget=Self-Reference()
    InAnimation=@n24
    StartAtTime=Select(Option 0=GetStartTime!(self:Widget Animation Object Reference=@n24) -> (ReturnValue:float), Option 1=Get PersistEndTime(), Index=@n22)
    EndAtTime=GetEndTime(self=@n24)
    NumLoopsToPlay=1
    PlaybackSpeed=1
   )
   Set RewardCounts(execute=@n17.else, RewardCounts=0) -> (Output_Get=0)
   n26 = Get Urgent Quest Data()
   Branch(
    execute=@n15.else
    Condition=IsValidSoftObjectReference(SoftObjectReference=Break Urgent Quest Data(UrgentQuestData=@n26) -> (bRewardPersist=~).AlertIcon)
   )
   > HideRewardText!(execute, self:WBP Athena Objectives Icon Object Reference=Get WBP_Athena_Objectives_Icon(), HideReward:bool=true) -> (then)
   > n27 = Branch(Condition="Equal (Byte)"(A=Get StateEnum() -> (StateEnum=NewEnumerator0)) -> (ReturnValue=~))
   n28 = Break Urgent Quest Data(UrgentQuestData=Get Urgent Quest Data()) -> (bRewardPersist=~)
   n29 = Get Anim_Intro_Reward()
   > Play Animation Time Range with Finished event(
    Widget=Self-Reference()
    InAnimation=@n29
    StartAtTime=0
    EndAtTime=Select(Option 0=GetEndTime(self=@n29), Option 1=1.5, Index=@n28.bRewardPersist)
    NumLoopsToPlay=1
    PlaybackSpeed=1
   )
   n30 = Get Anim_Intro_Reward()
   Play Animation Time Range with Finished event(
    execute=@n27.else
    Widget=Self-Reference()
    InAnimation=@n30
    StartAtTime=Select(Option 0=GetStartTime(self=@n30), Option 1=1.5, Index=@n28.bRewardPersist)
    EndAtTime=GetEndTime(self=@n30)
    NumLoopsToPlay=1
    PlaybackSpeed=1
   )
   n31 = OnMaterialOverride_Event!() -> (OutputDelegate, then, MaterialOverride:Material Instance Soft Object Reference) {is=title K2Node_CustomEvent}
   > n32 = Async Load Asset(Asset=@n31.MaterialOverride)
   n33 = Cast To MaterialInstance!(execute=@n32.Completed, Object=@n32) -> (then, CastFailed, AsMaterial Instance:Material Instance Object Reference) {is=title K2Node_DynamicCast}
   > n34 = CreateBrushFromAsset!(
    execute
    LoadedAsset:Object Reference=@n33
    BrushSize:"Vector 2D Structure (by ref)"=GetStandardBrushSize!(BrushSize:EFortBrushSize Enum=Medium) -> (ReturnValue:Vector 2D Structure)
   ) -> (then, ReturnValue:Slate Brush Structure)
   n35 = Get Image_Main_Bg() -> (Image_Main_Bg:Image Object Reference)
   > SetBrush(self=@n35, InBrush=@n34)
   > n36 = GetDynamicMaterial(self=@n35)
   > n37 = Get Scalar Parameter Value!(execute, self:Material Instance Dynamic Object Reference=@n36, ParameterName:Name=ShowItemImage) -> (then, ReturnValue:float) {is=title K2Node_CallFunction}
   > n38 = Set ShowIconImage(ShowIconImage:bool="To Bool (Double)"!(Value:double=@n37) -> (ReturnValue:bool)) -> (Output_Get:bool)
   n39 = Get BigIconFullBody() -> (BigIconFullBody:Fort Lazy Image Object Reference)
   > SetVisibility(
    self=[@n39, Get TopIconSpacer() -> (TopIconSpacer:Spacer Object Reference)]
    InVisibility=Select(Option 0:ESlateVisibility Enum=Collapsed, Option 1:ESlateVisibility Enum=SelfHitTestInvisible, Index=@n38) -> (ReturnValue:ESlateVisibility Enum)
   )
   > Branch(Condition=Get ShowIconImage() -> (ShowIconImage:bool))
   > n40 = GetDynamicMaterial(self=@n39)
   > n41 = Get Texture Parameter Value!(execute, self:Material Instance Dynamic Object Reference=@n36, ParameterName:Name=Texture) -> (then, ReturnValue:Texture Object Reference) {is=title K2Node_CallFunction}
   > SetTextureParameterValue(self=@n40, ParameterName=Item, Value=@n41)
   IntroAnimation() {is=title K2Node_CustomEvent}
   > SetVisibility(self=Get Overlay_Root() -> (Overlay_Root:Overlay Object Reference), InVisibility=SelfHitTestInvisible)
   n42 = Break Urgent Quest Data!(UrgentQuestData=Get Urgent Quest Data()) -> (VisualStyleTag, bQuickIntro:bool=~, bRewardPersist=~, CharacterIcon)
   > SetVisibility(
    self=Get Overlay_IntroOutroTitle() -> (Overlay_IntroOutroTitle:Overlay Object Reference)
    InVisibility=Select(Option 0=SelfHitTestInvisible, Option 1=Collapsed, Index=@n42.bQuickIntro)
   )
   > n43 = "Play Animation with Finished event (legacy)"(
    Widget=Self-Reference()
    InAnimation=Select(
     Option 0:Widget Animation Object Reference=Get Anim_Intro_Started() -> (Anim_Intro_Started:Widget Animation Object Reference)
     Option 1:Widget Animation Object Reference=Get Anim_Intro_Simple() -> (Anim_Intro_Simple:Widget Animation Object Reference)
     Index=@n42.bQuickIntro
    ) -> (ReturnValue:Widget Animation Object Reference)
    StartAtTime=0
    NumLoopsToPlay=1
    PlaybackSpeed=Select(Option 0:float=1, Option 1:float=Get QuickIntroPlayback() -> (QuickIntroPlayback:double), Index=@n42.bQuickIntro) -> (ReturnValue:float)
   )
   > n44 = Set Timer by Event(Delegate=Create Event(), Time=0.5, bLooping=false)
   > Set RewardAnimationTimer(RewardAnimationTimer:Timer Handle Structure=@n44) -> (Output_Get:Timer Handle Structure)
   SetVisibility(
    execute=@n43.Finished
    self=[Get Overlay_IntroOutroTitle(), Get Text_ObjState_AdditionalInfo() -> (Text_ObjState_AdditionalInfo:Common Text Object Reference)]
    InVisibility=Collapsed
   )
   n45 = Event OnDescOverrideChanged!() -> (OutputDelegate, then, OverrideTag:"Gameplay Tag Structure (by ref)") {is=title K2Node_Event}
   > Set members in Urgent Quest Data!(
    execute:Exec
    StructRef:"Urgent Quest Data Structure (by ref)"=Get Urgent Quest Data()
    EventDescriptionOverrideTag:Gameplay Tag Structure=@n45.OverrideTag
   ) -> (then:Exec, StructOut:"Urgent Quest Data Structure (by ref)") {is=title K2Node_SetFieldsInStruct}
   > UpdateTaskTitle()
   n46 = Event OnUpdatePrimaryObjectiveProgress!() -> (OutputDelegate, then, Text:"Text (by ref)") {is=title K2Node_Event}
   > Branch(Condition=Get IsMultiStepExpanded() -> (IsMultiStepExpanded:bool))
   > SetObjectiveProgressText!(execute, self, Text:Text=@n46.Text) -> (then)
   > PlayProgressGainedAnim(self:WBP Athena Objectives Progress Object Reference=Get WBP_Objectives_Progress())
   n47 = Event On Entry Completed!() -> (
    OutputDelegate
    then
    UrgentQuestData:"Urgent Quest Data Structure (by ref)"
    bIsQuestCompleted:bool
    CompletionPathContext:"Gameplay Tag Container Structure (by ref)"
   ) {is=title K2Node_Event}
   > Set Urgent Quest Data(Urgent Quest Data:Urgent Quest Data Structure=@n47.UrgentQuestData) -> (Output_Get:Urgent Quest Data Structure)
   > Set Is Quest Completed(Is Quest Completed:bool=@n47.bIsQuestCompleted) -> (Output_Get:bool)
   > Set Completion Path Context(Completion Path Context:Gameplay Tag Container Structure=@n47.CompletionPathContext) -> (Output_Get:Gameplay Tag Container Structure)
   > Update Reward Quantity()
   > UpdateObjectiveState!(execute, self, IsEntrySetup:bool=false) -> (then)
   n48 = Get Completion Path Context() -> (Completion Path Context:Gameplay Tag Container Structure)
   n49 = AND Boolean(
    A=Get Is Quest Completed()
    B=NOT Boolean(A=HasTag!(
      TagContainer:"Gameplay Tag Container Structure (by ref)"=@n48
      Tag:Gameplay Tag Structure=(TagName="Athena.Quests.CompletionPath.Failed")
      bExactMatch:bool=false
     ) -> (ReturnValue:bool))
   )
   > Set StateEnum(StateEnum:Enum_UrgentQuestState Enum=Select(Option 0:Enum_UrgentQuestState Enum=NewEnumerator3, Option 1:Enum_UrgentQuestState Enum=NewEnumerator2, Index=@n49) -> (ReturnValue:Enum_UrgentQuestState Enum)) -> (Output_Get:Enum_UrgentQuestState Enum=NewEnumerator0)
   > n50 = Branch(Condition=@n49)
   > n51 = Branch(Condition=Break Urgent Quest Data!(UrgentQuestData=Get Urgent Quest Data()) -> (VisualStyleTag, bQuickIntro=~, bQuickOutro:bool=~, bRewardPersist=~, CharacterIcon).bQuickOutro)
   > n52 = Play Animation Time Range with Finished event(
    Widget=Self-Reference()
    InAnimation=Get Anim_Succeed_Intro() -> (Anim_Succeed_Intro:Widget Animation Object Reference)
    StartAtTime=0
    EndAtTime=0.75
    NumLoopsToPlay=1
    PlaybackSpeed=1
   )
   > Set Timer by Event(Delegate=Create Event(), Time=0.2, bLooping=false)
   Event Dismiss Objective Entry(execute=@n52.Finished) {is=title K2Node_CallFunction}
   PlayObjectiveCompletedAnim(execute=@n51.else, self:WBP Athena Objectives Progress Object Reference=Get WBP_Objectives_Progress())
   > PlayQuestCompleteSound()
   > n53 = "Play Animation with Finished event (legacy)"(Widget=Self-Reference(), InAnimation=Get Anim_Succeed_Intro(), StartAtTime=0, NumLoopsToPlay=1, PlaybackSpeed=1)
   > Set Timer by Event(Delegate=Create Event(), Time=0.5, bLooping=false)
   > PlaySuccessAnims()
   Delay(execute=@n53.Finished, Duration=0.2)
   > SpawnSound2D!(
    execute
    Sound:Sound Base Object Reference=/Game/Sounds/Fort_UI_Sounds/S15/Quest_Complete_Cue
    VolumeMultiplier:float=1
    PitchMultiplier:float=1
    StartTime:float=0
    ConcurrencySettings:Sound Concurrency Object Reference
    bPersistAcrossLevelTransition:bool=false
    bAutoDestroy:bool=true
   ) -> (then, ReturnValue:Audio Component Object Reference)
   > Event Dismiss Objective Entry()
   Event Dismiss Objective Entry(execute=@n50.else)
   > PlayQuestFailedSound()
   Event PreConstruct()
   n54 = Get Text_ProgressValue!(self:WBP Athena Objectives Progress Object Reference=Get WBP_Objectives_Progress()) -> (Text_ProgressValue:Common Text Object Reference)
   > Set MinDesiredWidth!(execute, MinDesiredWidth:float=64, self:Text Object Reference=@n54) -> (then, Output_Get:float)
   > Set Justification!(execute, Justification:ETextJustify Enum=Center, self:Text Layout Widget Object Reference=@n54) -> (then, Output_Get:ETextJustify Enum=Left)
   n55 = Event Construct()
   n56 = HandleInputMethodChanged!() -> (OutputDelegate, then, bNewInputType:ECommonInputType Enum) {is=title K2Node_CustomEvent}
   n57 = Get CommonInputSubsystem()
   Bind Event to On Input Method Changed(execute=@n55.then, self=@n57, Delegate=@n56)
   Event OnObjectiveInitialize() {is=title K2Node_Event}
   > Clear!(execute, TargetSet:Set of Reward Quantity Pair Structures=Get RewardQuantities()) -> (then) {is=title K2Node_CallFunction}
   n58 = Event OnUpdateBountyThreatLevel!() -> (OutputDelegate, then, NewPlayerBountyThreatLevel:EPlayerBountyThreatLevel Enum=Low) {is=title K2Node_Event}
   > Set Bounty Threat Level(Bounty Threat Level:EPlayerBountyThreatLevel Enum=@n58.NewPlayerBountyThreatLevel) -> (Output_Get:EPlayerBountyThreatLevel Enum=Low)
   > Update Bounty Threat()
   Event Destruct()
   > Unbind Event from On Input Method Changed(self=@n57, Delegate=@n56)
   > Clear and Invalidate Timer by Handle(Handle=Get QuickIntroTimer() -> (QuickIntroTimer:Timer Handle Structure))
   > Clear and Invalidate Timer by Handle(Handle=Get RewardAnimationTimer() -> (RewardAnimationTimer:Timer Handle Structure))
   n59 = Event OnObjectiveEntrySetup!() -> (OutputDelegate, then, UrgentQuestData:"Urgent Quest Data Structure (by ref)") {is=title K2Node_Event}
   > SetVisibility(self=[Get Overlay_IntroOutroTitle(), Get Text_ObjState_AdditionalInfo()], InVisibility=Collapsed)
   > SetVisibility(self=[Get CommonBorderEntries() -> (CommonBorderEntries:Common Border Object Reference), Get Overlay_Root()], InVisibility=Collapsed)
   > Set Is Quest Completed(Is Quest Completed=false)
   > Set Urgent Quest Data(Urgent Quest Data=@n59.UrgentQuestData)
   > "SetText (Text)"(self=Get Text_StageProgressValue() -> (Text_StageProgressValue:Common Text Object Reference))
   > Set Text(self=Get Text_ObjectiveTitle_Main() -> (Text_ObjectiveTitle_Main:Fort Rich Text Block Keybind Object Reference))
   > SetVisibility(self=Get BackPlate_StageProgress() -> (BackPlate_StageProgress:Border Object Reference), InVisibility=Collapsed)
   > Set IsMultiStep(IsMultiStep:bool=false) -> (Output_Get:bool)
   > n60 = Sequence()
   > n61 = GetDynamicMaterial(self=Get Image_Main_Bg())
   > Set EntryDynamicMaterial(EntryDynamicMaterial:Material Instance Dynamic Object Reference=@n61) -> (Output_Get:Material Instance Dynamic Object Reference)
   > Update Objective Styling()
   > UpdateTaskTitle()
   > n62 = Sequence()
   n63 = Break Urgent Quest Data!(UrgentQuestData=Get Urgent Quest Data()) -> (AlertIcon:Soft Object Reference, CharacterIcon)
   n64 = IsValidSoftObjectReference(SoftObjectReference=@n63.CharacterIcon)
   > PortraitSetUp!(execute, self, HasIcon:bool=@n64) -> (then)
   > Branch(Condition=@n64)
   > n65 = Async Load Asset(Asset=@n63.CharacterIcon)
   n66 = Cast To Texture2D(execute=@n65.Completed, Object=@n65)
   > SetBrushFromTexture!(
    execute
    self:Image Object Reference=Get Image_Icon() -> (Image_Icon:Fort Lazy Image Object Reference)
    Texture:Texture 2D Object Reference=@n66
    bMatchSize:bool=false
   ) -> (then)
   n67 = GetMatchIcon!(UrgentQuestData:Urgent Quest Data Structure=Get Urgent Quest Data()) -> (MatchIcon:Soft Object Reference)
   n68 = Async Load Asset(execute=@n62.then_1, Asset=@n67)
   n69 = NOT Boolean(A=IsAssetNull!(Asset:Soft Object Reference=@n67) -> (ReturnValue:bool))
   n70 = Get ShowIconImage()
   SetVisibility(
    execute=@n68.Completed
    self=[Get TopIconSpacer(), Get BigIcon() -> (BigIcon:Fort Lazy Image Object Reference)]
    InVisibility=Select(Option 0:ESlateVisibility Enum=Collapsed, Option 1:ESlateVisibility Enum=SelfHitTestInvisible, Index=OR Boolean(A=@n70, B=@n69)) -> (ReturnValue:ESlateVisibility Enum)
   )
   > Branch(Condition=@n69)
   > n71 = Cast To Texture2D(Object=@n68)
   > n72 = GetDynamicMaterial(self=Get BigIcon())
   > SetTextureParameterValue(self=@n72, ParameterName=Item, Value=@n71)
   n73 = Get bRecentlyStarted() -> (bRecentlyStarted:bool)
   Set StateEnum(
    execute=@n60.then_1
    StateEnum=Select(Option 0:Enum_UrgentQuestState Enum=NewEnumerator1, Option 1:Enum_UrgentQuestState Enum=NewEnumerator0, Index=@n73) -> (ReturnValue:Enum_UrgentQuestState Enum)
   ) -> (Output_Get=NewEnumerator0)
   > n74 = Branch(Condition=@n73)
   > UpdateObjectiveState(IsEntrySetup=true)
   n75 = Get Urgent Quest Data()
   > n76 = Set Timer by Event(
    Delegate=Create Event()
    Time=Select(
     Option 0:float=0.1
     Option 1:float=Get QuickIntroDelay() -> (QuickIntroDelay:double)
     Index=Break Urgent Quest Data!(UrgentQuestData=@n75) -> (VisualStyleTag:Gameplay Tag Container Structure, bQuickIntro:bool=~, bRewardPersist:bool=~, CharacterIcon).bQuickIntro
    ) -> (ReturnValue:float)
    bLooping=false
   )
   > Set QuickIntroTimer(QuickIntroTimer:Timer Handle Structure=@n76) -> (Output_Get:Timer Handle Structure)
   SetVisibility(execute=@n74.else, self=Get Overlay_Root(), InVisibility=SelfHitTestInvisible)
   > SetVisibility(self=[Get Overlay_IntroOutroTitle(), Get Text_ObjState_AdditionalInfo()], InVisibility=Collapsed)
   Event OnObjectiveEntriesInitialized() {is=title K2Node_Event}
   > SetCurrentOptionalEntry()
   > SetVisibility(
    self=Get Text_ObjectiveTitle_Main()
    InVisibility=Select(Option 0:ESlateVisibility Enum=Collapsed, Option 1:ESlateVisibility Enum=HitTestInvisible, Index=Get ShowTitle() -> (ShowTitle:bool)) -> (ReturnValue:ESlateVisibility Enum)
   )
   > SetMultiStep()
   > n77 = Branch(Condition=Get IsMultiStepExpanded())
   > Progress Multi Step Hide!(execute:Exec) -> (then:Exec) {is=title K2Node_MacroInstance}
   SetMultiStepText(execute=@n77.else)
   > SetEntryProgressMultiStep()
   n78 = Event OnTimerTextBlockUpdated!() -> (OutputDelegate, then, RemainingTime:double) {is=title K2Node_Event}
   > n79 = Branch(Condition="float <= float"(A=@n78.RemainingTime, B=Get TimerCountdownMax() -> (TimerCountdownMax:double)) -> (ReturnValue=~))
   > n80 = Branch(Condition="In Range (Float)"!(
     Value:double=@n78.RemainingTime
     Min:double=Get TimerCountdownMin() -> (TimerCountdownMin:double)
     Max:double=Get TimerCountdownMax()
     InclusiveMin:bool=true
     InclusiveMax:bool=true
    ) -> (ReturnValue:bool))
   > SpawnSound2D(
    Sound=/Game/Sounds/Fort_UI_Sounds/SharedQuests/JobBoard_Timer_Tick_Cue
    VolumeMultiplier=1
    PitchMultiplier=1
    StartTime=0
    bPersistAcrossLevelTransition=false
    bAutoDestroy=true
   )
   Do Once(execute=@n80.else, Reset=@n82.then)
   > SpawnSound2D(
    Sound=/Game/Sounds/Fort_UI_Sounds/SharedQuests/JobBoard_Timer_FinalTick_Cue
    VolumeMultiplier=1
    PitchMultiplier=1
    StartTime=0
    bPersistAcrossLevelTransition=false
    bAutoDestroy=true
   )
   n81 = Get TimerCountdownMax()
   n82 = Branch(execute=@n79.else, Condition="Equal (Float)"(A=float + float(A=@n81, B=1), B=@n78.RemainingTime))
   n83 = Event OnObjectiveRewardSet!() -> (OutputDelegate, then, RewardQuantity:int=0, RewardType:EFortEarnedSubRewardType Enum=XP) {is=title K2Node_Event}
   > Add!(
    execute
    TargetSet:Set of Reward Quantity Pair Structures=Get RewardQuantities()
    NewItem_RewardQuantity_2_#31:int=@n83.RewardQuantity
    NewItem_RewardType_5_#32:EFortEarnedSubRewardType Enum=@n83.RewardType
   ) -> (then) {is=title K2Node_CallFunction}
   UpdateTaskTitle()
   Set Rewards(Quantity=0, Type=XP)
   PortraitSetUp(HasIcon=Get Has Icon() -> (Has Icon:bool))
   Event OnUpdateBountyThreatInformationVisibilityChange!() -> (OutputDelegate, then, NewVisibility:ESlateVisibility Enum=Visible) {is=title K2Node_Event}
   Format Text!(Format="LOCTEXT(\"{Main} Name\")", Main:Wildcard) -> (Result)
  ]}
 OnEntryComplete {nodes [OnEntryComplete!() -> (then, ObjectiveEntry:WBP Athena Objectives Entry Object Reference) {is=title K2Node_FunctionEntry}]}
 OnMaterialOverride {nodes [OnMaterialOverride() {is=title K2Node_FunctionEntry}]}
 OnSetUrgentQuest {nodes [OnSetUrgentQuest() {is=title K2Node_FunctionEntry}]}
 PlayFaliureAnims {nodes [
   n1 = Get WBP_Athena_Objectives_Icon()
   BindToAnimationFinished(
    self=@n1
    Animation=Get Anim_Outro_Failed!(self:WBP Athena Objectives Icon Object Reference=@n1) -> (Anim_Outro_Failed:Widget Animation Object Reference)
    Delegate=Create Event()
   )
   > PlayOutroFailedAnim(self:WBP Athena Objectives Icon Object Reference=Get WBP_Athena_Objectives_Icon())
   PlayFaliureAnims() {is=title K2Node_FunctionEntry}
   > Event Dismiss Objective Entry()
  ]}
 PlayQuestCompleteSound {nodes [
   PlayQuestCompleteSound() {is=title K2Node_FunctionEntry}
   > n1 = Branch(Condition="Equal (Name)"(A=Get Style Type() -> (Style Type:Name=None), B=MatchQuest))
   PlaySound2D(execute=@n1.else, Sound=Get Quest Complete Sound() -> (Quest Complete Sound:Sound Base Object Reference))
  ]}
 PlayQuestFailedSound {nodes [
   PlayQuestFailedSound() {is=title K2Node_FunctionEntry}
   > n1 = Branch(Condition="Equal (Name)"(A=Get Style Type() -> (Style Type=None), B=MatchQuest))
   PlaySound2D(execute=@n1.else, Sound=Get Quest Failed Sound() -> (Quest Failed Sound:Sound Base Object Reference))
  ]}
 PlayRewardIconSound {nodes [
   PlayRewardIconSound() {is=title K2Node_FunctionEntry}
   > n1 = Branch(Condition="Equal (Name)"(A=Get Style Type() -> (Style Type=None), B=MatchQuest))
   PlaySound2D(execute=@n1.else, Sound=Get Reward Icon Sound() -> (Reward Icon Sound:Sound Base Object Reference))
  ]}
 PlaySuccessAnims {nodes [
   PlaySuccessAnims() {is=title K2Node_FunctionEntry}
   n1 = Get WBP_Athena_Objectives_Icon()
   > BindToAnimationFinished(
    self=@n1
    Animation=Get Anim_Outro_Success!(self:WBP Athena Objectives Icon Object Reference=@n1) -> (Anim_Outro_Success:Widget Animation Object Reference)
    Delegate=Create Event()
   )
   > PlayOutroSuccessAnim(self:WBP Athena Objectives Icon Object Reference=Get WBP_Athena_Objectives_Icon())
  ]}
 PlayThreatSound {nodes [
   PlayThreatSound() {is=title K2Node_FunctionEntry}
   > PlaySound2D(Sound=Select!(
     Low:Sound Base Object Reference=Get Threat Low Sound() -> (Threat Low Sound:Sound Base Object Reference)
     Medium:Sound Base Object Reference=Get Threat Medium Sound() -> (Threat Medium Sound:Sound Base Object Reference)
     High:Sound Base Object Reference=Get Threat High Sound() -> (Threat High Sound:Sound Base Object Reference)
     Index:EPlayerBountyThreatLevel Enum=Get Bounty Threat Level() -> (Bounty Threat Level:EPlayerBountyThreatLevel Enum=Low)
    ) -> (ReturnValue:Sound Base Object Reference))
  ]}
 PortraitSetUp {nodes [
   n1 = PortraitSetUp!() -> (then, HasIcon:bool=~) {is=title K2Node_FunctionEntry}
   > SetVisibility(
    self=Get Image_Icon()
    InVisibility=Select!(Option 0:ESlateVisibility Enum=Collapsed, Option 1:ESlateVisibility Enum=SelfHitTestInvisible, Index:bool=@n1) -> (ReturnValue:ESlateVisibility Enum)
   )
  ]}
 ProgressMultiStepHide {nodes [
   Inputs()
   > SetVisibility(self=[Get BackPlate_StageProgress(), Get Text_StageProgressValue()], InVisibility=Collapsed)
   > n1 = Switch on ESlateVisibility!(
    execute
    Selection:ESlateVisibility Enum=Get Visibility!(self:Widget Object Reference=Get Text_ObjectiveTitle_Main()) -> (Visibility:ESlateVisibility Enum=Visible)
   ) -> (Visible:Exec, Collapsed:Exec, Hidden:Exec, HitTestInvisible:Exec, SelfHitTestInvisible:Exec) {is=title K2Node_SwitchEnum}
   n2 = Get BigIcon()
   SetVisibility(
    execute=@n1.Collapsed
    self=Get MultiStepCollapsed() -> (MultiStepCollapsed:Overlay Object Reference)
    InVisibility=Select(Option 0=Collapsed, Option 1=SelfHitTestInvisible, Index=IsVisible(self=@n2))
   )
   > Outputs!(then:Exec)
  ]}
 ResetCollapsed {nodes [
   ResetCollapsed() {is=title K2Node_FunctionEntry}
   > SetVisibility(self=Get CommonBorderEntries(), InVisibility=SelfHitTestInvisible)
   > Set IsMultiStepExpanded(IsMultiStepExpanded:bool=false) -> (Output_Get:bool)
   > Set CurrentOptionalEntry(CurrentOptionalEntry:int=0) -> (Output_Get:int=0)
   > SetCurrentOptionalEntry()
   > Clear(TargetSet=Get RewardQuantities())
  ]}
 ResetWidgetState {nodes [
   ResetWidgetState() {is=title K2Node_FunctionEntry}
   n1 = Get Overlay_Root()
   > SetRenderTranslation(self=@n1)
   > SetRenderOpacity(self=@n1, InOpacity=1)
   n2 = Get Overlay_RewardIcon() -> (Overlay_RewardIcon:Overlay Object Reference)
   > SetVisibility(self=@n2, InVisibility=Collapsed)
   > SetRenderOpacity(self=@n2, InOpacity=0)
   > SetVisibility(self=Get WBP_Objectives_Progress(), InVisibility=HitTestInvisible)
   > Set StateEnum(StateEnum=NewEnumerator4) -> (Output_Get=NewEnumerator0)
  ]}
 Set Rewards {nodes [
   n1 = Set Rewards!() -> (then, Quantity:int, Type:EFortEarnedSubRewardType Enum=XP) {is=title K2Node_FunctionEntry}
   > n2 = Set PrimaryObjRewardQuantity(PrimaryObjRewardQuantity:int=@n1) -> (Output_Get:int=0)
   > n3 = Set PrimaryObjRewardType(PrimaryObjRewardType:EFortEarnedSubRewardType Enum=@n1.Type) -> (Output_Get:EFortEarnedSubRewardType Enum=XP)
   > Set Reward Quantity and Type!(
    execute
    self:WBP Athena Objectives Icon Object Reference=Get WBP_Athena_Objectives_Icon()
    InText:Text="To Text (Integer)"(Value=@n2)
    InType:EFortEarnedSubRewardType Enum=@n3
   ) -> (then) {is=title K2Node_CallFunction}
   > n4 = Set HasReward(HasReward:bool="integer > integer"(A=Get PrimaryObjRewardQuantity() -> (PrimaryObjRewardQuantity:int=0))) -> (Output_Get:bool)
   > HideRewardText(self=Get WBP_Athena_Objectives_Icon(), HideReward=NOT Boolean(A=@n4))
  ]}
 SetCurrentOptionalEntry {nodes [
   n1 = SetCurrentOptionalEntry!() -> (then, OptionalEntry:WBP Athena Objectives Optional Entry Object Reference) {is=title K2Node_FunctionEntry}
   n2 = Get CurrentObjectiveOptionalEntry() -> (CurrentObjectiveOptionalEntry:WBP Athena Objectives Optional Entry Object Reference)
   > n3 = Delta Gate Wildcard!(
    execute:Exec
    Variable:"WBP Athena Objectives Optional Entry Object Reference (by ref)"=@n2
    New Value:WBP Athena Objectives Optional Entry Object Reference=@n1
   ) -> (
    Success:Exec
    Updated Variable:"WBP Athena Objectives Optional Entry Object Reference (by ref)"
    Previous Variable:WBP Athena Objectives Optional Entry Object Reference
   ) {is=title K2Node_MacroInstance}
   > n4 = Sequence()
   > Is Valid(InputObject=@n3.Previous Variable)
   > Unbind all Events from On Optional Complete!(execute:Exec, self:WBP Athena Objectives Optional Entry Object Reference=@n3.Previous Variable) -> (then:Exec) {is=title K2Node_ClearDelegate}
   > Unbind all Events from On Optional Progress!(execute:Exec, self:WBP Athena Objectives Optional Entry Object Reference=@n3.Previous Variable) -> (then:Exec) {is=title K2Node_ClearDelegate}
   Is Valid(exec=@n4.then_1, InputObject=@n2)
   > Bind Event to On Optional Progress(self:WBP Athena Objectives Optional Entry Object Reference=Get CurrentObjectiveOptionalEntry(), Delegate=Create Event()) {is=title K2Node_AddDelegate}
   > Bind Event to On Optional Complete(self:WBP Athena Objectives Optional Entry Object Reference=Get CurrentObjectiveOptionalEntry(), Delegate=Create Event()) {is=title K2Node_AddDelegate}
  ]}
 SetCurrentStage {nodes [
   n1 = SetCurrentStage!() -> (then, CurrentStage:int, TotalStages:int) {is=title K2Node_FunctionEntry}
   > n2 = Set Text(
    Text=Format Text!(Format="LOCTEXT(\"{CurrentStage} of {TotalStages} \")", CurrentStage:int=@n1, TotalStages:int=@n1.TotalStages) -> (Result)
    self:Text Object Reference=Get Text_StageProgressValue()
   )
   n3 = TextIsEmpty(InText=@n2)
   n4 = "integer < integer"(A=@n1.TotalStages, B=2)
   n5 = Get BackPlate_StageProgress()
   > SetVisibility(self=@n5, InVisibility=Select(Option 0=SelfHitTestInvisible, Option 1=Collapsed, Index=OR Boolean(A=@n3, B=@n4)))
   > n6 = GetDynamicMaterial(self:Border Object Reference=@n5)
   > SetScalarParameterValue(self=@n6, ParameterName=Backplate Switch, Value=4)
  ]}
 SetEntryColorTile {nodes [
   n1 = GetDynamicMaterial(self:Image Object Reference=Get ObjectiveBGCollapse() -> (ObjectiveBGCollapse:Image Object Reference))
   > SetVectorParameterValue(self=@n1, ParameterName=Color, Value=Get ObjectiveBGColor() -> (ObjectiveBGColor:Linear Color Structure))
   n2 = SetEntryColorTile!() -> (
    then
    Material Instance:Material Instance Dynamic Object Reference
    BGColor1:Linear Color Structure
    BGColor2:Linear Color Structure
    TextColorObjective:Slate Color Structure
    TimerTextColor:Slate Color Structure
    TimerBorderColor1:Linear Color Structure
    TimerBorderColor2:Linear Color Structure
    ProgressBarStyle:Progress Bar Style Structure
    MaterialOverride:Material Instance Soft Object Reference
    ProgressBGColor:Linear Color Structure
   ) {is=title K2Node_FunctionEntry}
   > n3 = Branch(Condition=IsValidSoftObjectReference(SoftObjectReference=@n2.MaterialOverride))
   > n4 = OnMaterialOverride_Event!(execute, self, MaterialOverride:Material Instance Soft Object Reference=@n2.MaterialOverride) -> (then) {is=title K2Node_CallFunction}
   SetVectorParameterValue(
    execute=[@n4.then, @n3.else]
    self=Get EntryDynamicMaterial() -> (EntryDynamicMaterial:Material Instance Dynamic Object Reference)
    ParameterName=BackgroundColor
    Value=Get BGColor1() -> (BGColor1:Linear Color Structure)
   )
   > SetVectorParameterValue(self=Get EntryDynamicMaterial(), ParameterName=ColorSecond, Value=Get BGColor2() -> (BGColor2:Linear Color Structure))
   > SetDefaultColorAndOpacity!(
    execute
    self:Rich Text Block Object Reference=Get Text_ObjectiveTitle() -> (Text_ObjectiveTitle:Common Rich Text Block Object Reference)
    InColorAndOpacity:Slate Color Structure=Get TextColorObjective() -> (TextColorObjective:Slate Color Structure)
   ) -> (then)
   > SetColorAndOpacity(
    self=Get Text_Timer() -> (Text_Timer:Fort Event Date Time Text Block Object Reference)
    InColorAndOpacity=Get TimerTextColor() -> (TimerTextColor:Slate Color Structure)
   )
   > Set ColorAndOpacity!(
    execute
    ColorAndOpacity:Linear Color Structure=Get ProgressBGColor() -> (ProgressBGColor:Linear Color Structure)
    self:Image Object Reference=Get Progress_BG() -> (Progress_BG:Image Object Reference)
   ) -> (then, Output_Get:Linear Color Structure)
   > Set WidgetStyle!(
    execute
    WidgetStyle:Progress Bar Style Structure=Get ProgressBarStyle() -> (ProgressBarStyle:Progress Bar Style Structure)
    self:Progress Bar Object Reference=Get ProgressBar_ObjectiveProgress() -> (ProgressBar_ObjectiveProgress:Progress Bar Object Reference)
   ) -> (then, Output_Get:Progress Bar Style Structure)
  ]}
 SetEntryProgressMultiStep {nodes [
   SetEntryProgressMultiStep() {is=title K2Node_FunctionEntry}
   > n1 = GetProgressText!(
    execute
    self:WBP Athena Objectives Progress Object Reference=Get WBP_Athena_Objectives_Progress!(self:WBP Athena Objectives Optional Entry Object Reference=Get CurrentObjectiveOptionalEntry()) -> (WBP_Athena_Objectives_Progress:WBP Athena Objectives Progress Object Reference)
   ) -> (then, currentProgress:Text)
   > SetObjectiveProgressText(self:WBP Athena Objectives Progress Object Reference=Get WBP_Objectives_Progress(), Text=@n1)
  ]}
 SetMultiStep {nodes [
   SetMultiStep() {is=title K2Node_FunctionEntry}
   n1 = Get DynamicEntryBox_OptionalEntries()
   > n2 = Set Total Optional Entries(Total Optional Entries:int=GetNumEntries!(self:Dynamic Entry Box Base Object Reference=@n1) -> (ReturnValue:int=0)) -> (Output_Get:int=0)
   > n3 = Branch(Condition="integer > integer"(A=@n2))
   > Set isMultiStep(isMultiStep:bool=true) -> (Output_Get:bool)
   > n4 = For Each Loop(Array=GetAllEntries(self=@n1))
   > n5 = Cast To WBP_Athena_Objectives_OptionalEntry(Object=@n4)
   > Set ObjectiveOptionalEntry(ObjectiveOptionalEntry:WBP Athena Objectives Optional Entry Object Reference=@n5) -> (Output_Get:WBP Athena Objectives Optional Entry Object Reference)
   n6 = int + int(A=@n4.Array Index)
   > SetIndexEntry!(
    execute
    self:WBP Athena Objectives Optional Entry Object Reference=Get ObjectiveOptionalEntry() -> (ObjectiveOptionalEntry:WBP Athena Objectives Optional Entry Object Reference)
    TopIndex:int=@n6
    Count:int=Get Total Optional Entries() -> (Total Optional Entries=0)
   ) -> (then)
   > n7 = Branch(Condition=Get IsMultiStepExpanded())
   n8 = Branch(
    execute=@n7.else
    Condition=Get OptionalObjectiveCompleted!(self:WBP Athena Objectives Optional Entry Object Reference=Get ObjectiveOptionalEntry()) -> (OptionalObjectiveCompleted:bool)
   )
   n9 = Get CurrentObjectiveOptionalEntry()
   n10 = Is Valid(exec=@n8.else, InputObject=@n9)
   Set CurrentOptionalEntry(execute=[@n11.then, @n10.Is Not Valid], CurrentOptionalEntry=@n6) -> (Output_Get=0)
   > SetCurrentStage!(
    execute
    self
    CurrentStage:int=Get CurrentOptionalEntry() -> (CurrentOptionalEntry=0)
    TotalStages:int=Get Total Optional Entries() -> (Total Optional Entries=0)
   ) -> (then) {is=title K2Node_CallFunction}
   > SetCurrentOptionalEntry(OptionalEntry=Get ObjectiveOptionalEntry()) {is=title K2Node_CallFunction}
   n11 = Branch(execute=@n10.Is Valid, Condition=Get OptionalObjectiveCompleted(self=@n9))
   n12 = Branch(execute=[@n4.Completed, @n18.then], Condition=Get IsMultiStepExpanded())
   n13 = Branch(execute=@n12.else, Condition=Get IsMultiStep() -> (IsMultiStep:bool))
   n14 = Get ObjectiveOptionalEntry()
   n15 = Is Valid(exec=[@n17.then, @n13.then], InputObject=@n14)
   Return Node(execute=[@n16.else, @n15.Is Not Valid])
   > n16 = Branch(Condition=Get OptionalObjectiveCompleted(self=@n14))
   > Set CurrentOptionalEntry(CurrentOptionalEntry=int + int(A=Get Total Optional Entries() -> (Total Optional Entries=0))) -> (Output_Get=0)
   > Return Node()
   n17 = SetVisibility(execute=@n13.else, self=Get CommonBorderEntries(), InVisibility=Collapsed)
   Set isMultiStep(execute=@n3.else, isMultiStep=false)
   > n18 = SetCurrentStage(CurrentStage=0, TotalStages=0)
  ]}
 SetMultiStepText {nodes [
   SetMultiStepText() {is=title K2Node_FunctionEntry}
   > n1 = Get!(execute:Exec) -> (then:Exec, else:Exec, CurrentObjectiveOptionalEntry:WBP Athena Objectives Optional Entry Object Reference)
   > SetText(
    self=Get Text_ObjectiveTitle()
    InText=GetText!(self:Rich Text Block Object Reference=Get Text_ObjectiveDescriptionText!(self:Athena Objective Task Entry Base Object Reference=@n1) -> (Text_ObjectiveDescriptionText:Fort Rich Text Block Keybind Object Reference)) -> (ReturnValue:Text)
   )
  ]}
 SetObjectiveProgressText {nodes [
   n1 = SetObjectiveProgressText!() -> (then, Text:Text) {is=title K2Node_FunctionEntry}
   > SetObjectiveProgressText(self=Get WBP_Objectives_Progress(), Text=@n1) {is=title K2Node_CallFunction}
   > n2 = Branch(Condition=Get Is ABounty() -> (Is ABounty:bool))
   SetVisibility(execute=@n2.else, self=Get Image_fx_Reward_Glow() -> (Image_fx_Reward_Glow:Image Object Reference), InVisibility=SelfHitTestInvisible)
  ]}
 SetStageTransitionText {nodes [
   n1 = SetStageTransitionText!() -> (then, CurrentStage:int) {is=title K2Node_FunctionEntry}
   > Set Text(
    Text=Format Text!(Format="LOCTEXT(\"<Transition>Stage {CurrentStage} Complete</>\")", CurrentStage=@n1) -> (Result)
    self:Rich Text Block Object Reference=Get Text_ObjectiveTransition()
   )
  ]}
 Unbind from Animations {nodes [
   Unbind from Animations() {is=title K2Node_FunctionEntry}
   n1 = Get WBP_Athena_Objectives_Icon()
   > UnbindAllFromAnimationFinished!(execute, self:User Widget Object Reference=@n1, Animation:Widget Animation Object Reference=Get Anim_Outro_Success(self=@n1)) -> (then)
   n2 = Get WBP_Athena_Objectives_Icon()
   > UnbindAllFromAnimationFinished(self=@n2, Animation=Get Anim_Outro_Failed(self=@n2))
  ]}
 Update Bounty Threat {nodes [
   Update Bounty Threat() {is=title K2Node_FunctionEntry}
   n1 = Get Bounty Threat Level() -> (Bounty Threat Level=Low)
   > SetPercent!(
    execute
    self:Progress Bar Object Reference=Get ProgressBar_ObjectiveProgress()
    InPercent:float=Select!(Low:double=0.33, Medium:double=0.66, High:double=1, Index:EPlayerBountyThreatLevel Enum=@n1) -> (ReturnValue:double)
   ) -> (then)
   > SetVisibility(
    self=Get Image_BountyThreatGlow() -> (Image_BountyThreatGlow:Image Object Reference)
    InVisibility=Select(Low:ESlateVisibility Enum=Collapsed, Medium:ESlateVisibility Enum=Collapsed, High:ESlateVisibility Enum=HitTestInvisible, Index=@n1) -> (ReturnValue:ESlateVisibility Enum)
   )
  ]}
 Update Objective Styling {nodes [
   Update Objective Styling() {is=title K2Node_FunctionEntry}
   n1 = Get Urgent Quest Data()
   n2 = GetStylingTypeFromTag!(UrgentQuestData:Urgent Quest Data Structure=@n1) -> (OutStyleType:Name=None, IsABounty:bool)
   > Set Style Type(Style Type:Name=@n2) -> (Output_Get:Name=None)
   > Set Is ABounty(Is ABounty:bool=@n2.IsABounty) -> (Output_Get:bool)
   n3 = Get Style Type() -> (Style Type=None)
   n4 = GetMainBGColor!(Type:Name=@n3, UrgentQuestData:Urgent Quest Data Structure=@n1) -> (Title BG Color 01:Linear Color Structure, Title BG Color 02:Linear Color Structure)
   n5 = GetTextColor!(Type:Name=@n3) -> (Objective Text Color:Linear Color Structure, Timer Text Color:Linear Color Structure)
   > SetEntryColorTile!(
    execute
    self
    Material Instance:Material Instance Dynamic Object Reference=Get EntryDynamicMaterial()
    BGColor1:Linear Color Structure=@n4
    BGColor2:Linear Color Structure=@n4.Title BG Color 02
    TextColorObjective:Slate Color Structure="To SlateColor (Color)"!(Value:Linear Color Structure=@n5) -> (ReturnValue:Slate Color Structure)
    TimerTextColor:Slate Color Structure="To SlateColor (Color)"(Value=@n5.Timer Text Color)
    TimerBorderColor1:Linear Color Structure=(R=0.0157,G=0.0863,B=0.4196,A=1)
    TimerBorderColor2:Linear Color Structure=(R=0.0157,G=0.0863,B=0.4196,A=1)
    ProgressBarStyle:Progress Bar Style Structure=GetProgressBarColor!(Type:Name=@n3) -> (Progress Bar Style:Progress Bar Style Structure)
    MaterialOverride:Material Instance Soft Object Reference=GetMaterialOverride!(Type:Name=@n3) -> (Material Override:Material Instance Soft Object Reference)
    ProgressBGColor:Linear Color Structure="GetObjective&&ProgressBGColor"!(Type:Name=@n3) -> (Progress BG Color:Linear Color Structure)
   ) -> (then) {is=title K2Node_CallFunction}
   n6 = Get Style Type() -> (Style Type=None)
   > SetColorAndOpacity(
    self=Get Text_ObjState() -> (Text_ObjState:Common Text Object Reference)
    InColorAndOpacity=Make Slate Color(SpecifiedColor=GetIntro/OutroTitleColors!(Type:Name=@n6) -> (
      Intro/Outro Text Color:Linear Color Structure
      Intro Shine Color:Linear Color Structure
      Success Shine Color:Linear Color Structure
      Failed Shine Color:Linear Color Structure
     ))
   )
   > SetVisibility(self=Get Image_fx_Reward_Glow(), InVisibility=Hidden)
   > UpdateBountyVisuals!(execute, self, IsBounty:bool=Get Is ABounty()) -> (then)
   > n7 = IsFullBodyIcon!(execute, Type:Name=Get Style Type() -> (Style Type=None)) -> (then, IsFullBodyIcon:bool)
   > SetActiveWidgetIndex(
    self:Common Visibility Switcher Object Reference=Get IconSwitcher() -> (IconSwitcher:Common Visibility Switcher Object Reference)
    Index=Select!(Option 0:int=0, Option 1:int=1, Index:bool=@n7) -> (ReturnValue:int)
   )
  ]}
 Update Reward Quantity {nodes [
   Update Reward Quantity() {is=title K2Node_FunctionEntry}
   > n1 = For Each Loop(Array=GetAllEntries(self=Get DynamicEntryBox_OptionalEntries()))
   > n2 = Cast To WBP_Athena_Objectives_OptionalEntry(Object=@n1)
   > Branch(Condition=Get OptionalObjectiveCompleted(self=@n2))
   > Set PrimaryObjRewardQuantity(PrimaryObjRewardQuantity=int + int(
     A=Get PrimaryObjRewardQuantity() -> (PrimaryObjRewardQuantity=0)
     B=Get OptionalObjRewardQuantity!(self:WBP Athena Objectives Optional Entry Object Reference=@n2) -> (OptionalObjRewardQuantity:int=0)
    )) -> (Output_Get=0)
   n3 = Get PrimaryObjRewardQuantity() -> (PrimaryObjRewardQuantity=0)
   n4 = Branch(execute=@n1.Completed, Condition="integer > integer"(A=@n3))
   > Set Reward Quantity and Type(
    self=Get WBP_Athena_Objectives_Icon()
    InText=Format Text!(Format="LOCTEXT(\"{RewardQuantity}\")", RewardQuantity:int=@n3) -> (Result)
    InType=Get PrimaryObjRewardType() -> (PrimaryObjRewardType:EFortEarnedSubRewardType Enum=XP)
   )
   > Return Node()
   Return Node(execute=@n4.else)
  ]}
 UpdateBountyVisuals {nodes [
   n1 = UpdateBountyVisuals!() -> (then, IsBounty:bool=~) {is=title K2Node_FunctionEntry}
   > n2 = Branch(Condition=@n1)
   > Set Rewards(Quantity=0, Type=Bars) {is=title K2Node_CallFunction}
   n3 = Get Urgent Quest Data()
   > n4 = Branch(Condition="Equal (GameplayTag)"(
     A=Get BountyHunterTag() -> (BountyHunterTag:Gameplay Tag Structure)
     B=Break Urgent Quest Data!(UrgentQuestData=@n3) -> (
      EventTag:Gameplay Tag Structure
      VisualStyleTag
      TotalEventTime:float=~
      bDisplayTimer:bool=~
      bQuestCompleteOnTimeExpiration:bool=~
      bQuickIntro=~
      bQuickOutro:bool=~
      bRewardPersist=~
      bDisplayMultiStepExpanded:bool=~
      EventMainTitle:Text
      EventTitle:Text
      EventSubtitle:Text
      EventSubtitleSecondary:Text
      EventFailureMessage:Text
      EventDescription:Text
      EventDescriptionGamepad:Text
      EventDescriptionTouch:Text
      EventDescriptionOverrides:Array of Urgent Quest Data Desc Overrides Structures
      EventDescriptionOverrideTag:Gameplay Tag Structure
      SortPriority:int
      bHideQuestProgressBar:bool=~
      bDisplayAllObjectivesIndividually:bool=~
      QuestProgressTextFormat:Text
      bReverseProgressBar:bool=~
      bShowBountyThreatInformation:bool=~
      bShowBountyPriceAndDistance:bool=~
      bDisplayStartedEventUI:bool=~
      bIsSharedUrgentQuest:bool=~
      DisplayPlayer:Fort Player State Athena Object Reference
      AcceptingPlayer:Fort Player State Athena Object Reference
      SocialAvatarBrushPtr:Texture 2D Soft Object Reference
      BountyPriceImage:Object Reference
      AlertIcon:Soft Object Reference
      CharacterIcon
      SpecialEventStartedSound:Sound Base Soft Object Reference
      AlertIconBorderColor:Linear Color Structure
      AlertIconBrushSize:Vector 2D Structure
      bOverrideWidgetPrimaryColor:bool=~
      OverriddenWidgetPrimaryColor:Linear Color Structure
      ObjectiveCategory:EObjectiveCategories Enum
     )
    ))
   > SetVisibility(self=[Get Overlay_ProgressBar() -> (Overlay_ProgressBar:Overlay Object Reference), Get WBP_Objectives_Progress()], InVisibility=Collapsed)
   SetVisibility(execute=@n4.else, self=[Get WBP_Objectives_Progress(), Get Overlay_ProgressBar()], InVisibility=SelfHitTestInvisible)
   > SetObjectiveProgressText(self:Self Object Reference, Text="LOCTEXT(\"THREAT\")")
   n5 = Get Urgent Quest Data()
   n6 = Break Progress Bar Style!(ProgressBarStyle:"Progress Bar Style Structure (by ref)"=GetProgressBarColor(Type=GetStylingTypeFromTag(UrgentQuestData=@n5) -> (OutStyleType=None))) -> (BackgroundImage:Slate Brush Structure, FillImage:Slate Brush Structure, MarqueeImage:Slate Brush Structure, EnableFillAnimation:bool=~) {is=title K2Node_BreakStruct}
   > SetColorAndOpacity(
    self:User Widget Object Reference=Get WBP_Objectives_Progress()
    InColorAndOpacity:Linear Color Structure="To Color (SlateColor)"!(Value:"Slate Color Structure (by ref)"=Break Slate Brush!(SlateBrush:"Slate Brush Structure (by ref)"=@n6.FillImage) -> (
      TintColor:Slate Color Structure
      DrawAs:ESlateBrushDrawType Enum
      Tiling:ESlateBrushTileType Enum
      Mirroring:ESlateBrushMirrorType Enum
      ImageSize:Vector 2D Structure
      Margin:Margin Structure
      ResourceObject:Object Reference
      OutlineSettings:Slate Brush Outline Settings Structure
     ) {is=title K2Node_BreakStruct}) -> (ReturnValue:Linear Color Structure)
   )
   SetColorAndOpacity(execute=@n2.else, self=Get WBP_Objectives_Progress(), InColorAndOpacity=(R=1,G=1,B=1,A=1))
  ]}
 UpdateMultiQuestInMain {nodes [
   n1 = UpdateMultiQuestInMain!() -> (then, ObjectiveProgress:Text) {is=title K2Node_FunctionEntry}
   > SetObjectiveProgressText(self:WBP Athena Objectives Progress Object Reference=Get WBP_Objectives_Progress(), Text=@n1)
  ]}
 UpdateObjectiveState {nodes [
   n1 = Get EventSubtitle() -> (EventSubtitle:Text)
   n2 = Get Text_ObjState_AdditionalInfo()
   "SetText (Text)"(self=@n2, InText=@n1)
   > n3 = SetVisibility(
    self=@n2
    InVisibility=Select(Option 0:ESlateVisibility Enum=HitTestInvisible, Option 1:ESlateVisibility Enum=Collapsed, Index=TextIsEmpty(InText=@n1)) -> (ReturnValue:ESlateVisibility Enum)
   )
   n4 = Branch(execute=[@n3.then, @n18.else], Condition=Get Is Quest Completed())
   n5 = Get Completion Path Context()
   n6 = HasTag(TagContainer=@n5, Tag=(TagName="Athena.Quests.CompletionPath.Failed"), bExactMatch=false)
   n7 = Get EventFailureMessage() -> (EventFailureMessage:Text)
   n8 = SelectText(
    A=Get EventTitle() -> (EventTitle:Text)
    B=SelectText(
     A="LOCTEXT(\"Poached\")"
     B=SelectText(A=@n7, B="LOCTEXT(\"Completed!\")", bPickA=@n6)
     bPickA=HasTag(TagContainer=@n5, Tag=(TagName="Athena.Quests.CompletionPath.Failed.Poached"), bExactMatch=false)
    )
    bPickA=Get Is ABounty()
   )
   > "SetText (Text)"(self=Get Text_ObjState(), InText=@n8)
   > SetVisibility(self=Get Overlay_IntroOutroTitle(), InVisibility=Select(Option 0=SelfHitTestInvisible, Option 1=Collapsed, Index=TextIsEmpty(InText=@n8)))
   > n9 = GetDynamicMaterial(self=Get Image_fx_Shine() -> (Image_fx_Shine:Image Object Reference))
   n10 = GetIntro/OutroTitleColors(Type=Get Style Type() -> (Style Type=None))
   n11 = Get Completion Path Context()
   n12 = OR Boolean(
    A=HasTag(TagContainer=@n11, Tag=(TagName="Athena.Quests.CompletionPath.Failed"), bExactMatch=false)
    B=HasTag(TagContainer=@n11, Tag=(TagName="Athena.Quests.CompletionPath.Failed.Poached"), bExactMatch=false)
   )
   > SetVectorParameterValue(self=@n9, ParameterName=Color, Value=SelectColor(A=@n10.Success Shine Color, B=@n10.Failed Shine Color, bPickA=NOT Boolean(A=@n12)))
   n13 = SelectText(A=Get EventTitle(), B=Get EventFailureMessage(), bPickA=Get Is ABounty())
   "SetText (Text)"(execute=@n4.else, self=Get Text_ObjState(), InText=@n13)
   > SetVisibility(self=Get Overlay_IntroOutroTitle(), InVisibility=Select(Option 0=SelfHitTestInvisible, Option 1=Collapsed, Index=TextIsEmpty(InText=@n13)))
   > n14 = GetDynamicMaterial(self=Get Image_fx_Shine())
   > SetVectorParameterValue(self=@n14, ParameterName=Color, Value=GetIntro/OutroTitleColors(Type=Get Style Type() -> (Style Type=None)).Failed Shine Color)
   UpdateObjectiveState!() -> (then, IsEntrySetup:bool=~) {is=title K2Node_FunctionEntry}
   n15 = Break Urgent Quest Data!(UrgentQuestData=Get Urgent Quest Data()) -> (
    EventTag
    VisualStyleTag
    bQuickIntro=~
    bQuickOutro=~
    bRewardPersist=~
    EventTitle
    EventSubtitle
    EventFailureMessage
    CharacterIcon
    SpecialEventStartedSound
   )
   > Set MessageEventTag(MessageEventTag:Gameplay Tag Structure=@n15) -> (Output_Get:Gameplay Tag Structure)
   > n16 = Branch(Condition=TextIsEmpty(InText=@n15.EventFailureMessage))
   Set EventTitle(execute=[@n16.then, @n23.then], EventTitle:Text=@n15.EventTitle) -> (Output_Get:Text)
   > Set EventSubtitle(EventSubtitle:Text=@n15.EventSubtitle) -> (Output_Get:Text)
   > Set EventStartedSoundOverride(EventStartedSoundOverride:Sound Base Soft Object Reference=@n15.SpecialEventStartedSound) -> (Output_Get:Sound Base Soft Object Reference)
   > n17 = Branch(Condition="Equal (Name)"(A=Get Style Type() -> (Style Type=None), B=MatchQuest))
   > SetVisibility(self=Get Overlay_IntroOutroTitle(), InVisibility=Collapsed)
   n18 = Branch(execute=@n17.else, Condition=Get IsEntrySetup() -> (IsEntrySetup:bool))
   n19 = Get EventTitle()
   > "SetText (Text)"(self=Get Text_ObjState(), InText=@n19)
   > SetVisibility(self=Get Overlay_IntroOutroTitle(), InVisibility=Select(Option 0=SelfHitTestInvisible, Option 1=Collapsed, Index=TextIsEmpty(InText=@n19)))
   > n20 = GetDynamicMaterial(self=Get Image_fx_Shine())
   > SetVectorParameterValue(self=@n20, ParameterName=Color, Value=GetIntro/OutroTitleColors(Type=Get Style Type() -> (Style Type=None)).Intro Shine Color)
   n21 = Get EventStartedSoundOverride() -> (EventStartedSoundOverride:Sound Base Soft Object Reference)
   > n22 = Branch(Condition=IsAssetNull(Asset=@n21))
   > SpawnSound2D(
    Sound=/Game/Sounds/Fort_UI_Sounds/SharedQuests/JobBoard_ShadowBriefing_Accept_Cue
    VolumeMultiplier=1
    PitchMultiplier=1
    StartTime=0
    bPersistAcrossLevelTransition=false
    bAutoDestroy=true
   )
   SpawnSound2D(
    execute=@n22.else
    Sound=Resolve Soft Reference!(Input:Sound Base Soft Object Reference=@n21) -> (Output:Sound Base Object Reference)
    VolumeMultiplier=1
    PitchMultiplier=1
    StartTime=0
    bPersistAcrossLevelTransition=false
    bAutoDestroy=true
   )
   n23 = Set EventFailureMessage(execute=@n16.else, EventFailureMessage:Text=@n15.EventFailureMessage) -> (Output_Get:Text)
  ]}
 UpdateTaskTitle {nodes [
   UpdateTaskTitle() {is=title K2Node_FunctionEntry}
   n1 = Break Urgent Quest Data!(UrgentQuestData=Get Urgent Quest Data()) -> (
    VisualStyleTag
    bDisplayTimer:bool=~
    bQuickIntro=~
    bQuickOutro=~
    bRewardPersist=~
    bDisplayMultiStepExpanded:bool=~
    EventMainTitle:Text
    EventDescriptionOverrides:Array of Urgent Quest Data Desc Overrides Structures
    EventDescriptionOverrideTag:Gameplay Tag Structure
    CharacterIcon
   )
   > Set IsMultiStepExpanded(IsMultiStepExpanded=@n1.bDisplayMultiStepExpanded)
   > Call On Set Urgent Quest!(execute:Exec, self:WBP Athena Objectives Entry Object Reference) -> (then:Exec) {is=title K2Node_CallDelegate}
   > n2 = Set Text(Text=TextToUpper!(InText:"Text (by ref)"=@n1.EventMainTitle) -> (ReturnValue:Text), self=Get Text_ObjectiveTitle_Main())
   > Set ShowTitle(ShowTitle:bool=NOT Boolean(A=TextIsEmpty(InText=@n2))) -> (Output_Get:bool)
   > n3 = Branch(Condition=IsGameplayTagValid!(GameplayTag:Gameplay Tag Structure=@n1.EventDescriptionOverrideTag) -> (ReturnValue:bool))
   > n4 = For Each Loop(Array:Array of Urgent Quest Data Desc Overrides Structures=@n1.EventDescriptionOverrides) -> (Array Element:Urgent Quest Data Desc Overrides Structure)
   n5 = Break Urgent Quest Data Desc Overrides!(UrgentQuestDataDescOverrides:"Urgent Quest Data Desc Overrides Structure (by ref)"=@n4) -> (OverrideTag:Gameplay Tag Structure, EventDescription:Text, EventDescriptionGamepad:Text, EventDescriptionTouch:Text) {is=title K2Node_BreakStruct}
   > Branch(Condition="Equal (GameplayTag)"(A=@n1.EventDescriptionOverrideTag, B=@n5))
   n6 = Select!(
    MouseAndKeyboard:"Text (by ref)"=@n5.EventDescription
    Gamepad:"Text (by ref)"=@n5.EventDescriptionGamepad
    Touch:"Text (by ref)"=@n5.EventDescriptionTouch
    Count:"Text (by ref)"=@n5.EventDescription
    Index:ECommonInputType Enum=GetCurrentInputType(self=Get CommonInputSubsystem())
   ) -> (ReturnValue:"Text (by ref)")
   n7 = Select!(Option 0:"Text (by ref)"=@n6, Option 1:"Text (by ref)"=@n5.EventDescription, Index:bool=TextIsEmpty(InText=@n6)) -> (ReturnValue)
   > n8 = Branch(Condition=TextIsEmpty(InText=@n7))
   SetText(execute=@n8.else, self=Get Text_ObjectiveTitle(), InText=@n7)
   > Return Node()
   n9 = Break Urgent Quest Data!(UrgentQuestData=Get Urgent Quest Data()) -> (
    VisualStyleTag
    bDisplayTimer=~
    bQuickIntro=~
    bQuickOutro=~
    bRewardPersist=~
    bDisplayMultiStepExpanded=~
    EventMainTitle
    EventFailureMessage:Text
    EventDescription:Text
    EventDescriptionGamepad:Text
    EventDescriptionTouch:Text
    EventDescriptionOverrides
    EventDescriptionOverrideTag
    bHideQuestProgressBar:bool=~
    bDisplayAllObjectivesIndividually:bool=~
    QuestProgressTextFormat:Text
    bIsSharedUrgentQuest:bool=~
    CharacterIcon
    bOverrideWidgetPrimaryColor:bool=~
    OverriddenWidgetPrimaryColor:Linear Color Structure
    ObjectiveCategory:EObjectiveCategories Enum
   )
   n10 = Select!(
    MouseAndKeyboard:"Text (by ref)"=@n9.EventDescription
    Gamepad:"Text (by ref)"=@n9.EventDescriptionGamepad
    Touch:"Text (by ref)"=@n9.EventDescriptionTouch
    Count:"Text (by ref)"=@n9.EventDescription
    Index:ECommonInputType Enum=GetCurrentInputType(self=Get CommonInputSubsystem())
   ) -> (ReturnValue)
   SetText(
    execute=[@n4.Completed, @n3.else]
    self=Get Text_ObjectiveTitle()
    InText=Select!(Option 0:"Text (by ref)"=@n10, Option 1:"Text (by ref)"=@n9.EventDescription, Index:bool=TextIsEmpty(InText=@n10)) -> (ReturnValue)
   )
  ]}
]
variables [
 AnimPlayed_Intro {type=boolean}
 AnimPlayed_Reward {type=boolean}
 Bounty Threat Level {type=EPlayerBountyThreatLevel, default=Low}
 BountyHunterTag {type=GameplayTag, default {tagName=SpecialEvent.huntingplayer.isHunter}}
 Completion Path Context {type=GameplayTagContainer, default {gameplayTags []}}
 CurrentObjectiveOptionalEntry {type=WBP_Athena_Objectives_OptionalEntry_C, default=null}
 CurrentOptionalEntry {type=integer}
 EntryDynamicMaterial {type=MaterialInstanceDynamic}
 Has Icon {type=boolean, default=true}
 HasReward {type=boolean}
 HasRewards {type=boolean}
 Is ABounty {type=boolean}
 Is Quest Completed {type=boolean}
 IsMultiStep {type=boolean, default=true}
 IsMultiStepExpanded {type=boolean, default=true}
 MessageEventTag {type=GameplayTag, default {tagName=None}}
 MultiStepObjectiveTitle {type=string, default=""}
 ObjectiveOptionalEntry {type=WBP_Athena_Objectives_OptionalEntry_C, default=null}
 OnEntryComplete {}
 OnMaterialOverride {}
 OnSetUrgentQuest {}
 PersistEndTime {type=number, default=1.5}
 PrimaryObjRewardQuantity {type=integer}
 PrimaryObjRewardType {type=EFortEarnedSubRewardType, default=XP}
 Quest Complete Sound {type=SoundBase, default=/Game/Sounds/Fort_UI_Sounds/SharedQuests/JobBoard_ShadowBriefing_Complete_Cue}
 Quest Failed Sound {type=SoundBase, default=/Game/Sounds/Fort_UI_Sounds/SharedQuests/JobBoard_ShadowBriefing_Failure_Cue}
 QuickIntroDelay {type=number, default=1.8}
 QuickIntroPlayback {type=number, default=1.5}
 QuickIntroTimer {type=TimerHandle}
 Reward Icon Sound {type=SoundBase, default=null}
 RewardAnimationTimer {type=TimerHandle}
 RewardCounts {type=integer}
 RewardQuantities {type="array<RewardQuantityPair>", default []}
 ShowIconImage {type=boolean}
 ShowTitle {type=boolean}
 StartState {type=Enum_UrgentQuestState, default=Started}
 StateEnum {type=Enum_UrgentQuestState, default=Started}
 Style Type {type=string, default=None}
 Threat High Sound {type=SoundBase, default=/Game/Sounds/Fort_UI_Sounds/Gameplay/Bounty/Bounty_HunterGettingVeryClose_Cue}
 Threat Low Sound {type=SoundBase, default=null}
 Threat Medium Sound {type=SoundBase, default=/Game/Sounds/Fort_UI_Sounds/Gameplay/Bounty/Bounty_HunterGettingClose_Cue}
 TimerCountdownMax {type=number, default=6}
 TimerCountdownMin {type=number, default=1}
 Total Optional Entries {type=integer}
 Urgent Quest Data {
  type=UrgentQuestData
  default {
   acceptingPlayer=null
   alertIcon=null
   bDisplayAllObjectivesIndividually=true
   bDisplayHUDData=true
   bDisplayMultiStepExpanded=true
   bDisplayStartedEventUI=true
   bDisplayTimer=true
   bHideQuestProgressBar=false
   bIsSharedUrgentQuest=false
   bOverrideWidgetPrimaryColor=false
   bQuestCompleteOnTimeExpiration=false
   bQuickIntro=false
   bQuickOutro=false
   bReverseProgressBar=true
   bRewardPersist=false
   bShowBountyPriceAndDistance=false
   bShowBountyThreatInformation=false
   bountyPriceImage=null
   characterIcon=null
   displayPlayer=null
   eventDescription=Description
   eventDescriptionGamepad=""
   eventDescriptionTouch=""
   eventFailureMessage=""
   eventMainTitle=""
   eventStartTime=0001.01.01-00.00.00
   eventSubtitle=Subtitle
   eventSubtitleSecondary=SubtitleSecondary
   eventTitle=Title
   objectiveCategory=Default
   questProgressTextFormat="{CurrentProgress} / {MaxProgress}"
   socialAvatarBrushPtr=null
   sortPriority=0
   specialEventStartedSound=null
   totalEventTime=60
   alertIconBorderColor=<rgba 1 0.4564 0 1>
   alertIconBrushSize=<0 0>
   eventDescriptionOverrides []
   eventTag {tagName=None}
   eventTimerHandle {}
   overriddenWidgetPrimaryColor=<rgba 1 0.4564 0 1>
   visualStyleTag {gameplayTags []}
  }
 }
]
widgets {
 info {inheritedWidgetCount=7, parentClass=/Script/FortniteUI.AthenaObjectiveEntryBase, rootWidgetClass=&Overlay, widgetCount=55}
 widgets [
  Overlay_Root {widgetClassPath=&Overlay, bIsVariable=true, widget=Overlay_Root}
  SizeBox_EntryWidth {
   widgetClassPath=&SizeBox
   parent=Overlay_Root
   widget=SizeBox_EntryWidth
   properties {bOverride_WidthOverride=true, widthOverride=470}
   slot {class=OverlaySlot, properties {horizontalAlignment=HAlign_Fill}}
  }
  Overlay_EntryContent {widgetClassPath=&Overlay, parent=SizeBox_EntryWidth, widget=Overlay_EntryContent, slot {class=SizeBoxSlot}}
  VerticalBox_0 {
   widgetClassPath=&VerticalBox
   parent=Overlay_EntryContent
   widget=VerticalBox_0
   slot {class=OverlaySlot, properties {horizontalAlignment=HAlign_Fill}}
  }
  TopIconSpacer {
   widgetClassPath=&Spacer
   bIsVariable=true
   parent=VerticalBox_0
   widget=TopIconSpacer
   properties {size=<1 25>}
   slot {class=VerticalBoxSlot, properties {horizontalAlignment=HAlign_Right, verticalAlignment=VAlign_Top}}
  }
  Overlay_IntroOutroTitle {
   widgetClassPath=&Overlay
   bIsVariable=true
   parent=VerticalBox_0
   widget=Overlay_IntroOutroTitle
   properties {visibility=Collapsed}
   slot {class=VerticalBoxSlot, properties {horizontalAlignment=HAlign_Left, verticalAlignment=VAlign_Top, padding=<ltrb 0 0 0 4>}}
  }
  Image_fx_Shine {
   widgetClassPath=&Image
   bIsVariable=true
   parent=Overlay_IntroOutroTitle
   widget=Image_fx_Shine
   properties {
    renderOpacity=0
    visibility=SelfHitTestInvisible
    brush {
     bIsDynamicallyLoaded=false
     drawAs=Image
     imageType=FullColor
     mirroring=NoMirror
     resourceName=None
     resourceObject=/Game/Athena/HUD/Discoverability/materials/MI_UI_Objectives_Glowfx
     tiling=NoTile
     imageSize=<454.6111 30>
     margin=<ltrb 0 0 0 0>
     outlineSettings {
      bUseBrushTransparency=false
      roundingType=HalfHeightRadius
      width=0
      color {colorUseRule=UseColor_Specified, specifiedColor=<rgba 0 0 0 0>}
      cornerRadii=<0 0 0 0>
     }
     tintColor {colorUseRule=UseColor_Specified, specifiedColor=<rgba 1 1 1 1>}
     uVRegion {bIsValid=false, max=<0 0>, min=<0 0>}
    }
   }
   slot {class=OverlaySlot}
  }
  Text_ObjState {
   widgetClassPath=&CommonTextBlock
   bIsVariable=true
   parent=Overlay_IntroOutroTitle
   widget=Text_ObjState
   properties {
    mobileFontSizeMultiplier=1.2
    text=Job Accepted
    textTransformPolicy=ToUpper
    font {
     bForceMonospaced=false
     bMaterialIsStencil=false
     fontMaterial=null
     fontObject=&HeadingNow
     letterSpacing=40
     monospacedWidth=1
     size=15
     skewAmount=0
     typefaceFontName=76 Bold Italic
     outlineSettings {
      bApplyOutlineToDropShadows=false
      bMiteredCorners=false
      bSeparateFillAlpha=false
      outlineMaterial=null
      outlineSize=2
      outlineColor=<rgba 0 0 0 0.5>
     }
    }
    shadowOffset=<0 0>
   }
   slot {class=OverlaySlot, properties {padding=<ltrb 10 -5 0 0>}}
  }
  Overlay_PrimaryObjective {widgetClassPath=&Overlay, parent=VerticalBox_0, widget=Overlay_PrimaryObjective, slot {class=VerticalBoxSlot}}
  Overlay_ObjectiveTitle {
   widgetClassPath=&Overlay
   bIsVariable=true
   parent=Overlay_PrimaryObjective
   widget=Overlay_ObjectiveTitle
   slot {class=OverlaySlot, properties {horizontalAlignment=HAlign_Fill, verticalAlignment=VAlign_Fill}}
  }
  Image_Main_Bg {
   widgetClassPath=&Image
   bIsVariable=true
   parent=Overlay_ObjectiveTitle
   widget=Image_Main_Bg
   properties {
    visibility=SelfHitTestInvisible
    brush {
     bIsDynamicallyLoaded=false
     drawAs=Image
     imageType=FullColor
     mirroring=NoMirror
     resourceName=None
     resourceObject=/Game/Athena/HUD/Discoverability/materials/CategoryMI/MI_UI_UrgentQuestDefault_Background
     tiling=NoTile
     imageSize=<32 32>
     margin=<ltrb 0.5 0.5 0.5 0.5>
     outlineSettings {
      bUseBrushTransparency=false
      roundingType=FixedRadius
      width=-30
      color {colorUseRule=UseColor_Specified, specifiedColor=<rgba 0 0 0 0>}
      cornerRadii=<10 10 10 10>
     }
     tintColor {colorUseRule=UseColor_Specified, specifiedColor=<rgba 1 1 1 0.78>}
     uVRegion {bIsValid=false, max=<0 0>, min=<0 0>}
    }
   }
   slot {class=OverlaySlot, properties {horizontalAlignment=HAlign_Fill, verticalAlignment=VAlign_Fill}}
  }
  Vertical_MainObjective {
   widgetClassPath=&VerticalBox
   bIsVariable=true
   parent=Overlay_ObjectiveTitle
   widget=Vertical_MainObjective
   properties {bOverride_Cursor=true}
   slot {class=OverlaySlot, properties {horizontalAlignment=HAlign_Fill, verticalAlignment=VAlign_Fill}}
  }
  MultiStepCollapsed {
   widgetClassPath=&Overlay
   bIsVariable=true
   parent=Vertical_MainObjective
   widget=MultiStepCollapsed
   properties {bOverride_Cursor=true}
   slot {class=VerticalBoxSlot}
  }
  ObjectiveBGCollapse {
   widgetClassPath=&Image
   bIsVariable=true
   parent=MultiStepCollapsed
   widget=ObjectiveBGCollapse
   properties {brush {
     bIsDynamicallyLoaded=false
     drawAs=Box
     imageType=FullColor
     mirroring=NoMirror
     resourceName=None
     resourceObject=/Game/Athena/HUD/Discoverability/materials/MI_UI_ObjectivesTitleBG
     tiling=NoTile
     imageSize=<32 32>
     margin=<ltrb 0 0 0 0>
     outlineSettings {
      bUseBrushTransparency=false
      roundingType=FixedRadius
      width=0
      color {colorUseRule=UseColor_Specified, specifiedColor=<rgba 0 0 0 0>}
      cornerRadii=<10 10 0 0>
     }
     tintColor {colorUseRule=UseColor_Specified, specifiedColor=<rgba 1 1 1 1>}
     uVRegion {bIsValid=false, max=<0 0>, min=<0 0>}
    }}
   slot {class=OverlaySlot, properties {horizontalAlignment=HAlign_Fill, verticalAlignment=VAlign_Fill}}
  }
  Text_ObjectiveTitle_Main {
   widgetClassPath=/Script/FortniteUI.FortRichTextBlockKeybind
   bIsVariable=true
   parent=MultiStepCollapsed
   widget=Text_ObjectiveTitle_Main
   properties {
    autoWrapText=true
    bOverrideDefaultStyle=true
    keybindWidgetClass=/FortUILibrary/Input/KeybindWidgetDark_Small.KeybindWidgetDark_Small_C
    mobileTextBlockScale=1.2
    text=OBJECTIVE TITLE
    textStyleSet=/Game/Athena/HUD/Discoverability/UrgentQuestRichTextStyles
    decoratorClasses [/Game/Athena/HUD/Discoverability/UrgentQuestRichTextDecorators.UrgentQuestRichTextDecorators_C]
    defaultTextStyleOverride {
     overflowPolicy=Clip
     transformPolicy=None
     colorAndOpacity {colorUseRule=UseColor_Specified, specifiedColor=<rgba 1 1 1 1>}
     font {
      bForceMonospaced=false
      bMaterialIsStencil=false
      fontMaterial=null
      fontObject=&HeadingNow
      letterSpacing=20
      monospacedWidth=1
      size=13.5
      skewAmount=0
      typefaceFontName=76 Bold
      outlineSettings {
       bApplyOutlineToDropShadows=false
       bMiteredCorners=false
       bSeparateFillAlpha=false
       outlineMaterial=null
       outlineSize=0
       outlineColor=<rgba 0 0 0 0.5>
      }
     }
     highlightColor {colorUseRule=UseColor_Specified, specifiedColor=<rgba 0 0 0 1>}
     highlightShape {
      bIsDynamicallyLoaded=false
      drawAs=Image
      imageType=NoImage
      mirroring=NoMirror
      resourceName=None
      resourceObject=null
      tiling=NoTile
      imageSize=<32 32>
      margin=<ltrb 0 0 0 0>
      outlineSettings {
       bUseBrushTransparency=false
       roundingType=HalfHeightRadius
       width=0
       color {colorUseRule=UseColor_Specified, specifiedColor=<rgba 0 0 0 0>}
       cornerRadii=<0 0 0 0>
      }
      tintColor {colorUseRule=UseColor_Specified, specifiedColor=<rgba 1 1 1 1>}
      uVRegion {bIsValid=false, max=<0 0>, min=<0 0>}
     }
     selectedBackgroundColor {colorUseRule=UseColor_Foreground, specifiedColor=<rgba 1 0 1 1>}
     shadowColorAndOpacity=<rgba 0 0 0 1>
     shadowOffset=<0 0>
     strikeBrush {
      bIsDynamicallyLoaded=false
      drawAs=Image
      imageType=NoImage
      mirroring=NoMirror
      resourceName=None
      resourceObject=null
      tiling=NoTile
      imageSize=<32 32>
      margin=<ltrb 0 0 0 0>
      outlineSettings {
       bUseBrushTransparency=false
       roundingType=HalfHeightRadius
       width=0
       color {colorUseRule=UseColor_Specified, specifiedColor=<rgba 0 0 0 0>}
       cornerRadii=<0 0 0 0>
      }
      tintColor {colorUseRule=UseColor_Specified, specifiedColor=<rgba 1 1 1 1>}
      uVRegion {bIsValid=false, max=<0 0>, min=<0 0>}
     }
     underlineBrush {
      bIsDynamicallyLoaded=false
      drawAs=Image
      imageType=NoImage
      mirroring=NoMirror
      resourceName=None
      resourceObject=null
      tiling=NoTile
      imageSize=<32 32>
      margin=<ltrb 0 0 0 0>
      outlineSettings {
       bUseBrushTransparency=false
       roundingType=HalfHeightRadius
       width=0
       color {colorUseRule=UseColor_Specified, specifiedColor=<rgba 0 0 0 0>}
       cornerRadii=<0 0 0 0>
      }
      tintColor {colorUseRule=UseColor_Specified, specifiedColor=<rgba 1 1 1 1>}
      uVRegion {bIsValid=false, max=<0 0>, min=<0 0>}
     }
    }
    desiredKeybindWidgetSize=<16 16>
    renderTransformPivot=<0.5 0>
   }
   slot {class=OverlaySlot, properties {horizontalAlignment=HAlign_Fill, verticalAlignment=VAlign_Center, padding=<ltrb 10 8 0 4>}}
  }
  IconSwitcher {
   widgetClassPath=&CommonVisibilitySwitcher
   bIsVariable=true
   parent=MultiStepCollapsed
   widget=IconSwitcher
   slot {class=OverlaySlot, properties {horizontalAlignment=HAlign_Right, verticalAlignment=VAlign_Bottom, padding=<ltrb 0 0 10 0>}}
  }
  FVA_BigIcon {
   widgetClassPath=&FortVisualAttachment
   parent=IconSwitcher
   widget=FVA_BigIcon
   properties {contentAnchor=<1 1>}
   slot {class=CommonVisibilitySwitcherSlot}
  }
  BigIcon {
   widgetClassPath=&FortLazyImage
   bIsVariable=true
   parent=FVA_BigIcon
   widget=BigIcon
   properties {
    materialTextureParamName=NPC-Portrait
    brush {
     bIsDynamicallyLoaded=false
     drawAs=Image
     imageType=FullColor
     mirroring=NoMirror
     resourceName=None
     resourceObject=/Game/Athena/HUD/Discoverability/materials/CategoryMI/MI_UI_Squad
     tiling=NoTile
     imageSize=<64 64>
     margin=<ltrb 0.25 0.25 0.25 0.25>
     outlineSettings {
      bUseBrushTransparency=false
      roundingType=HalfHeightRadius
      width=0
      color {colorUseRule=UseColor_Specified, specifiedColor=<rgba 0 0 0 0>}
      cornerRadii=<0 0 0 0>
     }
     tintColor {colorUseRule=UseColor_Specified, specifiedColor=<rgba 1 1 1 1>}
     uVRegion {bIsValid=false, max=<0 0>, min=<0 0>}
    }
    loadingBackgroundBrush {
     bIsDynamicallyLoaded=false
     drawAs=Image
     imageType=FullColor
     mirroring=NoMirror
     resourceName=None
     resourceObject=null
     tiling=NoTile
     imageSize=<128 128>
     margin=<ltrb 0 0 0 0>
     outlineSettings {
      bUseBrushTransparency=false
      roundingType=HalfHeightRadius
      width=0
      color {colorUseRule=UseColor_Specified, specifiedColor=<rgba 0 0 0 0>}
      cornerRadii=<0 0 0 0>
     }
     tintColor {colorUseRule=UseColor_Specified, specifiedColor=<rgba 1 1 1 1>}
     uVRegion {bIsValid=false, max=<0 0>, min=<0 0>}
    }
   }
   slot {class=SizeBoxSlot, properties {horizontalAlignment=HAlign_Center, verticalAlignment=VAlign_Center}}
  }
  FortVisualAttachment_168 {
   widgetClassPath=&FortVisualAttachment
   parent=IconSwitcher
   widget=FortVisualAttachment_168
   properties {contentAnchor=<1 1>}
   slot {class=CommonVisibilitySwitcherSlot}
  }
  BigIconFullBody {
   widgetClassPath=&FortLazyImage
   bIsVariable=true
   parent=FortVisualAttachment_168
   widget=BigIconFullBody
   properties {
    materialTextureParamName=NPC-Portrait
    brush {
     bIsDynamicallyLoaded=false
     drawAs=Image
     imageType=FullColor
     mirroring=NoMirror
     resourceName=None
     resourceObject=/Game/Athena/HUD/Discoverability/materials/CategoryMI/MI_UI_CharacterGradient
     tiling=NoTile
     imageSize=<128 128>
     margin=<ltrb 0.25 0.25 0.25 0.25>
     outlineSettings {
      bUseBrushTransparency=false
      roundingType=HalfHeightRadius
      width=0
      color {colorUseRule=UseColor_Specified, specifiedColor=<rgba 0 0 0 0>}
      cornerRadii=<0 0 0 0>
     }
     tintColor {colorUseRule=UseColor_Specified, specifiedColor=<rgba 1 1 1 1>}
     uVRegion {bIsValid=false, max=<0 0>, min=<0 0>}
    }
    loadingBackgroundBrush {
     bIsDynamicallyLoaded=false
     drawAs=Image
     imageType=FullColor
     mirroring=NoMirror
     resourceName=None
     resourceObject=null
     tiling=NoTile
     imageSize=<128 128>
     margin=<ltrb 0 0 0 0>
     outlineSettings {
      bUseBrushTransparency=false
      roundingType=HalfHeightRadius
      width=0
      color {colorUseRule=UseColor_Specified, specifiedColor=<rgba 0 0 0 0>}
      cornerRadii=<0 0 0 0>
     }
     tintColor {colorUseRule=UseColor_Specified, specifiedColor=<rgba 1 1 1 1>}
     uVRegion {bIsValid=false, max=<0 0>, min=<0 0>}
    }
   }
   slot {class=SizeBoxSlot, properties {horizontalAlignment=HAlign_Right, verticalAlignment=VAlign_Bottom}}
  }
  Overlay_Description {
   widgetClassPath=&Overlay
   bIsVariable=true
   parent=Vertical_MainObjective
   widget=Overlay_Description
   slot {class=VerticalBoxSlot}
  }
  HorizontalBox_2 {
   widgetClassPath=&HorizontalBox
   parent=Overlay_Description
   widget=HorizontalBox_2
   slot {class=OverlaySlot, properties {horizontalAlignment=HAlign_Fill, verticalAlignment=VAlign_Fill, padding=<ltrb 10 0 10 0>}}
  }
  SizeBox_Icon {
   widgetClassPath=&SizeBox
   parent=HorizontalBox_2
   widget=SizeBox_Icon
   properties {heightOverride=70, widthOverride=70}
   slot {class=HorizontalBoxSlot, properties {verticalAlignment=VAlign_Bottom}}
  }
  Image_Icon {
   widgetClassPath=&FortLazyImage
   bIsVariable=true
   parent=SizeBox_Icon
   widget=Image_Icon
   properties {
    materialTextureParamName=NPC-Portrait
    brush {
     bIsDynamicallyLoaded=false
     drawAs=Image
     imageType=FullColor
     mirroring=NoMirror
     resourceName=None
     resourceObject=null
     tiling=NoTile
     imageSize=<70 70>
     margin=<ltrb 0.25 0.25 0.25 0.25>
     outlineSettings {
      bUseBrushTransparency=false
      roundingType=HalfHeightRadius
      width=0
      color {colorUseRule=UseColor_Specified, specifiedColor=<rgba 0 0 0 0>}
      cornerRadii=<0 0 0 0>
     }
     tintColor {colorUseRule=UseColor_Foreground, specifiedColor=<rgba 1 1 1 1>}
     uVRegion {bIsValid=false, max=<0 0>, min=<0 0>}
    }
    loadingBackgroundBrush {
     bIsDynamicallyLoaded=false
     drawAs=Image
     imageType=FullColor
     mirroring=NoMirror
     resourceName=None
     resourceObject=null
     tiling=NoTile
     imageSize=<128 128>
     margin=<ltrb 0 0 0 0>
     outlineSettings {
      bUseBrushTransparency=false
      roundingType=HalfHeightRadius
      width=0
      color {colorUseRule=UseColor_Specified, specifiedColor=<rgba 0 0 0 0>}
      cornerRadii=<0 0 0 0>
     }
     tintColor {colorUseRule=UseColor_Specified, specifiedColor=<rgba 1 1 1 1>}
     uVRegion {bIsValid=false, max=<0 0>, min=<0 0>}
    }
   }
   slot {class=SizeBoxSlot, properties {padding=<ltrb 0 0 10 0>}}
  }
  SizeBox_Objective {
   widgetClassPath=&SizeBox
   parent=HorizontalBox_2
   widget=SizeBox_Objective
   properties {bOverride_MinDesiredHeight=true, minDesiredHeight=62}
   slot {class=HorizontalBoxSlot, properties {padding=<ltrb 0 0 5 0>, size {sizeRule=Fill, value=1}}}
  }
  Text_ObjectiveTitle {
   widgetClassPath=&FortRichTextBlockKeybind
   bInherited=true
   bIsVariable=true
   parent=SizeBox_Objective
   widget=Text_ObjectiveTitle
   properties {
    autoWrapText=true
    keybindWidgetClass=&KeybindWidgetDark_Small_C
    mobileTextBlockScale=1.2
    text="Title - Long Objective Name, Testing with more text <Strike>Uncommon</> "
    textStyleSet=&UrgentQuestRichTextStyles
    wrapTextAt=450
    decoratorClasses [
     &UrgentQuestRichTextDecorators_C
     /Game/Athena/HUD/QuestIcons/BP_QuestTile_IconDecorator.BP_QuestTile_IconDecorator_C
    ]
    defaultTextStyleOverride {
     overflowPolicy=Clip
     transformPolicy=None
     colorAndOpacity {colorUseRule=UseColor_Specified, specifiedColor=<rgba 1 1 1 1>}
     font {
      bForceMonospaced=false
      bMaterialIsStencil=false
      fontMaterial=null
      fontObject=&HeadingNow
      letterSpacing=0
      monospacedWidth=1
      size=12
      skewAmount=0
      typefaceFontName=75 Medium
      outlineSettings {
       bApplyOutlineToDropShadows=false
       bMiteredCorners=false
       bSeparateFillAlpha=false
       outlineMaterial=null
       outlineSize=0
       outlineColor=<rgba 0 0 0 0.5>
      }
     }
     highlightColor {colorUseRule=UseColor_Specified, specifiedColor=<rgba 0 0 0 1>}
     highlightShape {
      bIsDynamicallyLoaded=false
      drawAs=Image
      imageType=NoImage
      mirroring=NoMirror
      resourceName=None
      resourceObject=null
      tiling=NoTile
      imageSize=<32 32>
      margin=<ltrb 0 0 0 0>
      outlineSettings {
       bUseBrushTransparency=false
       roundingType=HalfHeightRadius
       width=0
       color {colorUseRule=UseColor_Specified, specifiedColor=<rgba 0 0 0 0>}
       cornerRadii=<0 0 0 0>
      }
      tintColor {colorUseRule=UseColor_Specified, specifiedColor=<rgba 1 1 1 1>}
      uVRegion {bIsValid=false, max=<0 0>, min=<0 0>}
     }
     selectedBackgroundColor {colorUseRule=UseColor_Foreground, specifiedColor=<rgba 1 0 1 1>}
     shadowColorAndOpacity=<rgba 0 0 0 1>
     shadowOffset=<0 0>
     strikeBrush {
      bIsDynamicallyLoaded=false
      drawAs=Image
      imageType=NoImage
      mirroring=NoMirror
      resourceName=None
      resourceObject=null
      tiling=NoTile
      imageSize=<32 32>
      margin=<ltrb 0 0 0 0>
      outlineSettings {
       bUseBrushTransparency=false
       roundingType=HalfHeightRadius
       width=0
       color {colorUseRule=UseColor_Specified, specifiedColor=<rgba 0 0 0 0>}
       cornerRadii=<0 0 0 0>
      }
      tintColor {colorUseRule=UseColor_Specified, specifiedColor=<rgba 1 1 1 1>}
      uVRegion {bIsValid=false, max=<0 0>, min=<0 0>}
     }
     underlineBrush {
      bIsDynamicallyLoaded=false
      drawAs=Image
      imageType=NoImage
      mirroring=NoMirror
      resourceName=None
      resourceObject=null
      tiling=NoTile
      imageSize=<32 32>
      margin=<ltrb 0 0 0 0>
      outlineSettings {
       bUseBrushTransparency=false
       roundingType=HalfHeightRadius
       width=0
       color {colorUseRule=UseColor_Specified, specifiedColor=<rgba 0 0 0 0>}
       cornerRadii=<0 0 0 0>
      }
      tintColor {colorUseRule=UseColor_Specified, specifiedColor=<rgba 1 1 1 1>}
      uVRegion {bIsValid=false, max=<0 0>, min=<0 0>}
     }
    }
    desiredKeybindWidgetSize=<16 16>
    renderTransformPivot=<0.5 0>
   }
   slot {class=SizeBoxSlot, properties {verticalAlignment=VAlign_Center}}
  }
  Overlay_RewardIcon {
   widgetClassPath=&Overlay
   bIsVariable=true
   parent=HorizontalBox_2
   widget=Overlay_RewardIcon
   properties {renderOpacity=0, visibility=Collapsed, renderTransformPivot=<0 0>}
   slot {class=HorizontalBoxSlot, properties {horizontalAlignment=HAlign_Center}}
  }
  Image_fx_Sparkle {
   widgetClassPath=&Image
   parent=Overlay_RewardIcon
   widget=Image_fx_Sparkle
   properties {
    renderOpacity=0
    brush {
     bIsDynamicallyLoaded=false
     drawAs=Image
     imageType=FullColor
     mirroring=NoMirror
     resourceName=None
     resourceObject=/Game/Athena/HUD/Materials/Victory/M_Flare_Cross_Animated
     tiling=NoTile
     imageSize=<40 40>
     margin=<ltrb 0 0 0 0>
     outlineSettings {
      bUseBrushTransparency=false
      roundingType=HalfHeightRadius
      width=0
      color {colorUseRule=UseColor_Specified, specifiedColor=<rgba 0 0 0 0>}
      cornerRadii=<0 0 0 0>
     }
     tintColor {colorUseRule=UseColor_Specified, specifiedColor=<rgba 1 1 1 1>}
     uVRegion {bIsValid=false, max=<0 0>, min=<0 0>}
    }
   }
   slot {class=OverlaySlot, properties {horizontalAlignment=HAlign_Right, verticalAlignment=VAlign_Center}}
  }
  Image_fx_Sparkle02 {
   widgetClassPath=&Image
   parent=Overlay_RewardIcon
   widget=Image_fx_Sparkle02
   properties {
    renderOpacity=0
    brush {
     bIsDynamicallyLoaded=false
     drawAs=Image
     imageType=FullColor
     mirroring=NoMirror
     resourceName=None
     resourceObject=&M_Flare_Cross_Animated
     tiling=NoTile
     imageSize=<40 40>
     margin=<ltrb 0 0 0 0>
     outlineSettings {
      bUseBrushTransparency=false
      roundingType=HalfHeightRadius
      width=0
      color {colorUseRule=UseColor_Specified, specifiedColor=<rgba 0 0 0 0>}
      cornerRadii=<0 0 0 0>
     }
     tintColor {colorUseRule=UseColor_Specified, specifiedColor=<rgba 1 1 1 1>}
     uVRegion {bIsValid=false, max=<0 0>, min=<0 0>}
    }
   }
   slot {class=OverlaySlot, properties {horizontalAlignment=HAlign_Right, verticalAlignment=VAlign_Center}}
  }
  WBP_Athena_Objectives_Icon {
   widgetClassPath=&/WBP_Athena_Objectives_Icon
   bIsVariable=true
   parent=Overlay_RewardIcon
   widget=WBP_Athena_Objectives_Icon
   properties {renderOpacity=0}
   slot {class=OverlaySlot, properties {horizontalAlignment=HAlign_Center, verticalAlignment=VAlign_Bottom}}
  }
  OverlayStageTransition {
   widgetClassPath=&Overlay
   parent=Overlay_Description
   widget=OverlayStageTransition
   properties {renderTransform {angle=0, scale=<0 1>, shear=<0 0>, translation=<0 0>}, renderTransformPivot=<0 0>}
   slot {class=OverlaySlot, properties {horizontalAlignment=HAlign_Fill, verticalAlignment=VAlign_Fill}}
  }
  Image_Main_Transition {
   widgetClassPath=&Image
   bIsVariable=true
   parent=OverlayStageTransition
   widget=Image_Main_Transition
   properties {visibility=SelfHitTestInvisible, colorAndOpacity=<rgba 1 1 1 0>}
   slot {class=OverlaySlot, properties {horizontalAlignment=HAlign_Fill, verticalAlignment=VAlign_Fill}}
  }
  Image_Material_Transition {
   widgetClassPath=&Image
   bIsVariable=true
   parent=OverlayStageTransition
   widget=Image_Material_Transition
   properties {
    brush {
     bIsDynamicallyLoaded=false
     drawAs=Image
     imageType=FullColor
     mirroring=NoMirror
     resourceName=None
     resourceObject=/Game/Athena/HUD/MiniMap/Art/M_UI_PinnedQuest_Rift
     tiling=NoTile
     imageSize=<400 40>
     margin=<ltrb 0 0 0 0>
     outlineSettings {
      bUseBrushTransparency=false
      roundingType=HalfHeightRadius
      width=0
      color {colorUseRule=UseColor_Specified, specifiedColor=<rgba 0 0 0 0>}
      cornerRadii=<0 0 0 0>
     }
     tintColor {colorUseRule=UseColor_Specified, specifiedColor=<rgba 1 1 1 1>}
     uVRegion {bIsValid=false, max=<0 0>, min=<0 0>}
    }
    renderTransform {angle=0, scale=<1.3 1.5>, shear=<0 0>, translation=<60 0>}
   }
   slot {class=OverlaySlot, properties {verticalAlignment=VAlign_Center}}
  }
  Text_ObjectiveTransition {
   widgetClassPath=&FortRichTextBlockKeybind
   bIsVariable=true
   parent=OverlayStageTransition
   widget=Text_ObjectiveTransition
   properties {
    autoWrapText=true
    bOverrideDefaultStyle=true
    keybindWidgetClass=&KeybindWidgetDark_Small_C
    mobileTextBlockScale=1.2
    text="<Transition>Stage 1 Complete</>"
    textStyleSet=&UrgentQuestRichTextStyles
    decoratorClasses [&UrgentQuestRichTextDecorators_C]
    defaultTextStyleOverride {
     overflowPolicy=Clip
     transformPolicy=None
     colorAndOpacity {colorUseRule=UseColor_Specified, specifiedColor=<rgba 1 1 1 1>}
     font {
      bForceMonospaced=false
      bMaterialIsStencil=false
      fontMaterial=null
      fontObject=&HeadingNow
      letterSpacing=0
      monospacedWidth=1
      size=12
      skewAmount=0
      typefaceFontName=76 Bold
      outlineSettings {
       bApplyOutlineToDropShadows=false
       bMiteredCorners=false
       bSeparateFillAlpha=false
       outlineMaterial=null
       outlineSize=0
       outlineColor=<rgba 0 0 0 0.5>
      }
     }
     highlightColor {colorUseRule=UseColor_Specified, specifiedColor=<rgba 0 0 0 1>}
     highlightShape {
      bIsDynamicallyLoaded=false
      drawAs=Image
      imageType=NoImage
      mirroring=NoMirror
      resourceName=None
      resourceObject=null
      tiling=NoTile
      imageSize=<32 32>
      margin=<ltrb 0 0 0 0>
      outlineSettings {
       bUseBrushTransparency=false
       roundingType=HalfHeightRadius
       width=0
       color {colorUseRule=UseColor_Specified, specifiedColor=<rgba 0 0 0 0>}
       cornerRadii=<0 0 0 0>
      }
      tintColor {colorUseRule=UseColor_Specified, specifiedColor=<rgba 1 1 1 1>}
      uVRegion {bIsValid=false, max=<0 0>, min=<0 0>}
     }
     selectedBackgroundColor {colorUseRule=UseColor_Foreground, specifiedColor=<rgba 1 0 1 1>}
     shadowColorAndOpacity=<rgba 0 0 0 1>
     shadowOffset=<0 0>
     strikeBrush {
      bIsDynamicallyLoaded=false
      drawAs=Image
      imageType=NoImage
      mirroring=NoMirror
      resourceName=None
      resourceObject=null
      tiling=NoTile
      imageSize=<32 32>
      margin=<ltrb 0 0 0 0>
      outlineSettings {
       bUseBrushTransparency=false
       roundingType=HalfHeightRadius
       width=0
       color {colorUseRule=UseColor_Specified, specifiedColor=<rgba 0 0 0 0>}
       cornerRadii=<0 0 0 0>
      }
      tintColor {colorUseRule=UseColor_Specified, specifiedColor=<rgba 1 1 1 1>}
      uVRegion {bIsValid=false, max=<0 0>, min=<0 0>}
     }
     underlineBrush {
      bIsDynamicallyLoaded=false
      drawAs=Image
      imageType=NoImage
      mirroring=NoMirror
      resourceName=None
      resourceObject=null
      tiling=NoTile
      imageSize=<32 32>
      margin=<ltrb 0 0 0 0>
      outlineSettings {
       bUseBrushTransparency=false
       roundingType=HalfHeightRadius
       width=0
       color {colorUseRule=UseColor_Specified, specifiedColor=<rgba 0 0 0 0>}
       cornerRadii=<0 0 0 0>
      }
      tintColor {colorUseRule=UseColor_Specified, specifiedColor=<rgba 1 1 1 1>}
      uVRegion {bIsValid=false, max=<0 0>, min=<0 0>}
     }
    }
    desiredKeybindWidgetSize=<16 16>
    renderTransformPivot=<0.5 0>
   }
   slot {class=OverlaySlot, properties {horizontalAlignment=HAlign_Fill, verticalAlignment=VAlign_Center, padding=<ltrb 10 0 0 0>}}
  }
  Overlay_Pro {widgetClassPath=&Overlay, parent=Vertical_MainObjective, widget=Overlay_Pro, slot {class=VerticalBoxSlot}}
  Progress_BG {
   widgetClassPath=&Image
   bIsVariable=true
   parent=Overlay_Pro
   widget=Progress_BG
   properties {
    brush {
     bIsDynamicallyLoaded=false
     drawAs=Image
     imageType=FullColor
     mirroring=NoMirror
     resourceName=None
     resourceObject=/Game/Athena/HUD/Discoverability/materials/MI_UI_ObjectivesProgressBG
     tiling=NoTile
     imageSize=<32 32>
     margin=<ltrb 0 0 0 0>
     outlineSettings {
      bUseBrushTransparency=false
      roundingType=HalfHeightRadius
      width=0
      color {colorUseRule=UseColor_Specified, specifiedColor=<rgba 0 0 0 0>}
      cornerRadii=<0 0 0 0>
     }
     tintColor {colorUseRule=UseColor_Specified, specifiedColor=<rgba 1 1 1 1>}
     uVRegion {bIsValid=false, max=<0 0>, min=<0 0>}
    }
    colorAndOpacity=<rgba 0.0024 0.063 0.2918 1>
   }
   slot {class=OverlaySlot, properties {horizontalAlignment=HAlign_Fill, verticalAlignment=VAlign_Fill}}
  }
  HorizontalBox_0 {
   widgetClassPath=&HorizontalBox
   parent=Overlay_Pro
   widget=HorizontalBox_0
   slot {class=OverlaySlot, properties {horizontalAlignment=HAlign_Fill, verticalAlignment=VAlign_Center, padding=<ltrb 10 0 10 0>}}
  }
  Overlay_Timer {
   widgetClassPath=&Overlay
   bInherited=true
   bIsVariable=true
   parent=HorizontalBox_0
   widget=Overlay_Timer
   slot {class=HorizontalBoxSlot, properties {horizontalAlignment=HAlign_Left, padding=<ltrb 0 0 10 0>}}
  }
  HorizontalBox_1 {
   widgetClassPath=&HorizontalBox
   parent=Overlay_Timer
   widget=HorizontalBox_1
   slot {class=OverlaySlot, properties {horizontalAlignment=HAlign_Fill, verticalAlignment=VAlign_Fill}}
  }
  Img_TimerIcon_1 {
   widgetClassPath=&Image
   parent=HorizontalBox_1
   widget=Img_TimerIcon_1
   properties {brush {
     bIsDynamicallyLoaded=false
     drawAs=Image
     imageType=FullColor
     mirroring=NoMirror
     resourceName=None
     resourceObject=/FortUILibrary/Textures/MaterialInstances/MI_UIKit_Icon_UpdatePending
     tiling=NoTile
     imageSize=<22 22>
     margin=<ltrb 0 0 0 0>
     outlineSettings {
      bUseBrushTransparency=false
      roundingType=HalfHeightRadius
      width=0
      color {colorUseRule=UseColor_Specified, specifiedColor=<rgba 0 0 0 0>}
      cornerRadii=<0 0 0 0>
     }
     tintColor {colorUseRule=UseColor_Specified, specifiedColor=<rgba 1 1 1 1>}
     uVRegion {bIsValid=false, max=<0 0>, min=<0 0>}
    }}
   slot {class=HorizontalBoxSlot, properties {horizontalAlignment=HAlign_Left, verticalAlignment=VAlign_Center, padding=<ltrb 0 0 3 0>}}
  }
  Text_Timer {
   widgetClassPath=/Script/FortniteUI.FortEventDateTimeTextBlock
   bInherited=true
   bIsVariable=true
   parent=HorizontalBox_1
   widget=Text_Timer
   properties {
    mobileFontSizeMultiplier=1.1
    text=0:00
    font {
     bForceMonospaced=true
     bMaterialIsStencil=false
     fontMaterial=null
     fontObject=&HeadingNow
     letterSpacing=100
     monospacedWidth=0.85
     size=13.5
     skewAmount=0
     typefaceFontName=75 Medium
     outlineSettings {
      bApplyOutlineToDropShadows=false
      bMiteredCorners=false
      bSeparateFillAlpha=false
      outlineMaterial=null
      outlineSize=0
      outlineColor=<rgba 0 0 0 0.5>
     }
    }
    shadowOffset=<0 0>
   }
   slot {class=HorizontalBoxSlot, properties {horizontalAlignment=HAlign_Left, verticalAlignment=VAlign_Center}}
  }
  Overlay_ProgressBar {
   widgetClassPath=&Overlay
   bInherited=true
   bIsVariable=true
   parent=HorizontalBox_0
   widget=Overlay_ProgressBar
   properties {visibility=HitTestInvisible}
   slot {class=HorizontalBoxSlot, properties {verticalAlignment=VAlign_Center, size {sizeRule=Fill, value=3}}}
  }
  Image_fx_TimeOut {
   widgetClassPath=&Image
   parent=Overlay_ProgressBar
   widget=Image_fx_TimeOut
   properties {
    renderOpacity=0
    visibility=Collapsed
    brush {
     bIsDynamicallyLoaded=false
     drawAs=Image
     imageType=NoImage
     mirroring=NoMirror
     resourceName=None
     resourceObject=/Game/Athena/HUD/LTM/Wax/Materials/M_UI_TextGlow_Pulse_Add
     tiling=NoTile
     imageSize=<110.8267 8.5754>
     margin=<ltrb 0 0 0 0>
     outlineSettings {
      bUseBrushTransparency=false
      roundingType=HalfHeightRadius
      width=0
      color {colorUseRule=UseColor_Specified, specifiedColor=<rgba 0 0 0 0>}
      cornerRadii=<0 0 0 0>
     }
     tintColor {colorUseRule=UseColor_Specified, specifiedColor=<rgba 1 1 1 0.5>}
     uVRegion {bIsValid=false, max=<0 0>, min=<0 0>}
    }
    colorAndOpacity=<rgba 0.6042 0 0.005 1>
   }
   slot {class=OverlaySlot, properties {horizontalAlignment=HAlign_Fill, verticalAlignment=VAlign_Center}}
  }
  SizeBox_0 {
   widgetClassPath=&SizeBox
   parent=Overlay_ProgressBar
   widget=SizeBox_0
   properties {bOverride_HeightOverride=true, heightOverride=10}
   slot {class=OverlaySlot, properties {horizontalAlignment=HAlign_Fill, verticalAlignment=VAlign_Center, padding=<ltrb 0 0 10 0>}}
  }
  ProgressBar_ObjectiveProgress {
   widgetClassPath=/Script/UMG.ProgressBar
   bInherited=true
   bIsVariable=true
   parent=SizeBox_0
   widget=ProgressBar_ObjectiveProgress
   properties {
    barFillStyle=Scale
    percent=0.2639
    borderPadding=<0.25 0.5>
    widgetStyle {
     enableFillAnimation=false
     backgroundImage {
      bIsDynamicallyLoaded=false
      drawAs=Image
      imageType=FullColor
      mirroring=NoMirror
      resourceName=None
      resourceObject=/Game/Athena/HUD/Discoverability/materials/MI_UI_ProgressFill
      tiling=NoTile
      imageSize=<40 10>
      margin=<ltrb 0 0 0 0>
      outlineSettings {
       bUseBrushTransparency=false
       roundingType=HalfHeightRadius
       width=0
       color {colorUseRule=UseColor_Specified, specifiedColor=<rgba 0 0 0 0>}
       cornerRadii=<0 0 0 0>
      }
      tintColor {colorUseRule=UseColor_Specified, specifiedColor=<rgba 1 1 1 0.35>}
      uVRegion {bIsValid=false, max=<0 0>, min=<0 0>}
     }
     fillImage {
      bIsDynamicallyLoaded=false
      drawAs=Image
      imageType=FullColor
      mirroring=NoMirror
      resourceName=None
      resourceObject=&MI_UI_ProgressFill
      tiling=NoTile
      imageSize=<32 10>
      margin=<ltrb 0 0 0 0>
      outlineSettings {
       bUseBrushTransparency=false
       roundingType=HalfHeightRadius
       width=0
       color {colorUseRule=UseColor_Specified, specifiedColor=<rgba 0 0 0 0>}
       cornerRadii=<0 0 0 0>
      }
      tintColor {colorUseRule=UseColor_Specified, specifiedColor=<rgba 0.1255 0.9373 0.9804 1>}
      uVRegion {bIsValid=false, max=<0 0>, min=<0 0>}
     }
     marqueeImage {
      bIsDynamicallyLoaded=false
      drawAs=Image
      imageType=NoImage
      mirroring=NoMirror
      resourceName=None
      resourceObject=null
      tiling=NoTile
      imageSize=<32 10>
      margin=<ltrb 0 0 0 0>
      outlineSettings {
       bUseBrushTransparency=false
       roundingType=HalfHeightRadius
       width=0
       color {colorUseRule=UseColor_Specified, specifiedColor=<rgba 0 0 0 0>}
       cornerRadii=<0 0 0 0>
      }
      tintColor {colorUseRule=UseColor_Specified, specifiedColor=<rgba 1 1 1 1>}
      uVRegion {bIsValid=false, max=<0 0>, min=<0 0>}
     }
    }
   }
   slot {class=SizeBoxSlot}
  }
  Image_BountyThreatGlow {
   widgetClassPath=&Image
   bIsVariable=true
   parent=Overlay_ProgressBar
   widget=Image_BountyThreatGlow
   properties {
    visibility=Collapsed
    brush {
     bIsDynamicallyLoaded=false
     drawAs=Image
     imageType=NoImage
     mirroring=NoMirror
     resourceName=None
     resourceObject=&M_UI_TextGlow_Pulse_Add
     tiling=NoTile
     imageSize=<155 15>
     margin=<ltrb 0 0 0 0>
     outlineSettings {
      bUseBrushTransparency=false
      roundingType=HalfHeightRadius
      width=0
      color {colorUseRule=UseColor_Specified, specifiedColor=<rgba 0 0 0 0>}
      cornerRadii=<0 0 0 0>
     }
     tintColor {colorUseRule=UseColor_Specified, specifiedColor=<rgba 1 1 1 0.5>}
     uVRegion {bIsValid=false, max=<0 0>, min=<0 0>}
    }
    colorAndOpacity=<rgba 0.6042 0 0.005 1>
    renderTransform {angle=0, scale=<1 1>, shear=<0 0>, translation=<53 1>}
   }
   slot {class=OverlaySlot, properties {horizontalAlignment=HAlign_Right, verticalAlignment=VAlign_Fill, padding=<ltrb -20 -20 -20 -20>}}
  }
  Overlay_Progress {
   widgetClassPath=&Overlay
   parent=HorizontalBox_0
   widget=Overlay_Progress
   slot {class=HorizontalBoxSlot, properties {horizontalAlignment=HAlign_Right, verticalAlignment=VAlign_Center}}
  }
  Image_fx_Reward_Glow {
   widgetClassPath=&Image
   bIsVariable=true
   parent=Overlay_Progress
   widget=Image_fx_Reward_Glow
   properties {
    renderOpacity=0
    visibility=HitTestInvisible
    brush {
     bIsDynamicallyLoaded=false
     drawAs=Image
     imageType=NoImage
     mirroring=NoMirror
     resourceName=None
     resourceObject=&M_UI_TextGlow_Pulse_Add
     tiling=NoTile
     imageSize=<50 7>
     margin=<ltrb 0 0 0 0>
     outlineSettings {
      bUseBrushTransparency=false
      roundingType=HalfHeightRadius
      width=0
      color {colorUseRule=UseColor_Specified, specifiedColor=<rgba 0 0 0 0>}
      cornerRadii=<0 0 0 0>
     }
     tintColor {colorUseRule=UseColor_Specified, specifiedColor=<rgba 1 1 1 0.25>}
     uVRegion {bIsValid=false, max=<0 0>, min=<0 0>}
    }
    colorAndOpacity=<rgba 0.6042 0.6042 0.6042 0.4217>
    renderTransform {angle=0, scale=<1 1>, shear=<0 0>, translation=<5 1>}
   }
   slot {class=OverlaySlot, properties {horizontalAlignment=HAlign_Center, verticalAlignment=VAlign_Center, padding=<ltrb -20 -20 -20 -20>}}
  }
  WBP_Objectives_Progress {
   widgetClassPath=&/WBP_Athena_Objectives_Progress
   bIsVariable=true
   parent=Overlay_Progress
   widget=WBP_Objectives_Progress
   properties {include Checkmark in Succes=false}
   slot {class=OverlaySlot, properties {horizontalAlignment=HAlign_Center, verticalAlignment=VAlign_Center}}
  }
  BackPlate_StageProgress {
   widgetClassPath=&Border
   bIsVariable=true
   parent=Overlay_ObjectiveTitle
   widget=BackPlate_StageProgress
   properties {
    background {
     bIsDynamicallyLoaded=false
     drawAs=Image
     imageType=FullColor
     mirroring=NoMirror
     resourceName=None
     resourceObject=/FortUILibrary/Systems/Backplate/MaterialInstances/MI_UIKit_Backplate
     tiling=NoTile
     imageSize=<32 32>
     margin=<ltrb 0 0 0 0>
     outlineSettings {
      bUseBrushTransparency=false
      roundingType=HalfHeightRadius
      width=0
      color {colorUseRule=UseColor_Specified, specifiedColor=<rgba 0 0 0 0>}
      cornerRadii=<0 0 0 0>
     }
     tintColor {colorUseRule=UseColor_Specified, specifiedColor=<rgba 1 1 1 1>}
     uVRegion {bIsValid=false, max=<0 0>, min=<0 0>}
    }
    padding=<ltrb 4 1 4 1>
    renderTransform {angle=0, scale=<1 1>, shear=<0 0>, translation=<0 -12>}
   }
   slot {class=OverlaySlot, properties {padding=<ltrb 10 0 0 3>}}
  }
  Text_StageProgressValue {
   widgetClassPath=&CommonTextBlock
   bIsVariable=true
   parent=BackPlate_StageProgress
   widget=Text_StageProgressValue
   properties {
    mobileFontSizeMultiplier=1.2
    renderOpacity=0.8
    text=0 of 5
    textTransformPolicy=ToUpper
    colorAndOpacity {colorUseRule=UseColor_Specified, specifiedColor=<rgba 0 0 0 1>}
    font {
     bForceMonospaced=false
     bMaterialIsStencil=false
     fontMaterial=null
     fontObject=&HeadingNow
     letterSpacing=-99
     monospacedWidth=1
     size=10.5
     skewAmount=0
     typefaceFontName=76 Bold
     outlineSettings {
      bApplyOutlineToDropShadows=false
      bMiteredCorners=false
      bSeparateFillAlpha=false
      outlineMaterial=null
      outlineSize=0
      outlineColor=<rgba 0 0 0 0.502>
     }
    }
    shadowOffset=<0 0>
   }
   slot {class=BorderSlot, properties {padding=<ltrb 4 1 4 1>}}
  }
  CommonBorderEntries {
   widgetClassPath=&CommonBorder
   bIsVariable=true
   parent=VerticalBox_0
   widget=CommonBorderEntries
   properties {padding=<ltrb 0 0 10 0>}
   slot {class=VerticalBoxSlot, properties {padding=<ltrb 10 5 0 0>}}
  }
  DynamicEntryBox_OptionalEntries {
   widgetClassPath=/Script/FortniteUI.FortDynamicEntryBox
   bInherited=true
   bIsVariable=true
   parent=CommonBorderEntries
   widget=DynamicEntryBox_OptionalEntries
   properties {
    entryBoxType=Vertical
    entryWidgetClass=&/WBP_Athena_Objectives_OptionalEntry.WBP_Athena_Objectives_OptionalEntry_C
    numDesignerPreviewEntries=2
    entrySpacing=<0 4>
   }
   slot {class=BorderSlot, properties {padding=<ltrb 0 0 10 0>}}
  }
  Text_ObjState_AdditionalInfo {
   widgetClassPath=&CommonTextBlock
   bIsVariable=true
   parent=VerticalBox_0
   widget=Text_ObjState_AdditionalInfo
   properties {
    renderOpacity=0
    text=Additional info
    visibility=Collapsed
    font {
     bForceMonospaced=false
     bMaterialIsStencil=false
     fontMaterial=null
     fontObject=&BurbankSmall
     letterSpacing=0
     monospacedWidth=1
     size=13.5
     skewAmount=0
     typefaceFontName=Bold
     outlineSettings {
      bApplyOutlineToDropShadows=false
      bMiteredCorners=false
      bSeparateFillAlpha=false
      outlineMaterial=null
      outlineSize=0
      outlineColor=<rgba 0 0 0 1>
     }
    }
    shadowColorAndOpacity=<rgba 0 0 0 0.7>
    shadowOffset=<0.5426 1.6607>
   }
   slot {class=VerticalBoxSlot, properties {horizontalAlignment=HAlign_Left, verticalAlignment=VAlign_Top, padding=<ltrb 0 10 0 0>}}
  }
  Overlay_ProgressValue_Title {
   widgetClassPath=&Overlay
   bInherited=true
   bIsVariable=true
   parent=Overlay_Root
   widget=Overlay_ProgressValue_Title
   slot {class=OverlaySlot}
  }
 ]
}
