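# Athena HUD container widget that shows match objective entries above a pinned content list, which is visible only in Default or Arena playlists.
# Adding or completing an entry rebalances the pinned count to fill three shared slots, or drops it to zero when compact UI has any objective.
# Activation plays a fade-and-slide intro if entries exist.
# Finishing with none plays the reverse, then collapses the panel after a retriggerable delay and rebalances.
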WidgetBlueprint /Game/Athena/HUD/Discoverability/WBP_Athena_Objectives_EntriesContainer
graphs [
 EventGraph {nodes [
   n1 = PlaylistDataReadyAsync!(execute:Exec) -> (
    then:Exec
    OnReady:Exec
    GameState:Fort Game State Athena Object Reference
    Playlist:Fort Playlist Object Reference
    PlaylistContextTags:"Gameplay Tag Container Structure (by ref)"
   ) {is=title K2Node_AsyncAction}
   n2 = HasAnyTags!(
    TagContainer:"Gameplay Tag Container Structure (by ref)"=@n1.PlaylistContextTags
    OtherContainer:"Gameplay Tag Container Structure (by ref)"=MakeLiteralGameplayTagContainer!(Value:Gameplay Tag Container Structure=(GameplayTags=((TagName="Athena.Playlist.Default"),(TagName="Athena.Playlist.Arena")))) -> (ReturnValue:Gameplay Tag Container Structure)
    bExactMatch:bool=false
   ) -> (ReturnValue:bool)
   Set isBR(execute=@n1.OnReady, isBR:bool=@n2) -> (Output_Get:bool)
   > SetVisibility(
    self=Get PinnedContentList() -> (PinnedContentList:Pinned Content List Object Reference)
    InVisibility=Select(Option 0=Collapsed, Option 1=HitTestInvisible, Index=@n2)
   )
   n3 = OnEntryCompleted!() -> (OutputDelegate, then, ObjectiveEntry:WBP Athena Objectives Entry Object Reference) {is=title K2Node_CustomEvent}
   > PinnedListAdjust()
   > Unbind all Events from On Entry Complete!(execute:Exec, self:WBP Athena Objectives Entry Object Reference=@n3.ObjectiveEntry) -> (then:Exec) {is=title K2Node_ClearDelegate}
   > Unbind all Events from On Set Urgent Quest!(execute:Exec, self:WBP Athena Objectives Entry Object Reference=@n3.ObjectiveEntry) -> (then:Exec) {is=title K2Node_ClearDelegate}
   n4 = Event BP_OnEntryAdded!() -> (OutputDelegate, then, Entry:Athena Challenge Entry Object Reference, NewEntryCount:int=0) {is=title K2Node_Event}
   > DelayUntilNextTick!(execute) -> (then)
   > PinnedListAdjust()
   > n5 = Cast To WBP_Athena_Objectives_Entry!(execute, Object=@n4.Entry) -> (then, CastFailed, AsWBP Athena Objectives Entry:WBP Athena Objectives Entry Object Reference) {is=title K2Node_DynamicCast}
   > Bind Event to On Entry Complete(self:WBP Athena Objectives Entry Object Reference=@n5, Delegate=Create Event()) {is=title K2Node_AddDelegate}
   Event Construct()
   > SetVisibility(
    self=Get AnimPanel_SimpleAnimations() -> (AnimPanel_SimpleAnimations:Fort Simple Widget Animations Panel Object Reference)
    InVisibility=Collapsed
   )
   Event BP_WidgetIsFinished() {is=title K2Node_Event}
   > n6 = Branch(Condition=HasChallengesToDisplay!() -> (bHasChallenges:bool))
   FinishOutro(execute=@n6.else)
   > RetriggerableDelay(Duration=1)
   > n7 = Branch(Condition=HasChallengesToDisplay())
   SetVisibility(execute=@n7.else, self=Get AnimPanel_SimpleAnimations(), InVisibility=Collapsed)
   > DelayUntilNextTick()
   > PinnedListAdjust()
   Event BP_OnWidgetActivated() {is=title K2Node_Event}
   > StartIntro()
  ]}
 FinishOutro {nodes [
   FinishOutro() {is=title K2Node_FunctionEntry}
   > PlayAnimationByName!(
    execute
    self:Fort Simple Widget Animations Panel Object Reference=Get AnimPanel_SimpleAnimations()
    Name:"Name (by ref)"=Get IntroAnimReverse() -> (IntroAnimReverse:Name=None)
    Speed:float=1
    StartAtTime:float=0
   ) -> (then, ReturnValue:bool)
  ]}
 HasChallengesToDisplay {nodes [
   HasChallengesToDisplay() {is=title K2Node_FunctionEntry}
   n1 = Get DynamicEntryBox_ChallengeEntries() -> (DynamicEntryBox_ChallengeEntries:Dynamic Entry Box Object Reference)
   > Return Node!(
    execute
    bHasChallenges:bool="integer > integer"(A=GetNumEntries!(self:Dynamic Entry Box Base Object Reference=@n1) -> (ReturnValue:int=0), B=0) {is=title K2Node_CallFunction}
   )
  ]}
 PinnedListAdjust {nodes [
   PinnedListAdjust() {is=title K2Node_FunctionEntry}
   n1 = IsUsingCompactUI()
   n2 = GetNumEntries(self=Get DynamicEntryBox_ChallengeEntries()) -> (ReturnValue=0)
   > n3 = Branch(Condition="integer >= integer"(A=@n2, B=Select(Option 0:int=3, Option 1:int=1, Index=@n1) -> (ReturnValue:int)))
   > SetPinnedAmmount!(execute, self:Pinned Content List Object Reference=Get PinnedContentList(), AmmountEntries:int=0) -> (then)
   n4 = Branch(execute=@n3.else, Condition="Equal (Integer)"(A=@n2))
   > SetPinnedAmmount(self=Get PinnedContentList(), AmmountEntries=3)
   SetPinnedAmmount(execute=@n4.else, self=Get PinnedContentList(), AmmountEntries=Select(Option 0=1, Option 1=2, Index="Equal (Integer)"(A=@n2, B=1)))
  ]}
 StartIntro {nodes [
   StartIntro() {is=title K2Node_FunctionEntry}
   > Branch(Condition=HasChallengesToDisplay() {is=title K2Node_CallFunction})
   > PinnedListAdjust() {is=title K2Node_CallFunction}
   > SetVisibility(self=Get AnimPanel_SimpleAnimations(), InVisibility=SelfHitTestInvisible)
   > PlayAnimationByName(self=Get AnimPanel_SimpleAnimations(), Name=Get IntroAnim() -> (IntroAnim:Name=None), Speed=1, StartAtTime=0)
  ]}
]
variables [
 ExpirationTimer {type=TimerHandle}
 IntroAnim {type=string, default=IntroAnim}
 IntroAnimReverse {type=string, default=IntroAnimReverse}
 isBR {type=boolean}
 OriginalHeaderText {type=string, default=""}
]
widgets {
 info {
  inheritedWidgetCount=2
  parentClass=/Script/FortniteUI.AthenaSpecialEventChallengesWidget
  rootWidgetClass=&Overlay
  widgetCount=8
 }
 widgets [
  Overlay_Root {widgetClassPath=&Overlay, widget=Overlay_Root}
  SizeBox_HeightClippingOverride {
   widgetClassPath=&SizeBox
   bIsVariable=true
   parent=Overlay_Root
   widget=SizeBox_HeightClippingOverride
   properties {bOverride_WidthOverride=true, heightOverride=500, maxDesiredHeight=320, widthOverride=510}
   slot {class=OverlaySlot, properties {horizontalAlignment=HAlign_Fill, verticalAlignment=VAlign_Fill, padding=<ltrb 0 2 0 2>}}
  }
  VerticalBox_0 {
   widgetClassPath=&VerticalBox
   parent=SizeBox_HeightClippingOverride
   widget=VerticalBox_0
   slot {class=SizeBoxSlot, properties {verticalAlignment=VAlign_Top}}
  }
  Overlay_42 {widgetClassPath=&Overlay, parent=VerticalBox_0, widget=Overlay_42, slot {class=VerticalBoxSlot}}
  AnimPanel_SimpleAnimations {
   widgetClassPath=/Script/FortniteUI.FortSimpleWidgetAnimationsPanel
   bIsVariable=true
   parent=Overlay_42
   widget=AnimPanel_SimpleAnimations
   properties {animations [
     IntroAnim {
      bIsEnabled=true
      onAnimationsFinished=()
      targets [{
        alpha=&/IntroAnim_RenderOpacity
        bResetOnFinish=false
        bShouldAnimateAlpha=true
        bShouldAnimateScale=false
        bShouldAnimateTranslation=true
        scale=null
        targetName=VerticalBox_0
        translation=&/IntroAnim_Translation
        materialScalarAnimations []
       }]
     }
     IntroAnimReverse {
      bIsEnabled=true
      onAnimationsFinished=()
      targets [{
        alpha=&/IntroAnim_RenderOpacityReverse
        bResetOnFinish=false
        bShouldAnimateAlpha=true
        bShouldAnimateScale=false
        bShouldAnimateTranslation=true
        scale=null
        targetName=VerticalBox_0
        translation=&/IntroAnim_TranslationReverse
        materialScalarAnimations []
       }]
     }
    ]}
   slot {class=OverlaySlot}
  }
  DynamicEntryBox_ChallengeEntries {
   widgetClassPath=/Script/FortniteUI.FortDynamicEntryBox
   bInherited=true
   bIsVariable=true
   parent=AnimPanel_SimpleAnimations
   widget=DynamicEntryBox_ChallengeEntries
   properties {
    entryBoxType=Vertical
    entryHorizontalAlignment=HAlign_Left
    entryVerticalAlignment=VAlign_Top
    entryWidgetClass=/Game/Athena/HUD/Discoverability/MatchObjectivesWidgets/WBP_Athena_Objectives_Entry.WBP_Athena_Objectives_Entry_C
    entrySpacing=<0 10>
   }
   slot {class=FortSimpleWidgetAnimationsSlot}
  }
  PinnedContentList {
   widgetClassPath=/Game/Athena/HUD/PinnedContentList
   bIsVariable=true
   parent=VerticalBox_0
   widget=PinnedContentList
   properties {bForceSinglePinnedQuestWidget=false}
   slot {class=VerticalBoxSlot}
  }
  TutorialHighlight {widgetClassPath=&TutorialHighlight}
 ]
}
