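# Athena HUD discoverability widget showing the name of a visited location beneath an "Entered" or "Landmark" label.
# Setter functions fill the name, flip a switcher between named-location and landmark labels, and swap both to "New Location!" or "New Landmark!" based on discovery state.
# PreConstruct runs UpdateMobileVisuals, which checks for touch input and Korean culture to adjust the name's top padding.
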WidgetBlueprint /Game/Athena/HUD/Discoverability/WBP_PlayerVisited_Context
graphs [
 EventGraph {nodes [
   n1 = Event PreConstruct()
   > Parent: PreConstruct(IsDesignTime=@n1.IsDesignTime)
   > UpdateMobileVisuals()
  ]}
 SetIsDiscovered {nodes [
   n1 = SetIsDiscovered!() -> (then, IsDiscovered:bool=~) {is=title K2Node_FunctionEntry}
   > "SetText (Text)"(
    self=Get Text_POIState() -> (Text_POIState:Common Text Object Reference)
    InText=Select(Option 0:Text="LOCTEXT(\"Entered\")", Option 1:Text="LOCTEXT(\"New Location!\")", Index=@n1) -> (ReturnValue:Text)
   )
   > "SetText (Text)"(
    self=Get Text_LandmarkState() -> (Text_LandmarkState:Common Text Object Reference)
    InText=Select(Option 0="LOCTEXT(\"Landmark\")", Option 1="LOCTEXT(\"New Landmark!\")", Index=@n1)
   )
  ]}
 SetIsNamedLocation {nodes [
   n1 = SetIsNamedLocation!() -> (then, IsNamedLocation:bool=~) {is=title K2Node_FunctionEntry}
   > SetActiveWidgetIndex(
    self=Get Switcher_landmarkerPOI() -> (Switcher_landmarkerPOI:Common Widget Switcher Legacy Object Reference)
    Index=Select(Option 0:int=1, Option 1:int=0, Index=@n1) -> (ReturnValue:int)
   )
  ]}
 SetLocationName {nodes [
   n1 = SetLocationName!() -> (then, LocationName:Text) {is=title K2Node_FunctionEntry}
   > "SetText (Text)"(self=Get MainLocationText() -> (MainLocationText:Common Text Object Reference), InText=@n1)
  ]}
 UpdateMobileVisuals {nodes [
   UpdateMobileVisuals() {is=title K2Node_FunctionEntry}
   n1 = IsUsingInputType!(InputType:ECommonInputType Enum=Touch, PlayerController:Player Controller Object Reference=GetOwningPlayer()) -> (ReturnValue:bool)
   > Branch(Condition=@n1)
   n2 = "Equal Exactly (String)"(A=GetCurrentCulture!() -> (ReturnValue:String), B=ko)
   > SetPadding!(
    execute
    self:Vertical Box Slot Object Reference=SlotAsVerticalBoxSlot(Widget=Get MainLocationText())
    InPadding:Margin Structure=Select!(
     Option 0_Left:float=0
     Option 0_Top:float=-6
     Option 0_Right:float=0
     Option 0_Bottom:float=0
     Option 1_Left:float=0
     Option 1_Top:float=16
     Option 1_Right:float=0
     Option 1_Bottom:float=0
     Index=AND Boolean(A=@n1, B=@n2)
    ) -> (ReturnValue:Margin Structure)
   ) -> (then)
  ]}
]
variables []
widgets {
 info {inheritedWidgetCount=1, parentClass=/Script/FortniteUI.AthenaCommonPrioritizedWidget, rootWidgetClass=&VerticalBox, widgetCount=6}
 widgets [
  MainBox {widgetClassPath=&VerticalBox, widget=MainBox}
  Switcher_landmarkerPOI {
   widgetClassPath=&CommonWidgetSwitcherLegacy
   bIsVariable=true
   parent=MainBox
   widget=Switcher_landmarkerPOI
   slot {class=VerticalBoxSlot, properties {verticalAlignment=VAlign_Bottom}}
  }
  Text_POIState {
   widgetClassPath=&CommonTextBlock
   bIsVariable=true
   parent=Switcher_landmarkerPOI
   widget=Text_POIState
   properties {
    bAutoCollapseWithEmptyText=true
    mobileFontSizeMultiplier=0.9
    text=Entered
    textTransformPolicy=ToUpper
    font {
     bForceMonospaced=false
     bMaterialIsStencil=false
     fontMaterial=null
     fontObject=&BurbankSmall
     letterSpacing=0
     monospacedWidth=1
     size=20
     skewAmount=0.25
     typefaceFontName=Black
     outlineSettings {
      bApplyOutlineToDropShadows=false
      bMiteredCorners=false
      bSeparateFillAlpha=false
      outlineMaterial=&M-FontShadow
      outlineSize=2
      outlineColor=<rgba 0 0 0 1>
     }
    }
    shadowOffset=<0 0>
   }
   slot {class=WidgetSwitcherSlot}
  }
  Text_LandmarkState {
   widgetClassPath=&CommonTextBlock
   bIsVariable=true
   parent=Switcher_landmarkerPOI
   widget=Text_LandmarkState
   properties {
    bAutoCollapseWithEmptyText=true
    mobileFontSizeMultiplier=0.9
    text=Landmark
    textTransformPolicy=ToUpper
    font {
     bForceMonospaced=false
     bMaterialIsStencil=false
     fontMaterial=null
     fontObject=&BurbankSmall
     letterSpacing=0
     monospacedWidth=1
     size=22
     skewAmount=0.25
     typefaceFontName=Bold
     outlineSettings {
      bApplyOutlineToDropShadows=false
      bMiteredCorners=false
      bSeparateFillAlpha=false
      outlineMaterial=null
      outlineSize=0
      outlineColor=<rgba 0 0 0 1>
     }
    }
    shadowOffset=<0 0>
   }
   slot {class=WidgetSwitcherSlot, properties {padding=<ltrb 0 0 0 1>}}
  }
  MainLocationText {
   widgetClassPath=&CommonTextBlock
   bIsVariable=true
   parent=MainBox
   widget=MainLocationText
   properties {
    bAutoCollapseWithEmptyText=true
    mobileFontSizeMultiplier=0.8
    text=Tilted Town
    textTransformPolicy=ToUpper
    font {
     bForceMonospaced=false
     bMaterialIsStencil=false
     fontMaterial=null
     fontObject=&Burbank
     letterSpacing=0
     monospacedWidth=1
     size=40
     skewAmount=0.25
     typefaceFontName=BurbankRegular-Black
     outlineSettings {
      bApplyOutlineToDropShadows=false
      bMiteredCorners=false
      bSeparateFillAlpha=false
      outlineMaterial=&M-FontShadow
      outlineSize=2
      outlineColor=<rgba 0 0 0 1>
     }
    }
    shadowOffset=<0 0>
   }
   slot {class=VerticalBoxSlot, properties {verticalAlignment=VAlign_Bottom, padding=<ltrb 0 -6 0 0>, size {sizeRule=Fill, value=1}}}
  }
  TutorialHighlight {widgetClassPath=&TutorialHighlight}
 ]
}
