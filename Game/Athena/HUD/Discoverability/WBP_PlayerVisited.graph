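# Athena HUD toast widget that announces visited locations, playing discovery music for newly discovered ones.
# OnNewVisited fills the context widget, throttles repeat locations with a cooldown console variable, and sends an asset-specific or fallback POI/landmark sound to MusicEventSubsystem.
# Becoming active plays the BriefShow slide-and-fade, or pauses it via a debug console variable.
# A fallback timer or failed stomp clears active content and collapses the widget and switcher children.
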WidgetBlueprint /Game/Athena/HUD/Discoverability/WBP_PlayerVisited
animations [BriefShow {
  binding_count=2
  length_seconds=3
  object=&self:BriefShow
  sequence {bindings [
    VerticalBox_0 {
     id=#1
     possessed_class=VerticalBox
     tracks [{
       type=MovieScene2DTransformTrack
       sections [{
         type=MovieScene2DTransformSection
         channels [
          Translation.X {type=MovieSceneScriptingFloatChannel, keys [[0,-63.6246],[15,0],[40,0],[60,55]]}
          Translation.Y {type=MovieSceneScriptingFloatChannel}
          Angle {type=MovieSceneScriptingFloatChannel}
          Scale.X {type=MovieSceneScriptingFloatChannel}
          Scale.Y {type=MovieSceneScriptingFloatChannel}
          Shear.X {type=MovieSceneScriptingFloatChannel}
          Shear.Y {type=MovieSceneScriptingFloatChannel}
         ]
        }]
      }]
    }
    Overlay_Main {
     id=#2
     possessed_class=Overlay
     tracks [{
       type=MovieSceneFloatTrack
       display_name=Render Opacity
       sections [{type=MovieSceneFloatSection, channels [None {type=MovieSceneScriptingFloatChannel, keys [[0,0],[8.7773,1],[42,1],[60,0]]}]}]
      }]
    }
   ]}
 }]
graphs [
 EndOfShow {nodes [
   EndOfShow() {is=title K2Node_FunctionEntry}
   > SetVisibility(InVisibility=Collapsed)
   > SetExtensionSlotsInactive()
  ]}
 EventGraph {nodes [
   n1 = OnBriefShowAnimTimerFinished() {is=title K2Node_CustomEvent}
   Clear and Invalidate Timer by Handle(execute=[@n16.then, @n1.then, @n15.else], Handle=Get BriefShowFallbackTimerHandle() -> (BriefShowFallbackTimerHandle:Timer Handle Structure))
   > SetHasActiveContent!(execute, self:Fort Prioritized Widget Interface, bNewActive:bool=false) -> (then)
   > EndOfShow() {is=title K2Node_CallFunction}
   Event Construct()
   > Parent: Construct()
   > SetVisibility(InVisibility=Collapsed)
   > SetMobileVisuals()
   Event Destruct()
   > UnsubscribeMusicEvent()
   n2 = Event OnNewVisited!() -> (
    OutputDelegate
    then
    LocalizedLocationName:"Text (by ref)"
    DiscoveredSoundCue:"Sound Cue Soft Object Reference (by ref)"
    bIsNamedLocation:bool
    bIsDiscovered:bool
   ) {is=title K2Node_Event}
   > Set Discovered Sound Cue(Discovered Sound Cue:Sound Cue Soft Object Reference=@n2.DiscoveredSoundCue) -> (Output_Get:Sound Cue Soft Object Reference)
   > Set IsNamedLocation(IsNamedLocation:bool=@n2.bIsNamedLocation) -> (Output_Get:bool)
   > n3 = Set Is Discovered(Is Discovered:bool=@n2.bIsDiscovered) -> (Output_Get:bool)
   > SetIsDiscovered!(
    execute
    self:WBP Player Visited Context Object Reference=Get WBP_PlayerVisited_Context() -> (WBP_PlayerVisited_Context:WBP Player Visited Context Object Reference)
    IsDiscovered:bool=@n3
   ) -> (then)
   > n4 = Branch(Condition="Not Equal Exactly (Text)"!(A:"Text (by ref)"=Get LastLocationVisited() -> (LastLocationVisited:Text), B:"Text (by ref)"=@n2.LocalizedLocationName) -> (ReturnValue:bool))
   > n5 = Set LastLocationVisited(LastLocationVisited:Text=@n2.LocalizedLocationName) -> (Output_Get:Text)
   Set TimeLastLocationVisited(execute=[@n14.then, @n5.then], TimeLastLocationVisited:double=GetGameTimeInSeconds()) -> (Output_Get:double)
   > SetLocationName!(execute, self:WBP Player Visited Context Object Reference=Get WBP_PlayerVisited_Context(), LocationName:Text=Get LastLocationVisited()) -> (then)
   > SetIsNamedLocation!(
    execute
    self:WBP Player Visited Context Object Reference=Get WBP_PlayerVisited_Context()
    IsNamedLocation:bool=Get IsNamedLocation() -> (IsNamedLocation:bool)
   ) -> (then)
   > SetVisibility(self=Get WBP_PlayerVisited_Context(), InVisibility=SelfHitTestInvisible)
   > SetHasActiveContent(bNewActive=true)
   > Branch(Condition=Get Is Discovered() -> (Is Discovered:bool))
   n6 = Get Discovered Sound Cue() -> (Discovered Sound Cue:Sound Cue Soft Object Reference)
   > n7 = Branch(Condition=IsValidSoftObjectReference(SoftObjectReference=@n6))
   > n8 = Async Load Asset(Asset=@n6)
   n9 = Cast To SoundCue!(execute=@n8.Completed, Object=@n8) -> (then, CastFailed, AsSound Cue:Sound Cue Object Reference) {is=title K2Node_DynamicCast}
   > SetPOIMusic!(execute, self, Asset:Sound Base Object Reference=@n9) -> (then)
   n10 = Branch(execute=@n7.else, Condition=Get IsNamedLocation())
   > SetPOIMusic(Asset=Get POI Discover Sound() -> (POI Discover Sound:Sound Base Object Reference))
   SetPOIMusic(execute=@n10.else, Asset=Get Landmark Sound() -> (Landmark Sound:Sound Base Object Reference))
   n11 = GetConsoleVariableFloatValue!(execute=@n4.else, VariableName:String=Fort.HUD.ToastCooldownDuration) -> (then, ReturnValue:float)
   n12 = GetGameTimeInSeconds()
   n13 = Get TimeLastLocationVisited() -> (TimeLastLocationVisited:double)
   > n14 = Branch(Condition="float > float"(A=float - float(A=@n12, B=@n13) {is=title K2Node_CallFunction}, B=@n11) {is=title K2Node_CallFunction})
   Event OnStompFailed() {is=title K2Node_Event}
   > n15 = Branch(Condition=IsAnimationPlaying(InAnimation=Get BriefShow() -> (BriefShow:Widget Animation Object Reference)))
   > n16 = StopAnimation(InAnimation=Get BriefShow())
   Event On Become Active() {is=title K2Node_Event}
   > Parent: On Become Active() {is=title K2Node_CallParentFunction}
   > SetVisibility(InVisibility=HitTestInvisible)
   > n17 = GetConsoleVariableBoolValue!(execute, VariableName:String=Fort.HUD.PauseAthenaVisitedAreaDisplayAnimation) -> (then, ReturnValue:bool)
   > n18 = Branch(Condition=@n17)
   n19 = Get BriefShow()
   > PlayAnimation(InAnimation=@n19, StartAtTime=2)
   > PauseAnimation!(execute, self:User Widget Object Reference, InAnimation:Widget Animation Object Reference=@n19) -> (then, ReturnValue:float)
   PlayAnimation(execute=@n18.else, InAnimation=@n19)
   > n20 = Set Timer by Event(Delegate=Create Event(), Time=GetEndTime(self=Get BriefShow()), bLooping=false)
   > Set BriefShowFallbackTimerHandle(BriefShowFallbackTimerHandle:Timer Handle Structure=@n20) -> (Output_Get:Timer Handle Structure)
  ]}
 SetExtensionSlotsInactive {nodes [
   SetExtensionSlotsInactive() {is=title K2Node_FunctionEntry}
   > n1 = For Each Loop(Array:Array of Widget Object References=GetAllChildren(self=Get Switcher_Layout() -> (Switcher_Layout:Fort Prioritized Widget Switcher Object Reference))) -> (Array Element:Widget Object Reference)
   n2 = For Each Loop(
    Exec=@n1.Completed
    Array=GetAllChildren(self=Get Switcher_Underlay() -> (Switcher_Underlay:Fort Prioritized Widget Switcher Object Reference))
   )
   > SetVisibility(self=@n2, InVisibility=Collapsed)
   SetVisibility(execute=@n1.LoopBody, self=@n1, InVisibility=Collapsed)
  ]}
 SetMobileVisuals {nodes [
   SetMobileVisuals() {is=title K2Node_FunctionEntry}
   > Branch(Condition=IsMobileGame())
   n1 = GetOwningPlayer()
   > SetRenderOpacity(
    self=Get Switcher_Underlay()
    InOpacity=Select(Option 0:float=1, Option 1:float=0.6, Index=IsUsingTouch(PlayerController=@n1)) -> (ReturnValue:float)
   )
   > SetVisibility(
    self=Get Spacer_Mobile() -> (Spacer_Mobile:Spacer Object Reference)
    InVisibility=Select(
     Option 0:ESlateVisibility Enum=SelfHitTestInvisible
     Option 1:ESlateVisibility Enum=Collapsed
     Index=IsTabletAspectRatio!() -> (ReturnValue:bool)
    ) -> (ReturnValue:ESlateVisibility Enum)
   )
  ]}
 SetPOIMusic {nodes [
   n1 = SetPOIMusic!() -> (then, Asset:Sound Base Object Reference) {is=title K2Node_FunctionEntry}
   > n2 = AddEvent!(
    execute
    self:Music Event Subsystem Object Reference=Get MusicEventSubsystem() -> (ReturnValue:Music Event Subsystem Object Reference) {is=title K2Node_GetSubsystem}
    Asset:Sound Base Object Reference=@n1
    EventTag:Gameplay Tag Structure=(TagName="MusicEventSubsystem.Event.POI.Discovery")
    EventBehaviorTags:Gameplay Tag Container Structure=(GameplayTags=((TagName="MusicEventSubsystem.Behavior.CanRetrigger")))
   ) -> (then, ReturnValue:Audio Component Object Reference)
   > Set DiscoveryMusic(DiscoveryMusic:Audio Component Object Reference=@n2) -> (Output_Get:Audio Component Object Reference)
  ]}
 ShowOverride {nodes [ShowOverride!() -> (then, Widget:Widget Object Reference) {is=title K2Node_FunctionEntry}]}
 UnsubscribeMusicEvent {nodes [
   UnsubscribeMusicEvent() {is=title K2Node_FunctionEntry}
   > n1 = Get!(execute:Exec) -> (then:Exec, else:Exec, DiscoveryMusic:Audio Component Object Reference)
   > RemoveEvent!(execute, self:Music Event Subsystem Object Reference=Get MusicEventSubsystem(), Component:Audio Component Object Reference=@n1) -> (then)
  ]}
]
variables [
 BriefShowFallbackTimerHandle {type=TimerHandle}
 Discovered Sound Cue {type=SoundCue, default=null, class=~, properties=~}
 DiscoveryMusic {type=AudioComponent, default=null}
 Is Discovered {type=boolean}
 IsNamedLocation {type=boolean}
 Landmark Sound {type=SoundBase, default=/Game/Sounds/Fort_UI_Sounds/Gameplay/Fort_UI_InGame_POI_Landmark_Discovered_Cue}
 LastLocationVisited {type=string, default=""}
 LocationVisitedCooldown {type=number, default=60}
 POI Discover Sound {type=SoundBase, default=/Game/Sounds/Fort_UI_Sounds/Gameplay/Fort_UI_InGame_POI_Discovered_Cue}
 TimeLastLocationVisited {type=number}
]
widgets {
 info {inheritedWidgetCount=2, parentClass=/Script/FortniteUI.AthenaVisitedAreaDisplay, rootWidgetClass=&Overlay, widgetCount=9}
 widgets [
  Overlay_Main {widgetClassPath=&Overlay, widget=Overlay_Main}
  Overlay_Alignment {
   widgetClassPath=&Overlay
   bIsVariable=true
   parent=Overlay_Main
   widget=Overlay_Alignment
   slot {class=OverlaySlot, properties {horizontalAlignment=HAlign_Fill, verticalAlignment=VAlign_Bottom}}
  }
  Switcher_Underlay {
   widgetClassPath=/Script/FortniteUI.FortPrioritizedWidgetSwitcher
   bIsVariable=true
   parent=Overlay_Alignment
   widget=Switcher_Underlay
   slot {class=OverlaySlot, properties {horizontalAlignment=HAlign_Fill, verticalAlignment=VAlign_Fill, padding=<ltrb -20 0 0 0>}}
  }
  VerticalBox_0 {
   widgetClassPath=&VerticalBox
   parent=Overlay_Alignment
   widget=VerticalBox_0
   slot {class=OverlaySlot, properties {padding=<ltrb 20 10 20 10>}}
  }
  Switcher_Layout {
   widgetClassPath=&FortPrioritizedWidgetSwitcher
   bIsVariable=true
   parent=VerticalBox_0
   widget=Switcher_Layout
   slot {class=VerticalBoxSlot}
  }
  WBP_PlayerVisited_Context {
   widgetClassPath=&/WBP_PlayerVisited_Context
   bIsVariable=true
   parent=Switcher_Layout
   widget=WBP_PlayerVisited_Context
   slot {class=WidgetSwitcherSlot}
  }
  CommonTextBlock_AvailableChallenges {
   widgetClassPath=&CommonTextBlock
   bInherited=true
   bIsVariable=true
   parent=VerticalBox_0
   widget=CommonTextBlock_AvailableChallenges
   properties {
    bAutoCollapseWithEmptyText=true
    mobileFontSizeMultiplier=0.9
    text=8 Challenges Available
    textTransformPolicy=ToUpper
    colorAndOpacity {colorUseRule=UseColor_Specified, specifiedColor=<rgba 1 0.6667 0 1>}
    font {
     bForceMonospaced=false
     bMaterialIsStencil=false
     fontMaterial=null
     fontObject=&Burbank
     letterSpacing=0
     monospacedWidth=1
     size=16
     skewAmount=0.25
     typefaceFontName=BurbankRegular-Black
     outlineSettings {
      bApplyOutlineToDropShadows=false
      bMiteredCorners=false
      bSeparateFillAlpha=false
      outlineMaterial=&M-FontShadow
      outlineSize=2
      outlineColor=<rgba 0 0 0 1>
     }
    }
    shadowOffset=<0 0>
   }
   slot {class=VerticalBoxSlot, properties {verticalAlignment=VAlign_Bottom, padding=<ltrb 4 -4 0 0>, size {sizeRule=Fill, value=1}}}
  }
  Spacer_Mobile {
   widgetClassPath=&Spacer
   bIsVariable=true
   parent=VerticalBox_0
   widget=Spacer_Mobile
   properties {visibility=Collapsed, size=<1 64>}
   slot {class=VerticalBoxSlot}
  }
  TutorialHighlight {widgetClassPath=&TutorialHighlight}
 ]
}
