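# Translucent UI material for discovery text that builds a soft, noise-broken radial reveal mask.
# A tiling blurred-noise texture plus TextureFill drives a SmoothStep whose lower edge is an inverted exponential radial gradient, widened by Softness.
# RevealCenter's red channel sets both center coordinates, and the pixel-size aspect term divides a value by itself, so it applies no correction.
# No connection from the mask to a material output is shown.
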Material /Game/Athena/HUD/Discoverability/materials/M_UI_DiscoveryText
graph {
 properties {blendMode=BLEND_Translucent, materialDomain=MD_UI, pixelDepthOffsetMode=PDOM_Legacy, usedShadingModels=Default Lit}
 nodes [
  Constant() {r=1}
  n1 = MaterialFunctionCall(Float2=MaterialFunctionCall() {materialFunction=&GetUserInterfaceUV}.Pixel Size) {materialFunction=&BreakOutFloat2Components}
  n2 = AppendVector(A=Divide(A=@n1.R, B=@n1.R), B=Constant() {r=1})
  n3 = TextureCoordinate()
  n4 = TextureSampleParameter2D(UVs=Multiply(A=Multiply(A=@n2, B=@n3) {constA=1}, B=ScalarParameter() {defaultValue=1, parameterName=PatternRepeat})) {parameterName=Texture, samplerType=SAMPLERTYPE_Grayscale, texture=&LowResBlurredNoise}
  n5 = ScalarParameter() {parameterName=TextureFill}
  n6 = VectorParameter() {parameterName=RevealCenter, defaultValue=<rgba 0 0 0 1>}
  n7 = OneMinus(None=MaterialFunctionCall(
    UVs=TextureCoordinate()
    CenterPosition=AppendVector(A=@n6.R, B=@n6.R)
    Radius=ScalarParameter() {defaultValue=1, parameterName=Radius}
    Density=ScalarParameter() {defaultValue=2, parameterName=Density}
   ) {materialFunction=&RadialGradientExponential}.RadialGradientExponential)
  MaterialFunctionCall(Alpha=Saturate(None=Add(A=@n4.R, B=@n5)), Min=@n7, Max=Add(A=@n7, B=ScalarParameter() {parameterName=Softness})) {materialFunction=&SmoothStep}
  MaterialFunctionCall!(TextureResolution) {materialFunction=/Engine/Functions/Engine_MaterialFunctions02/Texturing/ScreenAlignedPixelToPixelUVs}
 ]
}
