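# UI material function in the shader library providing Photoshop/Figma-style lighten blends of a Top and Bottom layer: Lighten, Screen, Color Dodge, Plus Lighter and Lighter Color.
# Both inputs get a cheap square-root linear-to-sRGB conversion before blending.
# Lighter Color compares desaturated luminance to pick one layer, while the others use simple math.
# Each result is saturated and raised to a power of 2, a cheaper stand-in for 2.2 gamma.
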MaterialFunction /FortUILibrary/MaterialFunctions/Color/MF_UI_BlendModes_Lighten
graph {
 properties {
  bExposeToLibrary=true
  description=Allows you to blend colors similarly as Photoshop and Figma
  previewMaterialDomain=MD_UI
  libraryCategoriesText [UI_ShaderLibrary]
 }
 nodes [
  n1 = NamedRerouteUsage()
  n2 = NamedRerouteUsage()
  n3 = LinearInterpolate(
   A=@n1
   B=@n2
   Alpha=Step(
    Y=Desaturation(None=@n1) {luminanceFactors=<rgba 0.299 0.587 0.114 1>}
    X=Desaturation(None=@n2) {luminanceFactors=<rgba 0.299 0.587 0.114 1>}
   )
  )
  FunctionOutput(None=Power(Base=Saturate(None=@n3), Exp=NamedRerouteUsage()) {constExponent=2.2}) {outputName=Lighter Color, sortPriority=50}
  n4 = Max(A=NamedRerouteUsage(), B=NamedRerouteUsage())
  FunctionOutput(None=Power(Base=Saturate(None=@n4), Exp=NamedRerouteUsage()) {constExponent=2.2}) {outputName=Lighten, sortPriority=30}
  n5 = OneMinus(None=NamedRerouteUsage())
  n6 = OneMinus(None=NamedRerouteUsage())
  n7 = OneMinus(None=Multiply(
    A=NamedRerouteDeclaration(None=@n5) {name=1-Bottom, nodeColor=<rgba 0 0 0 1>}
    B=NamedRerouteDeclaration(None=@n6) {name=1-Top, nodeColor=<rgba 0 0 0 1>}
   ))
  FunctionOutput(None=Power(Base=Saturate(None=@n7), Exp=NamedRerouteUsage()) {constExponent=2.2}) {outputName=Screen, sortPriority=35}
  n8 = Divide(A=NamedRerouteUsage(), B=Max(A=NamedRerouteUsage()) {constB=0})
  FunctionOutput(None=Power(Base=Saturate(None=@n8), Exp=NamedRerouteUsage()) {constExponent=2.2}) {outputName=Color Dodge, sortPriority=40}
  n9 = Add(A=NamedRerouteUsage(), B=NamedRerouteUsage())
  FunctionOutput(None=Power(Base=Saturate(None=@n9), Exp=NamedRerouteUsage()) {constExponent=2.2}) {outputName=Plus Lighter, sortPriority=45}
  n10 = FunctionInput() {bUsePreviewValueAsDefault=true, description=Top layer which will be used to blend, inputName=Top Layer, previewValue=<1 0 0.1276 1>}
  NamedRerouteDeclaration(None=SquareRoot(None=@n10) {desc=Cheap Linear to sRGB}) {name=Top, nodeColor=<rgba 0 0 0 1>}
  n11 = FunctionInput() {
   bUsePreviewValueAsDefault=true
   description=Bottom layer which is the base that will be influenced
   inputName=Bottom Layer
   sortPriority=10
   previewValue=<0.011 0.2285 1 1>
  }
  NamedRerouteDeclaration(None=SquareRoot(None=@n11) {desc=Cheap Linear to sRGB}) {name=Bottom, nodeColor=<rgba 0 0 0 1>}
  NamedRerouteDeclaration(None=Constant() {desc="Should be 2.2, but 2 to be cheaper", r=2}) {name=Gamma Correction, nodeColor=<rgba 0 0 0 1>}
 ]
}
