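# UI material function that builds the background color and alpha for the quest variant of the HazelnutSpread Discoverability HUD.
# A large, offset soft SDF circle, with smoothing at twice its size, blends two background colors via cubic easing.
# A second soft circle forms an intensity-scaled hotspot, color-dodged over that gradient and saturated.
# Alpha is the sine-eased gradient circle; a quart ease on the hotspot feeds no output.
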MaterialFunction /Game/Athena/HUD/Discoverability/materials/Functions/MF_HazelnutSpread_Background_Color_Quest
graph {
 properties {previewMaterialDomain=MD_UI}
 nodes [
  n1 = Add(
   A=MaterialFunctionCall() {materialFunction=&GetUserInterfaceUV}.Normalized UV
   B=FunctionInput(Preview=Constant2Vector() {g=0.25, r=0.5}) {bUsePreviewValueAsDefault=true, inputName=HotSpotLocation, inputType=FunctionInput_Vector2}
  ) {constB=0.5}
  n2 = FunctionInput(Preview=Constant() {r=0.5}) {bUsePreviewValueAsDefault=true, inputName=HotspotSize, inputType=FunctionInput_Scalar}
  n3 = MaterialFunctionCall(
   DistanceField=MaterialFunctionCall(UVs=MaterialFunctionCall(UV=@n1) {materialFunction=&MF_UI_UVScaling}.ScaleX, Size=@n2) {materialFunction=&MF_UI_Circle}.SDF
   Smoothing=Multiply(A=@n2) {constB=2}
  ) {materialFunction=&MF_UI_AntiAliasing}
  n4 = Multiply(
   A=FunctionInput(Preview=Constant() {r=0.25}) {bUsePreviewValueAsDefault=true, inputName=HotspotIntensity, inputType=FunctionInput_Scalar}
   B=@n3.Inner
  )
  n5 = Add(
   A=MaterialFunctionCall() {materialFunction=&GetUserInterfaceUV}.Normalized UV
   B=FunctionInput(Preview=Constant2Vector() {g=0.25, r=0.5}) {bUsePreviewValueAsDefault=true, inputName=Offset, inputType=FunctionInput_Vector2}
  ) {constB=0.5}
  n6 = FunctionInput(Preview=Constant() {r=0.75}) {bUsePreviewValueAsDefault=true, inputName=Size, inputType=FunctionInput_Scalar}
  n7 = MaterialFunctionCall(
   DistanceField=MaterialFunctionCall(UVs=MaterialFunctionCall(UV=@n5) {materialFunction=&MF_UI_UVScaling}.ScaleX, Size=@n6) {materialFunction=&MF_UI_Circle}.SDF
   Smoothing=Multiply(A=@n6) {constB=2}
  ) {materialFunction=&MF_UI_AntiAliasing}
  n8 = LinearInterpolate(
   A=FunctionInput(Preview=Constant3Vector() {constant=<rgba 0.0033 0.0055 0.04 1>}) {bUsePreviewValueAsDefault=true, inputName=BgColor1_1, sortPriority=2}
   B=FunctionInput(Preview=Constant3Vector() {constant=<rgba 0 0.1578 0.36 1>}) {bUsePreviewValueAsDefault=true, inputName=BgColor2_1, sortPriority=3}
   Alpha=Saturate(None=MaterialFunctionCall!("Value (0-1)"=@n7.Inner) {materialFunction=/FortUILibrary/MaterialFunctions/Basics/MF_UI_EaseCurves_03_Cubic}.In)
  )
  FunctionOutput(None=Saturate(None=MaterialFunctionCall!(Blend=@n4, Base=@n8) {materialFunction=/Engine/Functions/Engine_MaterialFunctions03/Blends/Blend_ColorDodge}.Result))
  FunctionOutput(None=MaterialFunctionCall!("Value (0-1)"=@n7.Inner) {materialFunction=/FortUILibrary/MaterialFunctions/Basics/MF_UI_EaseCurves_01_Sine}.In) {outputName=Alpha, sortPriority=1}
  MaterialFunctionCall!("Value (0-1)"=@n3.Inner) {materialFunction=/FortUILibrary/MaterialFunctions/Basics/MF_UI_EaseCurves_04_Quart}
 ]
}
