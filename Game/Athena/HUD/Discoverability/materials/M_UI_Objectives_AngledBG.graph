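# Translucent UI material for the angled background bar of HUD objectives in the Athena Discoverability UI.
# A rotated UV coordinate, scaled and offset into a sharp slanted edge, is multiplied by a rounded MF_UI_Box fill.
# Hover blends between two angles, and Left-Aligned mirrors the slant.
# A static switch colors it with either a positionable two-color gradient or one color masked by a hard horizontal cutoff.
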Material /Game/Athena/HUD/Discoverability/materials/M_UI_Objectives_AngledBG
graph {
 properties {blendMode=BLEND_Translucent, materialDomain=MD_UI, pixelDepthOffsetMode=PDOM_Legacy, usedShadingModels=Default Lit}
 nodes [
  Constant() {r=1}
  n1 = LinearInterpolate(
   A=ScalarParameter() {defaultValue=0.7, group=Dont edit on instances, parameterName=Angle}
   B=ScalarParameter() {defaultValue=0.7, group=Dont edit on instances, parameterName=Angle-Hover}
   Alpha=ScalarParameter() {parameterName=Hover, sliderMax=1}
  )
  n2 = LinearInterpolate(Alpha=ScalarParameter() {group=Dont edit on instances, parameterName=Left-Aligned, sliderMax=1}) {constA=1, constB=-1, desc=flip angle if left aligned}
  n3 = ComponentMask(None=Rotator(
    Coordinate=MaterialFunctionCall() {materialFunction=&GetUserInterfaceUV}.Normalized UV
    Time=Multiply(A=@n1, B=@n2)
   ) {centerX=0, centerY=0, speed=0.2}) {r=true}
  n4 = OneMinus(None=@n3)
  n5 = ScalarParameter() {group=Dont edit on instances, parameterName=Left-Aligned, sliderMax=1}
  n6 = Add(
   A=Multiply(
    A=LinearInterpolate(A=@n3, B=@n4, Alpha=@n5) {constB=0}
    B=ScalarParameter() {defaultValue=100, group=Dont edit on instances, parameterName=strenght}
   )
   B=ScalarParameter() {defaultValue=-10, group=Dont edit on instances, parameterName=Angle Offset}
  ) {constB=-15}
  Multiply(
   A=NamedRerouteDeclaration(None=Saturate(None=@n6)) {name=Angled bar, nodeColor=<rgba 1 0 0.3281 1>}
   B=MaterialFunctionCall(Size=Constant() {r=1}, CornerRoundness=ScalarParameter() {defaultValue=0.4, group=Dont edit on instances, parameterName=CornerRoundness}) {materialFunction=&MF_UI_Box}.Fill
  )
  n7 = Subtract(
   A=MaterialFunctionCall() {materialFunction=&GetUserInterfaceUV}.Normalized UV
   B=ScalarParameter() {parameterName=Gradient_Position}
  )
  n8 = ScalarParameter() {defaultValue=0.75, parameterName=Gradient_Sharpness}
  n9 = Constant() {r=100}
  n10 = Saturate(None=Multiply(A=ComponentMask(None=@n7) {r=true}, B=StaticSwitchParameter(True=@n8, False=@n9) {defaultValue=true, parameterName=IsBGGradient?}))
  StaticSwitchParameter(
   True=LinearInterpolate(
    A=VectorParameter() {group=Colors, parameterName=BG_Color1, sortPriority=1, defaultValue=<rgba 0.7582 0.0655 0 1>}.RGB
    B=VectorParameter() {group=Colors, parameterName=BG_Color2, sortPriority=1, defaultValue=<rgba 1 0.4806 0 1>}.RGB
    Alpha=@n10
   )
   False=Multiply(A=VectorParameter() {group=Colors, parameterName=BG_Color1, sortPriority=1, defaultValue=<rgba 0.7582 0.0655 0 1>}.RGB, B=@n10)
  ) {defaultValue=true, parameterName=IsBGGradient?}
 ]
}
