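# Translucent UI material for a glow in the Athena HUD's objectives discoverability display.
# A LinearGradient U value times its inverse forms a horizontal band that peaks at the center and fades to both sides, with no texture.
# Power and Multiply parameters shape that band's falloff and intensity before a saturate clamps it.
# A Color parameter, light blue by default, supplies the tint.
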Material /Game/Athena/HUD/Discoverability/materials/M_UI_Objectives_Glow
graph {
 properties {blendMode=BLEND_Translucent, materialDomain=MD_UI, pixelDepthOffsetMode=PDOM_Legacy, usedShadingModels=Default Lit}
 nodes [
  n1 = MaterialFunctionCall(UV Channel=TextureCoordinate()) {materialFunction=&LinearGradient}
  n2 = Multiply(A=@n1.UGradient, B=OneMinus(None=@n1.UGradient))
  n3 = ScalarParameter() {defaultValue=3.5, parameterName=Power}
  Saturate(None=Multiply(A=Power(Base=@n2, Exp=@n3) {constExponent=3}, B=ScalarParameter() {defaultValue=130, parameterName=Multiply}) {constA=60})
  VectorParameter() {parameterName=Color, defaultValue=<rgba 0.0666 0.5271 0.956 1>}
 ]
}
