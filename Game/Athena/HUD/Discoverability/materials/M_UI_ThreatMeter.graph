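# Translucent UI material for a segmented threat meter bar that fills horizontally by FillAmount.
# A hard UV threshold separates background from fill, whose color shifts from FillColor toward FillColorMax as FillAmount rises.
# Opacity multiplies the background-or-fill alpha by two SmoothStep masks: rotated sine-wave segment dividers and an angled edge cut set by TopAngle.
# A gradient texture sample on the segment UVs feeds no other node.
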Material /Game/Athena/HUD/Discoverability/materials/M_UI_ThreatMeter
graph {
 properties {blendMode=BLEND_Translucent, materialDomain=MD_UI, pixelDepthOffsetMode=PDOM_Legacy, usedShadingModels=Default Lit}
 nodes [
  n1 = VectorParameter() {parameterName=BGColor, defaultValue=<rgba 0.0052 0.0052 0.0056 0.8>}
  n2 = ScalarParameter() {defaultValue=0.33, parameterName=FillAmount}
  n3 = TextureCoordinate()
  n4 = Constant() {r=1}
  n5 = OneMinus(None=If(A=ComponentMask(None=@n3) {r=true}, B=@n2, "A > B"=@n4, "A == B"=@n4, "A < B"=Constant()))
  LinearInterpolate(
   A=@n1.RGB
   B=LinearInterpolate(
    A=VectorParameter() {parameterName=FillColor, defaultValue=<rgba 1 0.491 0 1>}.RGB
    B=VectorParameter() {parameterName=FillColorMax, defaultValue=<rgba 1 0 0 1>}.RGB
    Alpha=LinearInterpolate(Alpha=@n2) {constA=-0.33}
   )
   Alpha=@n5
  )
  n6 = TextureCoordinate()
  n7 = ScalarParameter() {defaultValue=0.25, parameterName=SegmentAngle}
  n8 = Constant()
  n9 = ScalarParameter() {defaultValue=0.004, parameterName=SegmentsOffset}
  n10 = ScalarParameter() {defaultValue=1, parameterName=Segments}
  n11 = Multiply(
   A=Add(
    A=MaterialFunctionCall(UVs=@n6, "Rotation Angle (0-1)"=@n7) {materialFunction=&CustomRotator}.Rotated Values
    B=AppendVector(A=@n8, B=@n9)
   ) {constB=0}
   B=AppendVector(A=Constant() {desc=stretch vertical to eliminate artifacts, r=0.01}, B=@n10)
  )
  n12 = ComponentMask(None=Abs(None=Sine(None=@n11) {period=2})) {g=true}
  n13 = Multiply(A=@n10, B=ScalarParameter() {defaultValue=0.0002, parameterName=SegmentEdge})
  n14 = Add(A=@n13, B=Multiply(A=@n10, B=ScalarParameter() {defaultValue=0, parameterName=SegmentEdgeSoftness}))
  n15 = VectorParameter() {parameterName=TopAngle, defaultValue=<rgba 0.5 0 0.1 0>}
  n16 = ComponentMask(None=MaterialFunctionCall(UVs=TextureCoordinate(), Rotation Center=AppendVector(A=@n15.R, B=@n15.R), "Rotation Angle (0-1)"=@n15.B) {materialFunction=&CustomRotator}.Rotated Values) {g=true}
  n17 = ScalarParameter() {parameterName=Edge}
  n18 = Add(A=Multiply(A=@n17, B=ScalarParameter() {parameterName=EdgeSoftness}), B=@n17)
  Multiply(
   A=LinearInterpolate(A=@n1.A, B=Constant() {r=1}, Alpha=@n5)
   B=Multiply(
    A=MaterialFunctionCall(Alpha=@n12, Min=@n13, Max=@n14) {materialFunction=&SmoothStep}.Result
    B=MaterialFunctionCall(Alpha=@n16, Min=@n17, Max=@n18) {materialFunction=&SmoothStep}.Result
   )
  )
  TextureSample(UVs=@n11) {texture=/Game/Global/Textures/Default/Masks_and_Gradients/T_UI_ReticleSegmentGradient}
 ]
}
