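# Sound cue for the AI companion eliminated notification, part of the in-game UI sounds for AI companion controls.
# A single wave player feeds the AI_Companion_Eliminated wave straight to the output, with no randomization or modulation.
# The cue plays boosted in volume through the non-critical in-game UI sound class, with a very large max distance.
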SoundCue /Game/Sounds/Fort_UI_Sounds/AI_CompanionControls/AI_Companion_Eliminated_Cue
root=WavePlayer_0
nodes [{id=WavePlayer_0, properties {soundWaveAssetPtr=&/AI_Companion_Eliminated}}]
properties {maxDistance=2097152, soundClassObject=&Fort_UI_InGame_NonCritical, volumeMultiplier=1.6}
