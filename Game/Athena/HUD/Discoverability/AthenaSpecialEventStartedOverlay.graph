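# HUD banner widget that announces special events and urgent quests, such as bounties and hired NPC eliminations.
# The urgent quest message event fills text, colors title and alert icon by tag or objective category, and chooses player portrait or alert icon.
# It async-loads the image and sound, then plays the HiredNPCElim animation and sets an outro-delay timer.
# FadeAnim plays when the prioritized display finishes; currency art follows any playspace override.
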WidgetBlueprint /Game/Athena/HUD/Discoverability/AthenaSpecialEventStartedOverlay
animations [HiredNPCElim {
  binding_count=8
  length_seconds=4.7462
  object=&self:HiredNPCElim
  sequence {bindings [
    Image_NPCElimPortrait {
     id=#1
     possessed_class=Image
     tracks [
      {
       type=MovieSceneWidgetMaterialTrack
       sections [{
         type=MovieSceneParameterSection
         channels [
          InnerMaskRevealRadius {type=MovieSceneScriptingFloatChannel}
          InnerMaskRevealDensity {type=MovieSceneScriptingFloatChannel}
          MaskRevealDensity {type=MovieSceneScriptingFloatChannel, keys [[23,6.7401],[46,0],[48,6.7401]]}
          MaskRevealRadius {type=MovieSceneScriptingFloatChannel, const [6.5759,1]}
          GlobalScale {type=MovieSceneScriptingFloatChannel, keys [[-6,0.7],[46,0.7],[48,1]]}
          Mask Size {type=MovieSceneScriptingFloatChannel, const [0.7048,1]}
          BackgroundRGBR {type=MovieSceneScriptingFloatChannel, keys [[-6,0.008],[70,0.008]]}
          BackgroundRGBG {type=MovieSceneScriptingFloatChannel, keys [[-6,0.9047],[70,0.906]]}
          BackgroundRGBB {type=MovieSceneScriptingFloatChannel, keys [[-6,0.9301],[70,0.929]]}
          BackgroundRGBA {type=MovieSceneScriptingFloatChannel, const [1,1]}
          HighlightRGBR {type=MovieSceneScriptingFloatChannel, const [0.008,2]}
          HighlightRGBG {type=MovieSceneScriptingFloatChannel, keys [[0,0.4425],[70,0.906]]}
          HighlightRGBB {type=MovieSceneScriptingFloatChannel, const [0.929,2]}
          HighlightRGBA {type=MovieSceneScriptingFloatChannel, const [1,2]}
         ]
        }]
      }
      {
       type=MovieScene2DTransformTrack
       sections [{
         type=MovieScene2DTransformSection
         channels [
          Translation.X {type=MovieSceneScriptingFloatChannel}
          Translation.Y {type=MovieSceneScriptingFloatChannel}
          Angle {type=MovieSceneScriptingFloatChannel}
          Scale.X {type=MovieSceneScriptingFloatChannel, keys [[0,0.5],[5,1.2],[8,1]]}
          Scale.Y {type=MovieSceneScriptingFloatChannel, keys [[0,0.5],[5,1.2],[8,1]]}
          Shear.X {type=MovieSceneScriptingFloatChannel}
          Shear.Y {type=MovieSceneScriptingFloatChannel}
         ]
        }]
      }
     ]
    }
    Image_Glow {
     id=#2
     display_name=Image_NPCElim_Glow
     possessed_class=Image
     tracks [
      {
       type=MovieSceneFloatTrack
       display_name=Render Opacity
       sections [{type=MovieSceneFloatSection, channels [None {type=MovieSceneScriptingFloatChannel, keys [[0,0],[5,0.5],[21,0.8],[42,1],[46,0],[70,0]]}]}]
      }
      {
       type=MovieSceneColorTrack
       display_name=Color and Opacity
       sections [{
         type=MovieSceneColorSection
         end_frame=70
         start_frame=-6
         channels [
          Color.R {type=MovieSceneScriptingFloatChannel}
          Color.G {type=MovieSceneScriptingFloatChannel}
          Color.B {type=MovieSceneScriptingFloatChannel}
          Color.A {type=MovieSceneScriptingFloatChannel}
         ]
        }]
      }
      {
       type=MovieSceneWidgetMaterialTrack
       sections [{
         type=MovieSceneParameterSection
         end_frame=70
         start_frame=-6
         channels [
          Color1R {type=MovieSceneScriptingFloatChannel, const [0,1]}
          Color1G {type=MovieSceneScriptingFloatChannel, const [0.0069,1]}
          Color1B {type=MovieSceneScriptingFloatChannel, const [0,1]}
          Color1A {type=MovieSceneScriptingFloatChannel, const [1,1]}
          Color2R {type=MovieSceneScriptingFloatChannel, const [0,1]}
          Color2G {type=MovieSceneScriptingFloatChannel, const [0.5776,1]}
          Color2B {type=MovieSceneScriptingFloatChannel, const [1,1]}
          Color2A {type=MovieSceneScriptingFloatChannel, const [1,1]}
         ]
        }]
      }
      {
       type=MovieScene2DTransformTrack
       sections [{
         type=MovieScene2DTransformSection
         end_frame=8
         start_frame=0
         channels [
          Translation.X {type=MovieSceneScriptingFloatChannel, const [0,1]}
          Translation.Y {type=MovieSceneScriptingFloatChannel, const [0,1]}
          Angle {type=MovieSceneScriptingFloatChannel, const [0,1]}
          Scale.X {type=MovieSceneScriptingFloatChannel, keys [[0,0.5],[8,0.7]]}
          Scale.Y {type=MovieSceneScriptingFloatChannel, keys [[0,0.5],[8,0.7]]}
          Shear.X {type=MovieSceneScriptingFloatChannel, const [0,1]}
          Shear.Y {type=MovieSceneScriptingFloatChannel, const [0,1]}
         ]
        }]
      }
     ]
    }
    Image_X_LR {
     id=#3
     display_name=Image_X_LeftRight
     possessed_class=Image
     tracks [
      {
       type=MovieScene2DTransformTrack
       sections [{
         type=MovieScene2DTransformSection
         end_frame=60
         start_frame=0
         channels [
          Translation.X {type=MovieSceneScriptingFloatChannel}
          Translation.Y {type=MovieSceneScriptingFloatChannel}
          Angle {type=MovieSceneScriptingFloatChannel}
          Scale.X {type=MovieSceneScriptingFloatChannel, keys [[3.312,0],[9.141,0],[12.903,1.2],[14.312,1],[15.022,1]]}
          Scale.Y {type=MovieSceneScriptingFloatChannel, keys [[3.312,0],[9.141,0],[12.903,1.2],[14.312,1],[15.022,1]]}
          Shear.X {type=MovieSceneScriptingFloatChannel}
          Shear.Y {type=MovieSceneScriptingFloatChannel}
         ]
        }]
      }
      {
       type=MovieSceneColorTrack
       display_name=Color and Opacity
       sections [{
         type=MovieSceneColorSection
         end_frame=60
         start_frame=0
         channels [
          Color.R {type=MovieSceneScriptingFloatChannel, const [1,1]}
          Color.G {type=MovieSceneScriptingFloatChannel, const [1,1]}
          Color.B {type=MovieSceneScriptingFloatChannel, const [1,1]}
          Color.A {type=MovieSceneScriptingFloatChannel, keys [[40,1],[43,0]]}
         ]
        }]
      }
     ]
    }
    HorizontalBox_1 {
     id=#4
     possessed_class=HorizontalBox
     tracks [
      {
       type=MovieSceneFloatTrack
       display_name=Render Opacity
       sections [{type=MovieSceneFloatSection, channels [None {type=MovieSceneScriptingFloatChannel, keys [[0,0],[3.4393,1],[49,1]]}]}]
      }
      {
       type=MovieScene2DTransformTrack
       sections [{
         type=MovieScene2DTransformSection
         channels [
          Translation.X {type=MovieSceneScriptingFloatChannel}
          Translation.Y {type=MovieSceneScriptingFloatChannel}
          Angle {type=MovieSceneScriptingFloatChannel}
          Scale.X {type=MovieSceneScriptingFloatChannel, keys [[0,2.7136],[42,2.7136],[46.9227,1]]}
          Scale.Y {type=MovieSceneScriptingFloatChannel, keys [[0,2.7136],[42,2.7136],[46.9227,1]]}
          Shear.X {type=MovieSceneScriptingFloatChannel}
          Shear.Y {type=MovieSceneScriptingFloatChannel}
         ]
        }]
      }
     ]
    }
    CommonWidgetSwitcher_IconType {
     id=#5
     possessed_class=CommonWidgetSwitcherLegacy
     tracks [{
       type=MovieSceneIntegerTrack
       display_name=Active Widget Index
       sections [{type=MovieSceneIntegerSection, channels [None {type=MovieSceneScriptingIntegerChannel, const [3,1]}]}]
      }]
    }
    Image_BorderFrame {
     id=#6
     display_name=Image_FrameBorder
     possessed_class=Image
     tracks [
      {
       type=MovieScene2DTransformTrack
       sections [{
         type=MovieScene2DTransformSection
         end_frame=84
         start_frame=0
         channels [
          Translation.X {type=MovieSceneScriptingFloatChannel}
          Translation.Y {type=MovieSceneScriptingFloatChannel}
          Angle {type=MovieSceneScriptingFloatChannel}
          Scale.X {type=MovieSceneScriptingFloatChannel, keys [[0,0],[9.312,0],[14.312,3.7],[42,3.6963],[46.9227,0]]}
          Scale.Y {type=MovieSceneScriptingFloatChannel, keys [[0,0],[9.312,0],[14.312,3.7],[42,3.6963],[46.9227,0]]}
          Shear.X {type=MovieSceneScriptingFloatChannel}
          Shear.Y {type=MovieSceneScriptingFloatChannel}
         ]
        }]
      }
      {
       type=MovieSceneWidgetMaterialTrack
       sections [{
         type=MovieSceneParameterSection
         end_frame=70
         start_frame=-6
         channels [
          Color1R {type=MovieSceneScriptingFloatChannel, const [0.008,1]}
          Color1G {type=MovieSceneScriptingFloatChannel, const [0.9047,1]}
          Color1B {type=MovieSceneScriptingFloatChannel, const [0.9301,1]}
          Color1A {type=MovieSceneScriptingFloatChannel, const [1,1]}
          Color2R {type=MovieSceneScriptingFloatChannel, const [0.008,1]}
          Color2G {type=MovieSceneScriptingFloatChannel, const [0.9047,1]}
          Color2B {type=MovieSceneScriptingFloatChannel, const [0.9301,1]}
          Color2A {type=MovieSceneScriptingFloatChannel, const [0.3,1]}
         ]
        }]
      }
      {
       type=MovieSceneColorTrack
       display_name=Color and Opacity
       sections [{
         type=MovieSceneColorSection
         end_frame=60
         start_frame=0
         channels [
          Color.R {type=MovieSceneScriptingFloatChannel, const [1,1]}
          Color.G {type=MovieSceneScriptingFloatChannel, const [1,1]}
          Color.B {type=MovieSceneScriptingFloatChannel, const [1,1]}
          Color.A {type=MovieSceneScriptingFloatChannel, keys [[48,1],[51,0]]}
         ]
        }]
      }
     ]
    }
    CommonBorder_0 {
     id=#7
     possessed_class=CommonBorder
     tracks [
      {
       type=MovieSceneByteTrack
       display_name=Clipping
       sections [{type=MovieSceneByteSection, channels [None {type=MovieSceneScriptingByteChannel, keys [[0,1],[52.567,0]]}]}]
      }
      {
       type=MovieSceneFloatVectorTrack
       display_name=Desired Size Scale
       sections [{
         type=MovieSceneFloatVectorSection
         channels [
          Vector.X {type=MovieSceneScriptingFloatChannel, keys [[47.567,0],[52.567,1]]}
          Vector.Y {type=MovieSceneScriptingFloatChannel}
         ]
        }]
      }
      {
       type=MovieSceneFloatTrack
       display_name=Render Opacity
       sections [{type=MovieSceneFloatSection, channels [None {type=MovieSceneScriptingFloatChannel, keys [[47.567,0],[48.239,1]]}]}]
      }
     ]
    }
    Image_X_RL {
     id=#8
     display_name=Image_X_RightLeft
     possessed_class=Image
     tracks [
      {
       type=MovieScene2DTransformTrack
       sections [{
         type=MovieScene2DTransformSection
         end_frame=14
         start_frame=0
         channels [
          Translation.X {type=MovieSceneScriptingFloatChannel}
          Translation.Y {type=MovieSceneScriptingFloatChannel}
          Angle {type=MovieSceneScriptingFloatChannel}
          Scale.X {type=MovieSceneScriptingFloatChannel, keys [[0.171,0],[6,0],[9.762,1.2],[11.881,1]]}
          Scale.Y {type=MovieSceneScriptingFloatChannel, keys [[0.171,0],[6,0],[9.762,1.2],[11.881,1]]}
          Shear.X {type=MovieSceneScriptingFloatChannel}
          Shear.Y {type=MovieSceneScriptingFloatChannel}
         ]
        }]
      }
      {
       type=MovieSceneColorTrack
       display_name=Color and Opacity
       sections [{
         type=MovieSceneColorSection
         end_frame=60
         start_frame=0
         channels [
          Color.R {type=MovieSceneScriptingFloatChannel, const [1,1]}
          Color.G {type=MovieSceneScriptingFloatChannel, const [1,1]}
          Color.B {type=MovieSceneScriptingFloatChannel, const [1,1]}
          Color.A {type=MovieSceneScriptingFloatChannel, keys [[40,1],[43,0]]}
         ]
        }]
      }
     ]
    }
   ]}
 }]
graphs [
 ClearActiveContent {nodes [
   ClearActiveContent() {is=title K2Node_FunctionEntry}
   > SetHasActiveContent!(execute, self:Fort Prioritized Widget Interface=Self-Reference(), bNewActive:bool=false) -> (then)
  ]}
 EventGraph {nodes [
   EventAssetsReady() {is=title K2Node_CustomEvent}
   > StopAllAnimations(self:User Widget Object Reference)
   > StopAnimationByName!(
    execute
    self:Fort Simple Widget Animations Panel Object Reference=Get SimpleAnimations() -> (SimpleAnimations:Fort Simple Widget Animations Panel Object Reference)
    Name:"Name (by ref)"=Get FadeAnim() -> (FadeAnim:Name=None)
   ) -> (then, ReturnValue:bool)
   > SetRenderOpacity(self=Get Overlay_Container() -> (Overlay_Container:Overlay Object Reference), InOpacity=1)
   > PlayAnimation(InAnimation=Get HiredNPCElim() -> (HiredNPCElim:Widget Animation Object Reference))
   > Clear and Invalidate Timer by Handle(Handle=Get TimerHandle() -> (TimerHandle:Timer Handle Structure))
   > n1 = Set Timer by Event(
    Delegate=Create Event()
    Time=GetOutroDelay!(self:Athena Special Event Started Overlay Object Reference) -> (ReturnValue:float)
    bLooping=false
   )
   > Set TimerHandle(TimerHandle:Timer Handle Structure=@n1) -> (Output_Get:Timer Handle Structure)
   PlayHiredNPCElimAudio() {is=title K2Node_CustomEvent}
   > Is Valid(InputObject=Get AudioMessage() -> (AudioMessage:Sound Base Object Reference))
   > PlayMessageAudio()
   n2 = Event OnShowUrgentQuestMessage!() -> (OutputDelegate, then, UrgentQuestData:"Urgent Quest Data Structure (by ref)", EventTimeRemaining:float) {is=title K2Node_Event}
   n3 = Break Urgent Quest Data!(UrgentQuestData:"Urgent Quest Data Structure (by ref)"=@n2.UrgentQuestData) -> (
    EventTag:Gameplay Tag Structure
    VisualStyleTag:Gameplay Tag Container Structure
    TotalEventTime:float=~
    bDisplayTimer:bool=~
    bQuestCompleteOnTimeExpiration:bool=~
    bQuickIntro:bool=~
    bQuickOutro:bool=~
    bRewardPersist:bool=~
    bDisplayMultiStepExpanded:bool=~
    EventMainTitle:Text
    EventTitle:Text
    EventSubtitle:Text
    EventSubtitleSecondary:Text
    EventFailureMessage:Text
    EventDescription:Text
    EventDescriptionGamepad:Text
    EventDescriptionTouch:Text
    EventDescriptionOverrides:Array of Urgent Quest Data Desc Overrides Structures
    EventDescriptionOverrideTag:Gameplay Tag Structure
    SortPriority:int
    bHideQuestProgressBar:bool=~
    bDisplayAllObjectivesIndividually:bool=~
    QuestProgressTextFormat:Text
    bReverseProgressBar:bool=~
    bShowBountyThreatInformation:bool=~
    bShowBountyPriceAndDistance:bool=~
    bDisplayStartedEventUI:bool=~
    bIsSharedUrgentQuest:bool=~
    DisplayPlayer:Fort Player State Athena Object Reference
    AcceptingPlayer:Fort Player State Athena Object Reference
    SocialAvatarBrushPtr:Texture 2D Soft Object Reference
    BountyPriceImage:Object Reference
    AlertIcon:Soft Object Reference
    CharacterIcon:Soft Object Reference
    SpecialEventStartedSound:Sound Base Soft Object Reference
    AlertIconBorderColor:Linear Color Structure
    AlertIconBrushSize:Vector 2D Structure
    bOverrideWidgetPrimaryColor:bool=~
    OverriddenWidgetPrimaryColor:Linear Color Structure
    ObjectiveCategory:EObjectiveCategories Enum
   ) {is=title K2Node_BreakStruct}
   > Set MessageEventTag(MessageEventTag:Gameplay Tag Structure=@n3) -> (Output_Get:Gameplay Tag Structure)
   > Set Objective Category(Objective Category:EObjectiveCategories Enum=@n3.ObjectiveCategory) -> (Output_Get:EObjectiveCategories Enum=Default)
   > SetText(self=Get RichText_Title() -> (RichText_Title:Common Rich Text Block Object Reference), InText=@n3.EventTitle)
   > UpdateTitleColor()
   > SetText(self=Get RichText_SubTitle() -> (RichText_SubTitle:Common Rich Text Block Object Reference), InText=@n3.EventSubtitle)
   > n4 = Sequence()
   > n5 = Is Valid(InputObject=@n3.DisplayPlayer)
   > SetActiveWidgetIndex(self=Get CommonWidgetSwitcher_IconType() -> (CommonWidgetSwitcher_IconType:Common Widget Switcher Legacy Object Reference), Index=1)
   > n6 = Async Load Asset(Asset=@n3.SocialAvatarBrushPtr)
   n7 = GetDynamicMaterial(execute=@n6.Completed, self=Get Image_PlayerPortrait() -> (Image_PlayerPortrait:Image Object Reference))
   > n8 = GetDynamicMaterial(self=Get Image_ElimPortrait() -> (Image_ElimPortrait:Image Object Reference))
   > n9 = GetDynamicMaterial(self=Get Image_NPCElimPortrait() -> (Image_NPCElimPortrait:Image Object Reference))
   > n10 = SetTextureParameterValue(
    self=[@n9, @n7, @n8]
    ParameterName=SmallPreview
    Value=Resolve Soft Reference!(Input:Texture 2D Soft Object Reference=@n3.SocialAvatarBrushPtr) -> (Output:Texture 2D Object Reference)
   )
   n11 = Select(
    Option 0:Sound Base Soft Object Reference=@n3.SpecialEventStartedSound
    Option 1:Sound Base Soft Object Reference=Get AudioMessage_HiredNPCEliminated() -> (AudioMessage_HiredNPCEliminated:Sound Base Soft Object Reference)
    Index="Equal (GameplayTag)"!(
     A:Gameplay Tag Structure=Get MessageEventTag() -> (MessageEventTag:Gameplay Tag Structure)
     B:Gameplay Tag Structure=Get HiredNPCEliminatedTag() -> (HiredNPCEliminatedTag:Gameplay Tag Structure)
    ) -> (ReturnValue:bool) {is=title K2Node_CallFunction}
   ) -> (ReturnValue:Sound Base Soft Object Reference)
   n12 = Async Load Asset(execute=[@n10.then, @n13.then], Asset=@n11)
   Set AudioMessage(
    execute=@n12.Completed
    AudioMessage:Sound Base Object Reference=Resolve Soft Reference(Input:Sound Base Soft Object Reference=@n11) -> (Output:Sound Base Object Reference)
   ) -> (Output_Get:Sound Base Object Reference)
   > EventAssetsReady() {is=title K2Node_CallFunction}
   SetActiveWidgetIndex(execute=@n5.Is Not Valid, self=Get CommonWidgetSwitcher_IconType(), Index=0)
   > n13 = Async Load Asset(Asset=@n3.AlertIcon)
   TryUpdateUrgentAlertIcon!(
    execute=@n13.Completed
    self
    AlertIcon:Object Reference=@n13
    AlertIconBrushSize:Vector 2D Structure=@n3.AlertIconBrushSize
    AlertIconBorderColor:Linear Color Structure=@n3.AlertIconBorderColor
   ) -> (then)
   n14 = TextIsEmpty(InText=@n3.EventSubtitleSecondary)
   n15 = Branch(execute=@n4.then_1, Condition=AND Boolean(A=IsValid(Object=@n3.AcceptingPlayer), B=NOT Boolean(A=@n14)))
   > SetVisibility(self=Get RichText_Description() -> (RichText_Description:Common Rich Text Block Object Reference), InVisibility=SelfHitTestInvisible)
   > SetText(self=Get RichText_Description(), InText=@n3.EventSubtitleSecondary)
   SetVisibility(execute=@n15.else, self=Get RichText_Description(), InVisibility=Collapsed)
   Event Construct()
   > BindToAnimationStarted!(execute, self:User Widget Object Reference, Animation:Widget Animation Object Reference=Get HiredNPCElim(), Delegate:Delegate=Create Event()) -> (then)
   > n16 = GetPlayspaceGlobalCurrencyOverrideForActor!(execute, Actor:Actor Object Reference=GetOwningPlayer()) -> (then, ReturnValue:Playspace Component Global Currency Override Object Reference)
   > Is Valid(InputObject=@n16)
   > SetBrushFromSoftTexture!(
    execute
    self:Image Object Reference=Get Image_Gold() -> (Image_Gold:Image Object Reference)
    SoftTexture:Texture 2D Soft Object Reference=GetLargePreviewImage!(self:Fort Item Definition Object Reference=Get OverrideGlobalCurrency!(self:Playspace Component Global Currency Override Object Reference=@n16) -> (OverrideGlobalCurrency:Fort World Item Definition Object Reference)) -> (ReturnValue:Texture 2D Soft Object Reference)
    bMatchSize:bool=false
   ) -> (then)
   Event OnPrioitizedFinished() {is=title K2Node_Event}
   > Clear and Invalidate Timer by Handle(Handle=Get TimerHandle())
   > PlayAnimationByName!(
    execute
    self:Fort Simple Widget Animations Panel Object Reference=Get SimpleAnimations()
    Name:"Name (by ref)"=Get FadeAnim() -> (FadeAnim=None)
    Speed:float=1
    StartAtTime:float=0
   ) -> (then, ReturnValue:bool)
  ]}
 PlayMessageAudio {nodes [
   PlayMessageAudio() {is=title K2Node_FunctionEntry}
   n1 = GetVisibility!(self:Widget Object Reference) -> (ReturnValue:ESlateVisibility Enum=Visible)
   > Branch(Condition=NOT Boolean(A="Equal (Enum)"(A:ESlateVisibility Enum=@n1, B:ESlateVisibility Enum=Collapsed)))
   > PlaySound2D(Sound=Get AudioMessage())
  ]}
 TryUpdateUrgentAlertIcon {nodes [
   n1 = TryUpdateUrgentAlertIcon!() -> (then, AlertIcon:Object Reference, AlertIconBrushSize:Vector 2D Structure, AlertIconBorderColor:Linear Color Structure) {is=title K2Node_FunctionEntry}
   > n2 = Sequence()
   > n3 = Is Valid(InputObject=@n1)
   > n4 = Sequence()
   > SetVisibility(self=Get Text_UrgentIcon() -> (Text_UrgentIcon:Common Text Object Reference), InVisibility=Collapsed)
   n5 = Get Image_UrgentIcon() -> (Image_UrgentIcon:Image Object Reference)
   > SetVisibility(self=@n5, InVisibility=HitTestInvisible)
   > SetBrushResourceObject!(execute, self:Image Object Reference=@n5, ResourceObject:Object Reference=@n1) -> (then)
   n6 = Get DefaultTint() -> (DefaultTint:Slate Color Structure)
   > SetBrushTintColor!(execute, self:Image Object Reference=Get Image_UrgentIcon(), TintColor:Slate Color Structure=@n6) -> (then)
   Branch(execute=@n4.then_1, Condition=NOT Boolean(A=IsZero2D!(A:"Vector 2D Structure (by ref)"=@n1.AlertIconBrushSize) -> (ReturnValue:bool)))
   > SetDesiredSizeOverride!(execute, self:Image Object Reference=@n5, DesiredSize:Vector 2D Structure=@n1.AlertIconBrushSize) -> (then)
   SetVisibility(execute=@n3.Is Not Valid, self=Get Text_UrgentIcon(), InVisibility=HitTestInvisible)
   > SetVisibility(self=Get Image_UrgentIcon(), InVisibility=Collapsed)
   n7 = GetDynamicMaterial(execute=@n2.then_1, self=Get Image_UrgentAlertBackground() -> (Image_UrgentAlertBackground:Image Object Reference))
   n8 = Get RankedQuestTag() -> (RankedQuestTag:Gameplay Tag Structure)
   n9 = Get MessageEventTag()
   n10 = Get Objective Category() -> (Objective Category:EObjectiveCategories Enum=Default)
   n11 = Get BountyTargetTag() -> (BountyTargetTag:Gameplay Tag Structure)
   n12 = Get BountyHunterTag() -> (BountyHunterTag:Gameplay Tag Structure)
   n13 = Get BountyFailed() -> (BountyFailed:Gameplay Tag Structure)
   n14 = Get BountySuccess() -> (BountySuccess:Gameplay Tag Structure)
   n15 = Get BountyPoached() -> (BountyPoached:Gameplay Tag Structure)
   n16 = SelectColor(
    A=(R=1,G=0.0168,B=0,A=1)
    B=SelectColor(
     A=(R=0.8469,G=0.5711,B=0.9734,A=1)
     B=(R=0,G=0.7157,B=1,A=1)
     bPickA="Equal (GameplayTag)"(A=@n9, B=@n8) {is=title K2Node_PromotableOperator}
    )
    bPickA=OR Boolean!(
     A="Equal (GameplayTag)"(A=@n9, B=@n11)
     B="Equal (GameplayTag)"(A=@n9, B=@n12)
     C:bool="Equal (GameplayTag)"(A=@n9, B=@n14)
     D:bool="Equal (GameplayTag)"(A=@n9, B=@n15)
     E:bool="Equal (GameplayTag)"(A=@n9, B=@n13)
     F:bool="Equal (Enum)"(A:EObjectiveCategories Enum=@n10, B:EObjectiveCategories Enum=Combat)
    ) -> (ReturnValue)
   )
   > SetVectorParameterValue(self=@n7, ParameterName=Get UrgentAlertIconVectorParameterValueName() -> (UrgentAlertIconVectorParameterValueName:Name=None), Value=@n16)
   > SetBrushTintColor(self=Get Image_UrgentPulse() -> (Image_UrgentPulse:Image Object Reference), TintColor=Make Slate Color(SpecifiedColor=@n16))
   Reroute Node(InputPin:Linear Color Structure=@n1.AlertIconBorderColor) -> (OutputPin:Linear Color Structure)
   n17 = Get DefaultTint()
   Select(
    Option 0:Slate Color Structure=Make Slate Color(SpecifiedColor=(R=1,G=1,B=1,A=1))
    Option 1:Slate Color Structure=@n6
    Index="Equal (LinearColor)"!(
     A:Linear Color Structure=Get AlertIconBorderColor() -> (AlertIconBorderColor:Linear Color Structure)
     B:Linear Color Structure=Break Slate Color!(SlateColor:"Slate Color Structure (by ref)"=@n17) -> (SpecifiedColor:Linear Color Structure, ColorUseRule:ESlateColorStylingMode Enum) {is=title K2Node_BreakStruct}
    ) -> (ReturnValue:bool)
   ) -> (ReturnValue:Slate Color Structure)
  ]}
 UpdateTitleColor {nodes [
   UpdateTitleColor() {is=title K2Node_FunctionEntry}
   n1 = Get MessageEventTag()
   n2 = SelectColor(A=(R=1,G=1,B=1,A=1), B=(R=0.0168,G=0.6867,B=1,A=1), bPickA="Equal (GameplayTag)"(A=@n1, B=Get RankedQuestTag()))
   n3 = OR Boolean(
    A="Equal (GameplayTag)"(A=@n1, B=Get BountyTargetTag())
    B="Equal (GameplayTag)"(A=@n1, B=Get BountyHunterTag())
    C="Equal (GameplayTag)"(A=@n1, B=Get BountySuccess())
    D="Equal (GameplayTag)"(A=@n1, B=Get BountyPoached())
    E="Equal (GameplayTag)"(A=@n1, B=Get BountyFailed())
    F="Equal (Enum)"(A=Get Objective Category() -> (Objective Category=Default), B=Combat)
   )
   > SetDefaultColorAndOpacity!(
    execute
    self:Rich Text Block Object Reference=Get RichText_Title()
    InColorAndOpacity:Slate Color Structure=Make Slate Color(SpecifiedColor=SelectColor(A=(R=1,G=0.0762,B=0,A=1), B=@n2, bPickA=@n3))
   ) -> (then)
   n4 = SelectColor(A=(R=0.107,G=0.0033,B=0,A=1), B=(R=0.0044,G=0.0723,B=0.1714,A=1), bPickA=@n3)
   n5 = MakeLiteralInt!(Value:int=40) -> (ReturnValue:int=0)
   > SetDefaultFont!(
    execute
    self:Rich Text Block Object Reference=Get RichText_Title()
    InFontInfo:Slate Font Info Structure=MakeSlateFontInfo!(
     FontObject:Object Reference=&Burbank
     FontMaterial:Object Reference=/Game/Athena/HUD/Discoverability/materials/M_UI_DiscoveryText_Inst
     OutlineSettings:Font Outline Settings Structure=Make Font Outline Settings!(
      OutlineSize:int=2
      bMiteredCorners:bool=False
      bSeparateFillAlpha:bool=False
      bApplyOutlineToDropShadows:bool=False
      OutlineMaterial:Object Reference
      OutlineColor:Linear Color Structure=@n4
     ) -> (FontOutlineSettings:Font Outline Settings Structure) {is=title K2Node_MakeStruct}
     TypefaceFontName:Name=BurbankRegular-Black
     Size:float="To Float (Integer)"(InInt=@n5)
     LetterSpacing:int=0
     SkewAmount:float=0
     bForceMonospaced:bool=false
     bMaterialIsStencil:bool=false
     MonospacedWidth:float=1
    ) -> (ReturnValue:Slate Font Info Structure)
   ) -> (then)
  ]}
]
variables [
 AudioMessage {type=SoundBase, default=null}
 AudioMessage_HiredNPCEliminated {type=SoundBase, default=/Game/Sounds/Fort_UI_Sounds/AI_CompanionControls/AI_Companion_Eliminated_Cue}
 BountyFailed {type=GameplayTag, default {tagName=SpecialEvent.HuntingPlayer.Failed}}
 BountyHunterTag {type=GameplayTag, default {tagName=SpecialEvent.huntingplayer.isHunter}}
 BountyPoached {type=GameplayTag, default {tagName=SpecialEvent.HuntingPlayer.Poached}}
 BountySuccess {type=GameplayTag, default {tagName=SpecialEvent.huntingPlayer.Success}}
 BountyTargetTag {type=GameplayTag, default {tagName=SpecialEvent.HuntingPlayer.Target}}
 DefaultTint {type=SlateColor, default {colorUseRule=UseColor_Specified, specifiedColor=<rgba 1 1 1 1>}}
 FadeAnim {type=string, default=FadeAnim}
 FreshCheeseIcon {type=Texture, default=/Game/UI/Foundation/Textures/Icons/Athena/T-T-Icon-BR-GalileoCubeCrate}
 HiredNPCEliminatedTag {type=GameplayTag, default {tagName=SpecialEvent.HiredNPCEliminated}}
 IntroAnim {type=string, default=IntroAnim}
 isTargetTeam {type=boolean}
 MessageEventTag {type=GameplayTag, default {tagName=None}}
 Objective Category {type=EObjectiveCategories, default=Default}
 RankedQuestTag {type=GameplayTag, default {tagName=ChallengeCategory.Habanero}}
 TimerHandle {type=TimerHandle}
 UrgentAlertIconVectorParameterValueName {type=string, default=Color1}
]
widgets {
 info {inheritedWidgetCount=1, parentClass=/Script/FortniteUI.AthenaSpecialEventStartedOverlay, rootWidgetClass=&SizeBox, widgetCount=39}
 widgets [
  SizeBox_0 {
   widgetClassPath=&SizeBox
   widget=SizeBox_0
   properties {bOverride_MinDesiredHeight=true, bOverride_MinDesiredWidth=true, minDesiredHeight=200, minDesiredWidth=600}
  }
  SimpleAnimations {
   widgetClassPath=/Script/FortniteUI.FortSimpleWidgetAnimationsPanel
   bIsVariable=true
   parent=SizeBox_0
   widget=SimpleAnimations
   properties {animations [
     IntroAnim {
      bIsEnabled=true
      onAnimationsFinished=()
      targets [{
        alpha=&/SpecialEventStarted_IntroAnim_RenderOpacity
        bResetOnFinish=false
        bShouldAnimateAlpha=true
        bShouldAnimateScale=false
        bShouldAnimateTranslation=false
        scale=null
        targetName=Overlay_Container
        translation=&/SpecialEventStarted_IntroAnim_Translation
        materialScalarAnimations []
       }]
     }
     FadeAnim {
      bIsEnabled=true
      onAnimationsFinished=()
      targets [{
        alpha=&/SpecialEventStarted_FadeAnim_RenderOpacity
        bResetOnFinish=false
        bShouldAnimateAlpha=true
        bShouldAnimateScale=false
        bShouldAnimateTranslation=false
        scale=null
        targetName=Overlay_Container
        translation=&/SpecialEventStarted_FadeAnim_Translation
        materialScalarAnimations []
       }]
     }
    ]}
   slot {class=SizeBoxSlot, properties {horizontalAlignment=HAlign_Center, verticalAlignment=VAlign_Top}}
  }
  Overlay_Container {
   widgetClassPath=&Overlay
   bIsVariable=true
   parent=SimpleAnimations
   widget=Overlay_Container
   slot {class=FortSimpleWidgetAnimationsSlot}
  }
  HBox_MainBox {
   widgetClassPath=&VerticalBox
   bIsVariable=true
   parent=Overlay_Container
   widget=HBox_MainBox
   properties {renderTransformPivot=<0.5 0>}
   slot {class=OverlaySlot, properties {horizontalAlignment=HAlign_Fill, verticalAlignment=VAlign_Fill}}
  }
  HorizontalBox_1 {
   widgetClassPath=&HorizontalBox
   parent=HBox_MainBox
   widget=HorizontalBox_1
   properties {renderTransformPivot=<0.5 0.3>}
   slot {class=VerticalBoxSlot, properties {horizontalAlignment=HAlign_Center, padding=<ltrb 0 0 0 -10>}}
  }
  CommonWidgetSwitcher_IconType {
   widgetClassPath=&CommonWidgetSwitcherLegacy
   bIsVariable=true
   parent=HorizontalBox_1
   widget=CommonWidgetSwitcher_IconType
   properties {activeWidgetIndex=1, transitionDuration=0, renderTransform {angle=0, scale=<1 1>, shear=<0 0>, translation=<19 0>}}
   slot {class=HorizontalBoxSlot, properties {horizontalAlignment=HAlign_Left, verticalAlignment=VAlign_Center, padding=<ltrb 0 0 32 0>}}
  }
  Overlay_UrgentIcon {
   widgetClassPath=&Overlay
   bIsVariable=true
   parent=CommonWidgetSwitcher_IconType
   widget=Overlay_UrgentIcon
   slot {class=WidgetSwitcherSlot}
  }
  Image_UrgentPulse {
   widgetClassPath=&Image
   bIsVariable=true
   parent=Overlay_UrgentIcon
   widget=Image_UrgentPulse
   properties {
    visibility=HitTestInvisible
    brush {
     bIsDynamicallyLoaded=false
     drawAs=Image
     imageType=NoImage
     mirroring=NoMirror
     resourceName=None
     resourceObject=/Game/Athena/HUD/LTM/Infiltration/Materials/M_UI_GameplayDiamondPing_Pulse_Add_Inst
     tiling=NoTile
     imageSize=<32 32>
     margin=<ltrb 0 0 0 0>
     outlineSettings {
      bUseBrushTransparency=false
      roundingType=HalfHeightRadius
      width=0
      color {colorUseRule=UseColor_Specified, specifiedColor=<rgba 0 0 0 0>}
      cornerRadii=<0 0 0 0>
     }
     tintColor {colorUseRule=UseColor_Specified, specifiedColor=<rgba 0 0.7157 1 1>}
     uVRegion {bIsValid=false, max=<0 0>, min=<0 0>}
    }
   }
   slot {class=OverlaySlot, properties {horizontalAlignment=HAlign_Fill, verticalAlignment=VAlign_Fill, padding=<ltrb -15 -15 -15 -15>}}
  }
  Image_UrgentAlertBackground {
   widgetClassPath=&Image
   bIsVariable=true
   parent=Overlay_UrgentIcon
   widget=Image_UrgentAlertBackground
   properties {
    visibility=HitTestInvisible
    brush {
     bIsDynamicallyLoaded=false
     drawAs=Image
     imageType=NoImage
     mirroring=NoMirror
     resourceName=None
     resourceObject=/Game/Athena/HUD/Discoverability/materials/M_UI_UrgentAlert
     tiling=NoTile
     imageSize=<128 128>
     margin=<ltrb 0 0 0 0>
     outlineSettings {
      bUseBrushTransparency=false
      roundingType=HalfHeightRadius
      width=0
      color {colorUseRule=UseColor_Specified, specifiedColor=<rgba 0 0 0 0>}
      cornerRadii=<0 0 0 0>
     }
     tintColor {colorUseRule=UseColor_Specified, specifiedColor=<rgba 1 1 1 1>}
     uVRegion {bIsValid=false, max=<0 0>, min=<0 0>}
    }
   }
   slot {class=OverlaySlot, properties {horizontalAlignment=HAlign_Center, verticalAlignment=VAlign_Center}}
  }
  Text_UrgentIcon {
   widgetClassPath=&CommonTextBlock
   bIsVariable=true
   parent=Overlay_UrgentIcon
   widget=Text_UrgentIcon
   properties {
    justification=Center
    text="!"
    visibility=HitTestInvisible
    font {
     bForceMonospaced=false
     bMaterialIsStencil=false
     fontMaterial=null
     fontObject=&Burbank
     letterSpacing=0
     monospacedWidth=1
     size=48
     skewAmount=0
     typefaceFontName=BurbankRegular-Black
     outlineSettings {
      bApplyOutlineToDropShadows=false
      bMiteredCorners=false
      bSeparateFillAlpha=false
      outlineMaterial=null
      outlineSize=0
      outlineColor=<rgba 0 0 0 1>
     }
    }
    renderTransform {angle=0, scale=<1 1>, shear=<-8 0>, translation=<0 0>}
    shadowOffset=<0 0>
   }
   slot {class=OverlaySlot, properties {horizontalAlignment=HAlign_Center, verticalAlignment=VAlign_Center}}
  }
  Image_UrgentIcon {
   widgetClassPath=&Image
   bIsVariable=true
   parent=Overlay_UrgentIcon
   widget=Image_UrgentIcon
   properties {
    visibility=Hidden
    brush {
     bIsDynamicallyLoaded=false
     drawAs=Image
     imageType=NoImage
     mirroring=NoMirror
     resourceName=None
     resourceObject=null
     tiling=NoTile
     imageSize=<128 128>
     margin=<ltrb 0 0 0 0>
     outlineSettings {
      bUseBrushTransparency=false
      roundingType=HalfHeightRadius
      width=0
      color {colorUseRule=UseColor_Specified, specifiedColor=<rgba 0 0 0 0>}
      cornerRadii=<0 0 0 0>
     }
     tintColor {colorUseRule=UseColor_Specified, specifiedColor=<rgba 1 0.4564 0 1>}
     uVRegion {bIsValid=false, max=<0 0>, min=<0 0>}
    }
   }
   slot {class=OverlaySlot, properties {horizontalAlignment=HAlign_Center, verticalAlignment=VAlign_Center}}
  }
  Overlay_SocialIcon {
   widgetClassPath=&Overlay
   bIsVariable=true
   parent=CommonWidgetSwitcher_IconType
   widget=Overlay_SocialIcon
   slot {class=WidgetSwitcherSlot, properties {horizontalAlignment=HAlign_Right, verticalAlignment=VAlign_Center, padding=<ltrb 0 -64 0 -64>}}
  }
  Image_PlayerPulse {
   widgetClassPath=&Image
   bIsVariable=true
   parent=Overlay_SocialIcon
   widget=Image_PlayerPulse
   properties {
    visibility=HitTestInvisible
    brush {
     bIsDynamicallyLoaded=false
     drawAs=Image
     imageType=NoImage
     mirroring=NoMirror
     resourceName=None
     resourceObject=&M_UI_GameplayDiamondPing_Pulse_Add_Inst
     tiling=NoTile
     imageSize=<32 32>
     margin=<ltrb 0 0 0 0>
     outlineSettings {
      bUseBrushTransparency=false
      roundingType=HalfHeightRadius
      width=0
      color {colorUseRule=UseColor_Specified, specifiedColor=<rgba 0 0 0 0>}
      cornerRadii=<0 0 0 0>
     }
     tintColor {colorUseRule=UseColor_Specified, specifiedColor=<rgba 1 1 1 1>}
     uVRegion {bIsValid=false, max=<0 0>, min=<0 0>}
    }
    colorAndOpacity=<rgba 1 0.0169 0 1>
   }
   slot {class=OverlaySlot, properties {horizontalAlignment=HAlign_Fill, verticalAlignment=VAlign_Fill, padding=<ltrb -15 -15 -15 -15>}}
  }
  Image_PlayerPortrait {
   widgetClassPath=&Image
   bIsVariable=true
   parent=Overlay_SocialIcon
   widget=Image_PlayerPortrait
   properties {
    visibility=SelfHitTestInvisible
    brush {
     bIsDynamicallyLoaded=false
     drawAs=Image
     imageType=NoImage
     mirroring=NoMirror
     resourceName=None
     resourceObject=/Game/UI/SocialPanel/Art/Materials/UI_UserPortrait_Enemy
     tiling=NoTile
     imageSize=<128 128>
     margin=<ltrb 0 0 0 0>
     outlineSettings {
      bUseBrushTransparency=false
      roundingType=HalfHeightRadius
      width=0
      color {colorUseRule=UseColor_Specified, specifiedColor=<rgba 0 0 0 0>}
      cornerRadii=<0 0 0 0>
     }
     tintColor {colorUseRule=UseColor_Specified, specifiedColor=<rgba 1 1 1 1>}
     uVRegion {bIsValid=false, max=<0 0>, min=<0 0>}
    }
   }
   slot {class=OverlaySlot, properties {horizontalAlignment=HAlign_Center, verticalAlignment=VAlign_Center}}
  }
  Image_Elim {
   widgetClassPath=&Image
   bIsVariable=true
   parent=Overlay_SocialIcon
   widget=Image_Elim
   properties {
    visibility=HitTestInvisible
    brush {
     bIsDynamicallyLoaded=false
     drawAs=Image
     imageType=NoImage
     mirroring=NoMirror
     resourceName=None
     resourceObject=/Game/Athena/Items/Accolades/Materials/M_UI_Icon_Elim
     tiling=NoTile
     imageSize=<75 75>
     margin=<ltrb 0 0 0 0>
     outlineSettings {
      bUseBrushTransparency=false
      roundingType=HalfHeightRadius
      width=0
      color {colorUseRule=UseColor_Specified, specifiedColor=<rgba 0 0 0 0>}
      cornerRadii=<0 0 0 0>
     }
     tintColor {colorUseRule=UseColor_Specified, specifiedColor=<rgba 1 1 1 1>}
     uVRegion {bIsValid=false, max=<0 0>, min=<0 0>}
    }
    renderTransform {angle=-14, scale=<1 1>, shear=<0 0>, translation=<-23 -52>}
   }
   slot {class=OverlaySlot, properties {horizontalAlignment=HAlign_Center, verticalAlignment=VAlign_Bottom, padding=<ltrb -64 -64 -64 -64>}}
  }
  Overlay_EliminatedIcon {widgetClassPath=&Overlay, parent=CommonWidgetSwitcher_IconType, widget=Overlay_EliminatedIcon, slot {class=WidgetSwitcherSlot}}
  Image_ElimPortrait {
   widgetClassPath=&Image
   bIsVariable=true
   parent=Overlay_EliminatedIcon
   widget=Image_ElimPortrait
   properties {brush {
     bIsDynamicallyLoaded=false
     drawAs=Image
     imageType=NoImage
     mirroring=NoMirror
     resourceName=None
     resourceObject=&UI_UserPortrait_Enemy
     tiling=NoTile
     imageSize=<128 128>
     margin=<ltrb 0 0 0 0>
     outlineSettings {
      bUseBrushTransparency=false
      roundingType=HalfHeightRadius
      width=0
      color {colorUseRule=UseColor_Specified, specifiedColor=<rgba 0 0 0 0>}
      cornerRadii=<0 0 0 0>
     }
     tintColor {colorUseRule=UseColor_Specified, specifiedColor=<rgba 1 1 1 1>}
     uVRegion {bIsValid=false, max=<0 0>, min=<0 0>}
    }}
   slot {class=OverlaySlot, properties {horizontalAlignment=HAlign_Fill, verticalAlignment=VAlign_Fill}}
  }
  Image_sparks {
   widgetClassPath=&Image
   bIsVariable=true
   parent=Overlay_EliminatedIcon
   widget=Image_sparks
   properties {
    renderOpacity=0
    visibility=SelfHitTestInvisible
    brush {
     bIsDynamicallyLoaded=false
     drawAs=Image
     imageType=NoImage
     mirroring=NoMirror
     resourceName=None
     resourceObject=/Game/Athena/UI/InGame/Interrogation/Materials/M_UI_Starburst_Inst3
     tiling=NoTile
     imageSize=<256 256>
     margin=<ltrb 0 0 0 0>
     outlineSettings {
      bUseBrushTransparency=false
      roundingType=HalfHeightRadius
      width=0
      color {colorUseRule=UseColor_Specified, specifiedColor=<rgba 0 0 0 0>}
      cornerRadii=<0 0 0 0>
     }
     tintColor {colorUseRule=UseColor_Specified, specifiedColor=<rgba 1 1 1 1>}
     uVRegion {bIsValid=false, max=<0 0>, min=<0 0>}
    }
   }
   slot {class=OverlaySlot, properties {horizontalAlignment=HAlign_Center, verticalAlignment=VAlign_Center, padding=<ltrb -128 -128 -128 -128>}}
  }
  Image_shards {
   widgetClassPath=&Image
   bIsVariable=true
   parent=Overlay_EliminatedIcon
   widget=Image_shards
   properties {
    renderOpacity=0
    visibility=SelfHitTestInvisible
    brush {
     bIsDynamicallyLoaded=false
     drawAs=Image
     imageType=NoImage
     mirroring=NoMirror
     resourceName=None
     resourceObject=/Game/Athena/UI/InGame/Interrogation/Materials/M_UI_Starburst_Triangle
     tiling=NoTile
     imageSize=<200 200>
     margin=<ltrb 0 0 0 0>
     outlineSettings {
      bUseBrushTransparency=false
      roundingType=HalfHeightRadius
      width=0
      color {colorUseRule=UseColor_Specified, specifiedColor=<rgba 0 0 0 0>}
      cornerRadii=<0 0 0 0>
     }
     tintColor {colorUseRule=UseColor_Specified, specifiedColor=<rgba 1 1 1 1>}
     uVRegion {bIsValid=false, max=<0 0>, min=<0 0>}
    }
   }
   slot {class=OverlaySlot, properties {horizontalAlignment=HAlign_Center, verticalAlignment=VAlign_Center, padding=<ltrb -256 -256 -256 -256>}}
  }
  Overlay_Reward {
   widgetClassPath=&Overlay
   parent=Overlay_EliminatedIcon
   widget=Overlay_Reward
   slot {class=OverlaySlot, properties {horizontalAlignment=HAlign_Center, verticalAlignment=VAlign_Center, padding=<ltrb -128 -128 -128 -128>}}
  }
  Image_Gold {
   widgetClassPath=&Image
   bIsVariable=true
   parent=Overlay_Reward
   widget=Image_Gold
   properties {brush {
     bIsDynamicallyLoaded=false
     drawAs=Image
     imageType=NoImage
     mirroring=NoMirror
     resourceName=None
     resourceObject=/Game/UI/Foundation/Textures/Icons/Items/T-Icon-S-Loot-Wads_01-L
     tiling=NoTile
     imageSize=<128 128>
     margin=<ltrb 0 0 0 0>
     outlineSettings {
      bUseBrushTransparency=false
      roundingType=HalfHeightRadius
      width=0
      color {colorUseRule=UseColor_Specified, specifiedColor=<rgba 0 0 0 0>}
      cornerRadii=<0 0 0 0>
     }
     tintColor {colorUseRule=UseColor_Specified, specifiedColor=<rgba 1 1 1 1>}
     uVRegion {bIsValid=false, max=<0 0>, min=<0 0>}
    }}
   slot {class=OverlaySlot}
  }
  Image_spark1 {
   widgetClassPath=&Image
   parent=Overlay_Reward
   widget=Image_spark1
   properties {
    visibility=SelfHitTestInvisible
    brush {
     bIsDynamicallyLoaded=false
     drawAs=Image
     imageType=NoImage
     mirroring=NoMirror
     resourceName=None
     resourceObject=/Game/Athena/HUD/Materials/Victory/M_Flare_Cross_Animated
     tiling=NoTile
     imageSize=<64 32>
     margin=<ltrb 0 0 0 0>
     outlineSettings {
      bUseBrushTransparency=false
      roundingType=HalfHeightRadius
      width=0
      color {colorUseRule=UseColor_Specified, specifiedColor=<rgba 0 0 0 0>}
      cornerRadii=<0 0 0 0>
     }
     tintColor {colorUseRule=UseColor_Specified, specifiedColor=<rgba 1 1 1 1>}
     uVRegion {bIsValid=false, max=<0 0>, min=<0 0>}
    }
    renderTransform {angle=0, scale=<1 1>, shear=<0 0>, translation=<-28 -13>}
   }
   slot {class=OverlaySlot, properties {horizontalAlignment=HAlign_Center, verticalAlignment=VAlign_Center}}
  }
  Image_spark2 {
   widgetClassPath=&Image
   parent=Overlay_Reward
   widget=Image_spark2
   properties {
    visibility=SelfHitTestInvisible
    brush {
     bIsDynamicallyLoaded=false
     drawAs=Image
     imageType=NoImage
     mirroring=NoMirror
     resourceName=None
     resourceObject=&M_Flare_Cross_Animated
     tiling=NoTile
     imageSize=<128 32>
     margin=<ltrb 0 0 0 0>
     outlineSettings {
      bUseBrushTransparency=false
      roundingType=HalfHeightRadius
      width=0
      color {colorUseRule=UseColor_Specified, specifiedColor=<rgba 0 0 0 0>}
      cornerRadii=<0 0 0 0>
     }
     tintColor {colorUseRule=UseColor_Specified, specifiedColor=<rgba 1 1 1 1>}
     uVRegion {bIsValid=false, max=<0 0>, min=<0 0>}
    }
    renderTransform {angle=0, scale=<1 1>, shear=<0 0>, translation=<-5 11>}
   }
   slot {class=OverlaySlot, properties {horizontalAlignment=HAlign_Center, verticalAlignment=VAlign_Center}}
  }
  Image_spark3 {
   widgetClassPath=&Image
   parent=Overlay_Reward
   widget=Image_spark3
   properties {
    visibility=SelfHitTestInvisible
    brush {
     bIsDynamicallyLoaded=false
     drawAs=Image
     imageType=NoImage
     mirroring=NoMirror
     resourceName=None
     resourceObject=&M_Flare_Cross_Animated
     tiling=NoTile
     imageSize=<94.0591 23.3964>
     margin=<ltrb 0 0 0 0>
     outlineSettings {
      bUseBrushTransparency=false
      roundingType=HalfHeightRadius
      width=0
      color {colorUseRule=UseColor_Specified, specifiedColor=<rgba 0 0 0 0>}
      cornerRadii=<0 0 0 0>
     }
     tintColor {colorUseRule=UseColor_Specified, specifiedColor=<rgba 1 1 1 1>}
     uVRegion {bIsValid=false, max=<0 0>, min=<0 0>}
    }
    renderTransform {angle=0, scale=<1 1>, shear=<0 0>, translation=<-10 -42>}
   }
   slot {class=OverlaySlot, properties {horizontalAlignment=HAlign_Center, verticalAlignment=VAlign_Center}}
  }
  Image_spark4 {
   widgetClassPath=&Image
   parent=Overlay_Reward
   widget=Image_spark4
   properties {
    visibility=SelfHitTestInvisible
    brush {
     bIsDynamicallyLoaded=false
     drawAs=Image
     imageType=NoImage
     mirroring=NoMirror
     resourceName=None
     resourceObject=/Game/Athena/HUD/Materials/Victory/M_Flare_Spikeball_Rotating2
     tiling=NoTile
     imageSize=<128 128>
     margin=<ltrb 0 0 0 0>
     outlineSettings {
      bUseBrushTransparency=false
      roundingType=HalfHeightRadius
      width=0
      color {colorUseRule=UseColor_Specified, specifiedColor=<rgba 0 0 0 0>}
      cornerRadii=<0 0 0 0>
     }
     tintColor {colorUseRule=UseColor_Specified, specifiedColor=<rgba 1 1 1 1>}
     uVRegion {bIsValid=false, max=<0 0>, min=<0 0>}
    }
    renderTransform {angle=0, scale=<1 1>, shear=<0 0>, translation=<30 -7>}
   }
   slot {class=OverlaySlot, properties {horizontalAlignment=HAlign_Center, verticalAlignment=VAlign_Center, padding=<ltrb -64 -64 -64 -64>}}
  }
  Image_Glow {
   widgetClassPath=&Image
   bIsVariable=true
   parent=Overlay_EliminatedIcon
   widget=Image_Glow
   properties {
    brush {
     bIsDynamicallyLoaded=false
     drawAs=Image
     imageType=NoImage
     mirroring=NoMirror
     resourceName=None
     resourceObject=/Game/Athena/HUD/LTM/Wax/Materials/M_UI_TextGlow_Add_Gradient_Gold
     tiling=NoTile
     imageSize=<256 256>
     margin=<ltrb 0 0 0 0>
     outlineSettings {
      bUseBrushTransparency=false
      roundingType=HalfHeightRadius
      width=0
      color {colorUseRule=UseColor_Specified, specifiedColor=<rgba 0 0 0 0>}
      cornerRadii=<0 0 0 0>
     }
     tintColor {colorUseRule=UseColor_Specified, specifiedColor=<rgba 1 1 1 1>}
     uVRegion {bIsValid=false, max=<0 0>, min=<0 0>}
    }
    colorAndOpacity=<rgba 0.276 0.0027 0 1>
   }
   slot {class=OverlaySlot, properties {horizontalAlignment=HAlign_Center, verticalAlignment=VAlign_Center, padding=<ltrb -128 -128 -128 -128>}}
  }
  Overlay_NPCEliminatedIcon {widgetClassPath=&Overlay, parent=CommonWidgetSwitcher_IconType, widget=Overlay_NPCEliminatedIcon, slot {class=WidgetSwitcherSlot}}
  Image_FrameBorder {
   widgetClassPath=&Image
   bIsVariable=true
   parent=Overlay_NPCEliminatedIcon
   widget=Image_FrameBorder
   properties {
    brush {
     bIsDynamicallyLoaded=false
     drawAs=Image
     imageType=NoImage
     mirroring=NoMirror
     resourceName=None
     resourceObject=/Game/Athena/HUD/Discoverability/materials/Hired_NPC_Elim/MI_UI_NPCElim_BorderFrame
     tiling=NoTile
     imageSize=<32 32>
     margin=<ltrb 0 0 0 0>
     outlineSettings {
      bUseBrushTransparency=false
      roundingType=HalfHeightRadius
      width=0
      color {colorUseRule=UseColor_Specified, specifiedColor=<rgba 0 0 0 0>}
      cornerRadii=<0 0 0 0>
     }
     tintColor {colorUseRule=UseColor_Specified, specifiedColor=<rgba 1 1 1 1>}
     uVRegion {bIsValid=false, max=<0 0>, min=<0 0>}
    }
    renderTransform {angle=0, scale=<4 4>, shear=<0 0>, translation=<0 0>}
   }
   slot {class=OverlaySlot, properties {horizontalAlignment=HAlign_Center, verticalAlignment=VAlign_Center}}
  }
  Image_NPCElimPortrait {
   widgetClassPath=&Image
   bIsVariable=true
   parent=Overlay_NPCEliminatedIcon
   widget=Image_NPCElimPortrait
   properties {brush {
     bIsDynamicallyLoaded=false
     drawAs=Image
     imageType=NoImage
     mirroring=NoMirror
     resourceName=None
     resourceObject=/Game/Athena/HUD/Discoverability/materials/Hired_NPC_Elim/MI_UI_NPCElim_UserPortrait_Enemy_Inst
     tiling=NoTile
     imageSize=<128 128>
     margin=<ltrb 0 0 0 0>
     outlineSettings {
      bUseBrushTransparency=false
      roundingType=HalfHeightRadius
      width=0
      color {colorUseRule=UseColor_Specified, specifiedColor=<rgba 0 0 0 0>}
      cornerRadii=<0 0 0 0>
     }
     tintColor {colorUseRule=UseColor_Specified, specifiedColor=<rgba 1 1 1 1>}
     uVRegion {bIsValid=false, max=<0 0>, min=<0 0>}
    }}
   slot {class=OverlaySlot, properties {horizontalAlignment=HAlign_Fill, verticalAlignment=VAlign_Fill}}
  }
  Image_NPCElim_Glow {
   widgetClassPath=&Image
   bIsVariable=true
   parent=Overlay_NPCEliminatedIcon
   widget=Image_NPCElim_Glow
   properties {brush {
     bIsDynamicallyLoaded=false
     drawAs=Image
     imageType=NoImage
     mirroring=NoMirror
     resourceName=None
     resourceObject=/Game/Athena/HUD/Discoverability/materials/Hired_NPC_Elim/MI_UI_NPCElim_Glow
     tiling=NoTile
     imageSize=<256 256>
     margin=<ltrb 0 0 0 0>
     outlineSettings {
      bUseBrushTransparency=false
      roundingType=HalfHeightRadius
      width=0
      color {colorUseRule=UseColor_Specified, specifiedColor=<rgba 0 0 0 0>}
      cornerRadii=<0 0 0 0>
     }
     tintColor {colorUseRule=UseColor_Specified, specifiedColor=<rgba 1 1 1 1>}
     uVRegion {bIsValid=false, max=<0 0>, min=<0 0>}
    }}
   slot {class=OverlaySlot, properties {horizontalAlignment=HAlign_Center, verticalAlignment=VAlign_Center, padding=<ltrb -128 -128 -128 -128>}}
  }
  Image_X_RightLeft {
   widgetClassPath=&Image
   bIsVariable=true
   parent=Overlay_NPCEliminatedIcon
   widget=Image_X_RightLeft
   properties {
    brush {
     bIsDynamicallyLoaded=false
     drawAs=Image
     imageType=NoImage
     mirroring=NoMirror
     resourceName=None
     resourceObject=/Game/Athena/HUD/Discoverability/materials/Hired_NPC_Elim/MI_UI_NPCElim_RightLeftX
     tiling=NoTile
     imageSize=<90 90>
     margin=<ltrb 0 0 0 0>
     outlineSettings {
      bUseBrushTransparency=false
      roundingType=HalfHeightRadius
      width=0
      color {colorUseRule=UseColor_Specified, specifiedColor=<rgba 0 0 0 0>}
      cornerRadii=<0 0 0 0>
     }
     tintColor {colorUseRule=UseColor_Specified, specifiedColor=<rgba 1 1 1 1>}
     uVRegion {bIsValid=false, max=<0 0>, min=<0 0>}
    }
    renderTransformPivot=<1 0>
   }
   slot {class=OverlaySlot, properties {horizontalAlignment=HAlign_Center, verticalAlignment=VAlign_Center}}
  }
  Image_X_LeftRight {
   widgetClassPath=&Image
   bIsVariable=true
   parent=Overlay_NPCEliminatedIcon
   widget=Image_X_LeftRight
   properties {
    brush {
     bIsDynamicallyLoaded=false
     drawAs=Image
     imageType=NoImage
     mirroring=NoMirror
     resourceName=None
     resourceObject=/Game/Athena/HUD/Discoverability/materials/Hired_NPC_Elim/MI_UI_NPCElim_LeftRightX
     tiling=NoTile
     imageSize=<90 90>
     margin=<ltrb 0 0 0 0>
     outlineSettings {
      bUseBrushTransparency=false
      roundingType=HalfHeightRadius
      width=0
      color {colorUseRule=UseColor_Specified, specifiedColor=<rgba 0 0 0 0>}
      cornerRadii=<0 0 0 0>
     }
     tintColor {colorUseRule=UseColor_Specified, specifiedColor=<rgba 1 1 1 1>}
     uVRegion {bIsValid=false, max=<0 0>, min=<0 0>}
    }
    renderTransformPivot=<0 0>
   }
   slot {class=OverlaySlot, properties {horizontalAlignment=HAlign_Center, verticalAlignment=VAlign_Center}}
  }
  CommonBorder_0 {
   widgetClassPath=&CommonBorder
   parent=HorizontalBox_1
   widget=CommonBorder_0
   properties {clipping=ClipToBounds, horizontalAlignment=HAlign_Left, padding=<ltrb 0 0 0 0>}
   slot {class=HorizontalBoxSlot, properties {horizontalAlignment=HAlign_Center, verticalAlignment=VAlign_Center, padding=<ltrb 5 0 0 0>}}
  }
  VerticalBox_0 {
   widgetClassPath=&VerticalBox
   parent=CommonBorder_0
   widget=VerticalBox_0
   slot {class=BorderSlot, properties {horizontalAlignment=HAlign_Left, padding=<ltrb 0 0 0 0>}}
  }
  RichText_Title {
   widgetClassPath=&CommonRichTextBlock
   bIsVariable=true
   parent=VerticalBox_0
   widget=RichText_Title
   properties {
    bOverrideDefaultStyle=true
    justification=Center
    lineHeightPercentage=0.8
    text=Urgent Quest
    textStyleSet=&/UrgentQuestRichTextStyles
    textTransformPolicy=ToUpper
    wrapTextAt=512
    defaultTextStyleOverride {
     overflowPolicy=Clip
     transformPolicy=None
     colorAndOpacity {colorUseRule=UseColor_Specified, specifiedColor=<rgba 0.0168 0.6867 1 1>}
     font {
      bForceMonospaced=false
      bMaterialIsStencil=false
      fontMaterial=&M_UI_DiscoveryText_Inst
      fontObject=&Burbank
      letterSpacing=0
      monospacedWidth=1
      size=40
      skewAmount=0
      typefaceFontName=BurbankRegular-Black
      outlineSettings {
       bApplyOutlineToDropShadows=false
       bMiteredCorners=false
       bSeparateFillAlpha=false
       outlineMaterial=null
       outlineSize=2
       outlineColor=<rgba 0.0044 0.0723 0.1714 1>
      }
     }
     highlightColor {colorUseRule=UseColor_Specified, specifiedColor=<rgba 0 0 0 1>}
     highlightShape {
      bIsDynamicallyLoaded=false
      drawAs=Image
      imageType=NoImage
      mirroring=NoMirror
      resourceName=None
      resourceObject=null
      tiling=NoTile
      imageSize=<32 32>
      margin=<ltrb 0 0 0 0>
      outlineSettings {
       bUseBrushTransparency=false
       roundingType=HalfHeightRadius
       width=0
       color {colorUseRule=UseColor_Specified, specifiedColor=<rgba 0 0 0 0>}
       cornerRadii=<0 0 0 0>
      }
      tintColor {colorUseRule=UseColor_Specified, specifiedColor=<rgba 1 1 1 1>}
      uVRegion {bIsValid=false, max=<0 0>, min=<0 0>}
     }
     selectedBackgroundColor {colorUseRule=UseColor_Foreground, specifiedColor=<rgba 1 0 1 1>}
     shadowColorAndOpacity=<rgba 0 0 0 0.5>
     shadowOffset=<5 5>
     strikeBrush {
      bIsDynamicallyLoaded=false
      drawAs=Image
      imageType=NoImage
      mirroring=NoMirror
      resourceName=None
      resourceObject=null
      tiling=NoTile
      imageSize=<32 32>
      margin=<ltrb 0 0 0 0>
      outlineSettings {
       bUseBrushTransparency=false
       roundingType=HalfHeightRadius
       width=0
       color {colorUseRule=UseColor_Specified, specifiedColor=<rgba 0 0 0 0>}
       cornerRadii=<0 0 0 0>
      }
      tintColor {colorUseRule=UseColor_Specified, specifiedColor=<rgba 1 1 1 1>}
      uVRegion {bIsValid=false, max=<0 0>, min=<0 0>}
     }
     underlineBrush {
      bIsDynamicallyLoaded=false
      drawAs=Image
      imageType=NoImage
      mirroring=NoMirror
      resourceName=None
      resourceObject=null
      tiling=NoTile
      imageSize=<32 32>
      margin=<ltrb 0 0 0 0>
      outlineSettings {
       bUseBrushTransparency=false
       roundingType=HalfHeightRadius
       width=0
       color {colorUseRule=UseColor_Specified, specifiedColor=<rgba 0 0 0 0>}
       cornerRadii=<0 0 0 0>
      }
      tintColor {colorUseRule=UseColor_Specified, specifiedColor=<rgba 1 1 1 1>}
      uVRegion {bIsValid=false, max=<0 0>, min=<0 0>}
     }
    }
    renderTransform {angle=0, scale=<1 1>, shear=<-12 0>, translation=<0 0>}
   }
   slot {class=VerticalBoxSlot, properties {horizontalAlignment=HAlign_Left}}
  }
  RichText_SubTitle {
   widgetClassPath=&CommonRichTextBlock
   bIsVariable=true
   parent=VerticalBox_0
   widget=RichText_SubTitle
   properties {
    bOverrideDefaultStyle=true
    justification=Center
    lineHeightPercentage=0.8
    text=Started
    textTransformPolicy=ToUpper
    wrapTextAt=512
    defaultTextStyleOverride {
     overflowPolicy=Clip
     transformPolicy=None
     colorAndOpacity {colorUseRule=UseColor_Specified, specifiedColor=<rgba 1 1 1 1>}
     font {
      bForceMonospaced=false
      bMaterialIsStencil=false
      fontMaterial=&M_UI_DiscoveryText_Inst
      fontObject=&Burbank
      letterSpacing=0
      monospacedWidth=1
      size=24
      skewAmount=0
      typefaceFontName=BurbankRegular-Black
      outlineSettings {
       bApplyOutlineToDropShadows=false
       bMiteredCorners=false
       bSeparateFillAlpha=false
       outlineMaterial=/Game/Athena/HUD/Materials/M_UI_OutlineShine_White_Blue_Inst
       outlineSize=1
       outlineColor=<rgba 1 1 1 1>
      }
     }
     highlightColor {colorUseRule=UseColor_Specified, specifiedColor=<rgba 0 0 0 1>}
     highlightShape {
      bIsDynamicallyLoaded=false
      drawAs=Image
      imageType=NoImage
      mirroring=NoMirror
      resourceName=None
      resourceObject=null
      tiling=NoTile
      imageSize=<32 32>
      margin=<ltrb 0 0 0 0>
      outlineSettings {
       bUseBrushTransparency=false
       roundingType=HalfHeightRadius
       width=0
       color {colorUseRule=UseColor_Specified, specifiedColor=<rgba 0 0 0 0>}
       cornerRadii=<0 0 0 0>
      }
      tintColor {colorUseRule=UseColor_Specified, specifiedColor=<rgba 1 1 1 1>}
      uVRegion {bIsValid=false, max=<0 0>, min=<0 0>}
     }
     selectedBackgroundColor {colorUseRule=UseColor_Foreground, specifiedColor=<rgba 1 0 1 1>}
     shadowColorAndOpacity=<rgba 0 0 0 0.5>
     shadowOffset=<5 5>
     strikeBrush {
      bIsDynamicallyLoaded=false
      drawAs=Image
      imageType=NoImage
      mirroring=NoMirror
      resourceName=None
      resourceObject=null
      tiling=NoTile
      imageSize=<32 32>
      margin=<ltrb 0 0 0 0>
      outlineSettings {
       bUseBrushTransparency=false
       roundingType=HalfHeightRadius
       width=0
       color {colorUseRule=UseColor_Specified, specifiedColor=<rgba 0 0 0 0>}
       cornerRadii=<0 0 0 0>
      }
      tintColor {colorUseRule=UseColor_Specified, specifiedColor=<rgba 1 1 1 1>}
      uVRegion {bIsValid=false, max=<0 0>, min=<0 0>}
     }
     underlineBrush {
      bIsDynamicallyLoaded=false
      drawAs=Image
      imageType=NoImage
      mirroring=NoMirror
      resourceName=None
      resourceObject=null
      tiling=NoTile
      imageSize=<32 32>
      margin=<ltrb 0 0 0 0>
      outlineSettings {
       bUseBrushTransparency=false
       roundingType=HalfHeightRadius
       width=0
       color {colorUseRule=UseColor_Specified, specifiedColor=<rgba 0 0 0 0>}
       cornerRadii=<0 0 0 0>
      }
      tintColor {colorUseRule=UseColor_Specified, specifiedColor=<rgba 1 1 1 1>}
      uVRegion {bIsValid=false, max=<0 0>, min=<0 0>}
     }
    }
    renderTransform {angle=0, scale=<1 1>, shear=<-12 0>, translation=<0 0>}
   }
   slot {class=VerticalBoxSlot, properties {horizontalAlignment=HAlign_Left, verticalAlignment=VAlign_Bottom}}
  }
  RichText_Description {
   widgetClassPath=&CommonRichTextBlock
   bIsVariable=true
   parent=VerticalBox_0
   widget=RichText_Description
   properties {
    bOverrideDefaultStyle=true
    justification=Center
    lineHeightPercentage=0.8
    text=From PlayerX
    textTransformPolicy=ToUpper
    wrapTextAt=512
    defaultTextStyleOverride {
     overflowPolicy=Clip
     transformPolicy=None
     colorAndOpacity {colorUseRule=UseColor_Specified, specifiedColor=<rgba 1 1 1 1>}
     font {
      bForceMonospaced=false
      bMaterialIsStencil=false
      fontMaterial=&M_UI_DiscoveryText_Inst
      fontObject=&Burbank
      letterSpacing=0
      monospacedWidth=1
      size=16
      skewAmount=0
      typefaceFontName=BurbankRegular-Black
      outlineSettings {
       bApplyOutlineToDropShadows=false
       bMiteredCorners=false
       bSeparateFillAlpha=false
       outlineMaterial=&M_UI_OutlineShine_White_Blue_Inst
       outlineSize=1
       outlineColor=<rgba 1 1 1 1>
      }
     }
     highlightColor {colorUseRule=UseColor_Specified, specifiedColor=<rgba 0 0 0 1>}
     highlightShape {
      bIsDynamicallyLoaded=false
      drawAs=Image
      imageType=NoImage
      mirroring=NoMirror
      resourceName=None
      resourceObject=null
      tiling=NoTile
      imageSize=<32 32>
      margin=<ltrb 0 0 0 0>
      outlineSettings {
       bUseBrushTransparency=false
       roundingType=HalfHeightRadius
       width=0
       color {colorUseRule=UseColor_Specified, specifiedColor=<rgba 0 0 0 0>}
       cornerRadii=<0 0 0 0>
      }
      tintColor {colorUseRule=UseColor_Specified, specifiedColor=<rgba 1 1 1 1>}
      uVRegion {bIsValid=false, max=<0 0>, min=<0 0>}
     }
     selectedBackgroundColor {colorUseRule=UseColor_Foreground, specifiedColor=<rgba 1 0 1 1>}
     shadowColorAndOpacity=<rgba 0 0 0 0.5>
     shadowOffset=<5 5>
     strikeBrush {
      bIsDynamicallyLoaded=false
      drawAs=Image
      imageType=NoImage
      mirroring=NoMirror
      resourceName=None
      resourceObject=null
      tiling=NoTile
      imageSize=<32 32>
      margin=<ltrb 0 0 0 0>
      outlineSettings {
       bUseBrushTransparency=false
       roundingType=HalfHeightRadius
       width=0
       color {colorUseRule=UseColor_Specified, specifiedColor=<rgba 0 0 0 0>}
       cornerRadii=<0 0 0 0>
      }
      tintColor {colorUseRule=UseColor_Specified, specifiedColor=<rgba 1 1 1 1>}
      uVRegion {bIsValid=false, max=<0 0>, min=<0 0>}
     }
     underlineBrush {
      bIsDynamicallyLoaded=false
      drawAs=Image
      imageType=NoImage
      mirroring=NoMirror
      resourceName=None
      resourceObject=null
      tiling=NoTile
      imageSize=<32 32>
      margin=<ltrb 0 0 0 0>
      outlineSettings {
       bUseBrushTransparency=false
       roundingType=HalfHeightRadius
       width=0
       color {colorUseRule=UseColor_Specified, specifiedColor=<rgba 0 0 0 0>}
       cornerRadii=<0 0 0 0>
      }
      tintColor {colorUseRule=UseColor_Specified, specifiedColor=<rgba 1 1 1 1>}
      uVRegion {bIsValid=false, max=<0 0>, min=<0 0>}
     }
    }
    renderTransform {angle=0, scale=<1 1>, shear=<-12 0>, translation=<0 0>}
   }
   slot {class=VerticalBoxSlot, properties {horizontalAlignment=HAlign_Left, verticalAlignment=VAlign_Bottom}}
  }
  Spacer_AnimationPadding {
   widgetClassPath=&Spacer
   parent=HBox_MainBox
   widget=Spacer_AnimationPadding
   properties {size=<1 110>}
   slot {class=VerticalBoxSlot}
  }
  TutorialHighlight {widgetClassPath=&TutorialHighlight}
 ]
}
